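# Control Rig for the F_MED_VenomSoul female medium body, correcting deform and clothing bones as the character poses.
# A shared DeformRig function runs first, then a Sequence adjusts skirt, belts, cuffs, lapels and chain.
# Spherical pose readers on thighs, upper arms and left clavicle weight additive transform offsets, while parent constraints blend belt and chain bones.
# The crouchwalking, isEmoting and isMoving2dFloat variables gate some corrections through AlphaInterp.
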ControlRigBlueprint /BRCosmetics/Characters/Player/Female/Medium/Bodies/F_MED_VenomSoul/Meshes/F_MED_VenomSoul_CtrlRig
graphs [
 ContainedGraph {nodes [
   n1 = RigVMFunctionEntryNode()
   RigUnit_GetTransform = &RigUnit_GetTransform(Space=LocalSpace, bInitial=false, Item=@n1.Joint)
   n2 = &RigVMFunction_MathQuaternionToEuler(RotationOrder=ZYX, Value=@RigUnit_GetTransform.Transform.Rotation)
   > RigUnit_SetTransform = &RigUnit_SetTransform(
    Space=LocalSpace
    Value=(Rotation=(X=0.0132,Y=0.4099,Z=-0.0222,W=0.9118),Translation=(X=23.6045,Y=-6.37,Z=140.1115),Scale3D=(X=1,Y=1,Z=1))
    bInitial=false
    Item=@n1.Joint
    Weight=@n1.Weight
    Value.Translation.Z=&RigVMFunction_MathFloatAdd(B=@n1.Translate Z, A=@RigUnit_GetTransform.Transform.Translation.Z) {template="Add::Execute(in A,in B,out Result)"}.Result
    Value.Scale3D.Z=&RigVMFunction_MathFloatAdd(B=@n1.Scale Z, A=@RigUnit_GetTransform.Transform.Scale3D.Z) {template="Add::Execute(in A,in B,out Result)"}.Result
    Value.Translation.Y=&RigVMFunction_MathFloatAdd(B=@n1.Translate Y, A=@RigUnit_GetTransform.Transform.Translation.Y) {template="Add::Execute(in A,in B,out Result)"}.Result
    Value.Scale3D.Y=&RigVMFunction_MathFloatAdd(B=@n1.Scale Y, A=@RigUnit_GetTransform.Transform.Scale3D.Y) {template="Add::Execute(in A,in B,out Result)"}.Result
    Value.Translation.X=&RigVMFunction_MathFloatAdd(B=@n1.Translate X, A=@RigUnit_GetTransform.Transform.Translation.X) {template="Add::Execute(in A,in B,out Result)"}.Result
    Value.Scale3D.X=&RigVMFunction_MathFloatAdd(B=@n1.Scale X, A=@RigUnit_GetTransform.Transform.Scale3D.X) {template="Add::Execute(in A,in B,out Result)"}.Result
    Value.Rotation=&RigVMFunction_MathQuaternionFromEuler(
     Euler=(X=0,Y=0,Z=0)
     Euler.X=&RigVMFunction_MathFloatAdd(B=@n1.Rotate X, A=@n2.Result.X) {template="Add::Execute(in A,in B,out Result)"}.Result
     Euler.Y=&RigVMFunction_MathFloatAdd(B=@n1.Rotate Y, A=@n2.Result.Y) {template="Add::Execute(in A,in B,out Result)"}.Result
     Euler.Z=&RigVMFunction_MathFloatAdd(B=@n1.Rotate Z, A=@n2.Result.Z) {template="Add::Execute(in A,in B,out Result)"}.Result
    ).Result
   )
   > RigVMFunctionReturnNode()
  ]}
 New Function_ContainedGraph {nodes [
   n1 = RigVMFunctionEntryNode()
   RigVMFunction_MathBoolToInteger = &RigVMFunction_MathBoolToInteger(Value=@n1.onlyTwist)
   RigVMFunction_MathQuaternionSwingTwist = &RigVMFunction_MathQuaternionSwingTwist(
    twistAxis=@n1.twistAxis
    Input=&RigUnit_GetRelativeTransformForItem(bChildInitial=False, bParentInitial=False, child=@n1.child, parent=@n1.parent).RelativeTransform.Rotation
   )
   n2 = RigVMDispatchNode(
    Values=((),())
    Index=@RigVMFunction_MathBoolToInteger.Result
    Values.0=@RigVMFunction_MathQuaternionSwingTwist.Swing
    Values.1=@RigVMFunction_MathQuaternionSwingTwist.Twist
   ) {template="DISPATCH_RigVMDispatch_SelectInt32(in Index,in Values,out Result)"}
   RigVMFunction_MathQuaternionToEuler = &RigVMFunction_MathQuaternionToEuler(Value=@n2.Result, RotationOrder=@n1.Rotation Order)
   > DISPATCH_RigVMDispatch_Print = RigVMDispatchNode(ScreenColor=(R=1,G=1,B=1,A=1), ScreenDuration=1.5, Enabled=@n1.debugRotation, Value=@RigVMFunction_MathQuaternionToEuler.Result) {template="DISPATCH_RigVMDispatch_Print(in Prefix,in Value,in Enabled,in ScreenDuration,in ScreenColor)"}
   Cast = RigVMDispatchNode(Value=@n1.triggerAxis) {template="DISPATCH_RigVMDispatch_CastEnumToInt(in Value,out Result)"}
   DISPATCH_RigVMDispatch_SelectInt32 = RigVMDispatchNode(
    Values=(0,0,0)
    Index=@Cast.Result
    Values.0=@RigVMFunction_MathQuaternionToEuler.Result.X
    Values.1=@RigVMFunction_MathQuaternionToEuler.Result.Y
    Values.2=@RigVMFunction_MathQuaternionToEuler.Result.Z
   ) {template="DISPATCH_RigVMDispatch_SelectInt32(in Index,in Values,out Result)"}
   Remap = &RigVMFunction_MathDoubleRemap(
    bClamp=False
    Value=@DISPATCH_RigVMDispatch_SelectInt32.Result
    TargetMinimum=@n1.outputMinMax.X
    TargetMaximum=@n1.outputMinMax.Y
    SourceMinimum=@n1.rotationMinMax.X
    SourceMaximum=@n1.rotationMinMax.Y
   )
   Absolute = &RigVMFunction_MathDoubleAbs(Value=@Remap.Result)
   Multiply = &RigVMFunction_MathDoubleMul(
    A=&RigVMFunction_MathDoubleClamp(Value=RigVMDispatchNode(True=@Absolute.Result, Condition=@n1.absoluteValue, False=@Remap.Result) {template="DISPATCH_RigVMDispatch_If(in Condition,in True,in False,out Result)"}.Result).Result
    B=@n1.premult
   )
   > RigVMFunctionReturnNode(outputQuat=@n2.Result, outputWeight=@Multiply.Result)
  ]}
 RigVMFunctionLibrary {nodes [
   DeltaRotation = RigVMCollapseNode(
    ExecuteContext=()
    Rotation Order=ZYX
    child=(Type=Bone,Name="None")
    outputMinMax=(X=0,Y=1)
    parent=(Type=Bone,Name="None")
    premult=1
    rotationMinMax=(X=0,Y=0)
    twistAxis=(X=1,Y=0,Z=0)
   )
   Set Transforms for Bone = RigVMCollapseNode(
    ExecuteContext=()
    Joint=()
    Rotate X=0
    Rotate Y=0
    Rotate Z=0
    Scale X=0
    Scale Y=0
    Scale Z=0
    Translate X=0
    Translate Y=0
    Translate Z=0
    Weight=1
   )
  ]}
 RigVMFunction_Sequence_1_ContainedGraph {nodes [
   n1 = RigVMFunctionEntryNode()
   RigVMFunction_Sequence = &RigVMFunction_Sequence(ExecuteContext=@n1.ExecuteContext)
   n2 = &RigVMFunction_Sequence(ExecuteContext=@RigVMFunction_Sequence.B)
   n3 = &RigVMFunction_Sequence(ExecuteContext=@n2.B)
   n4 = &RigVMFunction_Sequence(ExecuteContext=@n3.B)
   RigVMFunctionReturnNode(A=@RigVMFunction_Sequence.A, B=@n2.A, C=@n3.A, D=@n4.A, E=@n4.B)
  ]}
 RigVMModel {nodes [
   RigUnit_BeginExecution = &RigUnit_BeginExecution()
   FN_DeformRig_FNC = RigVMFunctionReferenceNode(
    DeformRig_ Curve="(deformRig_l_upperarm_up_alpha_72_#1=False,deformRig_r_upperarm_up_alpha_73_#2=False,deformRig_l_upperarm_fr_alpha_74_#3=False,deformRig_r_upperarm_fr_alpha_75_CB51B8 ...(truncated)"
    DeformRig_Bool="(upperarm_1_#4=True,elbow_3_#5=True,wrist_7_#6=True,groin_8_#7=true,glute_17_#8=false,knee_9_40C408324CA26 ...(truncated)"
    DeformRig_Debug=(upperarm_debug_18_#4=False,wrist_debug_20_#6=False,lat_debug_26_#9=False,groinGlute_debug_27_#7=False)
    DeformRig_Value="(upperarm_up_18_#10=(Rotation=(X=0,Y=0,Z=0,W=1),Translation=(X=0,Y=0,Z=8),Scale3D=(X=0.7,Y=1,Z=1)),upperarm_fwd_fr_19_#11=(Rotation=(X=0,Y=0,Z=0,W=1),Translation=(X=0,Y=10,Z=0),Scale3D=(X=0.5,Y=1,Z= ...(truncated)"
    ExecuteContext=@RigUnit_BeginExecution.ExecutePin
   )
   > RigVMFunction_Sequence = RigVMAggregateNode()
   &RigUnit_ModifyTransforms(
    ItemToModify=((Item=(Type=Bone,Name="base_skirt_fr_02_l"),Transform=(Rotation=(X=0,Y=0,Z=0,W=1),Translation=(X=0,Y=0,Z=-1),Scale3D=(X=1,Y=1,Z=1))))
    Weight=1
    WeightMinimum=0
    ExecutePin=@RigVMFunction_Sequence.A
   )
   > &RigUnit_ModifyTransforms(
    ItemToModify=((Item=(Type=Bone,Name="dyn_skirt_fr_04_l"),Transform=(Rotation=(X=0,Y=0,Z=0.1736,W=0.9848),Translation=(X=0,Y=0,Z=0),Scale3D=(X=1,Y=1,Z=1))))
    Weight=1
    WeightMinimum=0
   )
   > &RigUnit_ParentConstraint(
    AdvancedSettings=(InterpolationType=Average,RotationOrderForFilter=XZY)
    Child=(Type=Bone,Name="dyn_belt_bk_06")
    Parents=((Item=(Type=Bone,Name="thigh_twist_01_l"),Weight=0.8),(Item=(Type=Bone,Name="dyn_belt_bk_05"),Weight=0.2))
    Weight=1
   )
   > &RigUnit_ParentConstraint(
    AdvancedSettings=(InterpolationType=Average,RotationOrderForFilter=XZY)
    Child=(Type=Bone,Name="dyn_belt_ft_03")
    Parents=((Item=(Type=Bone,Name="thigh_twist_01_l"),Weight=0.8),(Item=(Type=Bone,Name="dyn_belt_ft_02"),Weight=0.2))
    Weight=1
   )
   n1 = &RigUnit_SphericalPoseReader(
    ActiveRegionScaleFactors=(PositiveWidth=3.5,NegativeWidth=1,PositiveHeight=3.5,NegativeHeight=4)
    ActiveRegionSize=0.1
    Debug=(bDrawDebug=false,DebugScale=25,DebugSegments=20,DebugThickness=0.25)
    DriverAxis=(X=-1,Y=0,Z=0)
    DriverItem=(Type=Bone,Name="thigh_l")
    FalloffRegionScaleFactors=(PositiveWidth=1,NegativeWidth=0.9,PositiveHeight=1,NegativeHeight=1)
    FalloffSize=0.45
    FlipHeightScaling=true
    FlipWidthScaling=true
    OptionalParentItem=(Type=Bone,Name="None")
    RotationOffset=(X=90,Y=180,Z=0)
    ExecutePin=@RigVMFunction_Sequence.B
   )
   > Set Transforms for Bone_1_1_1_4_1_1 = RigVMFunctionReferenceNode(
    Joint=(Type=Bone,Name="dyn_belt_bk_03")
    Rotate X=0
    Rotate Y=0
    Rotate Z=0
    Scale X=0
    Scale Y=0
    Scale Z=0
    Translate X=0
    Translate Y=0
    Translate Z=10
    Weight=@n1.OutputParam
   )
   > n2 = &RigUnit_SphericalPoseReader(
    ActiveRegionScaleFactors=(PositiveWidth=3.5,NegativeWidth=0.5,PositiveHeight=3.5,NegativeHeight=4)
    ActiveRegionSize=0.1
    Debug=(bDrawDebug=false,bDrawLocalAxes=false,DebugScale=25,DebugSegments=20,DebugThickness=0.25)
    DriverAxis=(X=1,Y=0,Z=0)
    DriverItem=(Type=Bone,Name="thigh_r")
    FalloffRegionScaleFactors=(PositiveWidth=1,NegativeWidth=1.1,PositiveHeight=1,NegativeHeight=1)
    FalloffSize=0.45
    FlipHeightScaling=true
    FlipWidthScaling=true
    OptionalParentItem=(Type=Bone,Name="None")
    RotationOffset=(X=-90,Y=0,Z=0)
   )
   > Set Transforms for Bone_1_1_1_4_1_1_1_1 = RigVMFunctionReferenceNode(
    Joint=(Type=Bone,Name="dyn_belt_bk_06")
    Rotate X=0
    Rotate Y=0
    Rotate Z=0
    Scale X=0
    Scale Y=0
    Scale Z=0
    Translate X=0
    Translate Y=13
    Translate Z=0
    Weight=@n2.OutputParam
   )
   > n3 = &RigUnit_SphericalPoseReader(
    ActiveRegionScaleFactors=(PositiveWidth=1,NegativeWidth=1,PositiveHeight=1,NegativeHeight=1)
    ActiveRegionSize=0.2
    Debug=(bDrawDebug=false,bDraw2D=False,bDrawLocalAxes=False,DebugScale=25,DebugSegments=20,DebugThickness=0.25)
    DriverAxis=(X=1,Y=0,Z=0)
    DriverItem=(Type=Bone,Name="thigh_r")
    FalloffRegionScaleFactors=(PositiveWidth=1,NegativeWidth=1,PositiveHeight=1,NegativeHeight=1)
    FalloffSize=0.45
    FlipHeightScaling=False
    FlipWidthScaling=false
    OptionalParentItem=(Type=Bone,Name="pelvis")
    RotationOffset=(X=-100,Y=0,Z=0)
   )
   > &RigUnit_ModifyTransforms(
    ItemToModify="((Item=(Type=Bone,Name=\"base_belt_ft_01\"),Transform=(Rotation=(X=0,Y=0,Z=0,W=1),Translation=(X=0,Y=-10,Z=0),Scale3D=(X=1,Y=1,Z=1))),(Item=(Type=Bone,Name=\"base_belt_ft_02\"),Transform=(Rotation=(X=0,Y=0,Z=0,W=1),Translation=(X=0,Y=-5,Z=0),Scale3D=(X=1,Y=1,Z ...(truncated)"
    WeightMinimum=0
    Weight=@n3.OutputParam
   )
   n4 = &RigVMFunction_AlphaInterp(
    InterpSpeedDecreasing=5
    InterpSpeedIncreasing=5
    Scale=1
    bClampResult=False
    Value=&RigVMFunction_MathDoubleMul(A=@n3.OutputParam, B=RigVMVariableNode() {variable=crouchwalking}.Value).Result
   )
   > &RigUnit_ModifyTransforms(
    ItemToModify="((Item=(Type=Bone,Name=\"base_belt_ft_03\"),Transform=(Rotation=(X=0,Y=0,Z=0,W=1),Translation=(X=0,Y=2,Z=5),Scale3D=(X=1,Y=1,Z=1))),(Item=(Type=Bone,Name=\"base_belt_ft_02\"),Transform=(Rotation=(X=0,Y=0,Z=0,W=1),Translation=(X=0,Y=0,Z=5),Scale3D=(X=1,Y=1,Z=1) ...(truncated)"
    WeightMinimum=0
    Weight=@n4.Result
   )
   > n5 = &RigUnit_SphericalPoseReader(
    ActiveRegionScaleFactors=(PositiveWidth=1,NegativeWidth=1,PositiveHeight=1,NegativeHeight=1)
    ActiveRegionSize=0.2
    Debug=(bDrawDebug=false,bDraw2D=False,bDrawLocalAxes=False,DebugScale=25,DebugSegments=20,DebugThickness=0.25)
    DriverAxis=(X=-1,Y=0,Z=0)
    DriverItem=(Type=Bone,Name="thigh_l")
    FalloffRegionScaleFactors=(PositiveWidth=1,NegativeWidth=1,PositiveHeight=1,NegativeHeight=1)
    FalloffSize=0.4
    FlipHeightScaling=False
    FlipWidthScaling=False
    OptionalParentItem=(Type=Bone,Name="pelvis")
    RotationOffset=(X=120,Y=0,Z=0)
   )
   > &RigUnit_ModifyTransforms(
    ItemToModify="((Item=(Type=Bone,Name=\"base_belt_bk_07\"),Transform=(Rotation=(X=0,Y=0,Z=0,W=1),Translation=(X=0,Y=0,Z=-5),Scale3D=(X=1,Y=1,Z=1))),(Item=(Type=Bone,Name=\"base_belt_ft_06\"),Transform=(Rotation=(X=0.4226,Y=0,Z=0,W=0.9063),Translation=(X=0,Y=0,Z=-1.5),Scale3D ...(truncated)"
    WeightMinimum=0
    Weight=@n5.OutputParam
   )
   > n6 = &RigUnit_SphericalPoseReader(
    ActiveRegionScaleFactors=(PositiveWidth=1,NegativeWidth=1,PositiveHeight=1,NegativeHeight=1)
    ActiveRegionSize=0.15
    Debug=(bDrawDebug=false,bDraw2D=False,bDrawLocalAxes=False,DebugScale=25,DebugSegments=20,DebugThickness=0.25)
    DriverAxis=(X=1,Y=0,Z=0)
    DriverItem=(Type=Bone,Name="thigh_r")
    FalloffRegionScaleFactors=(PositiveWidth=1,NegativeWidth=1,PositiveHeight=1,NegativeHeight=1)
    FalloffSize=0.4
    FlipHeightScaling=False
    FlipWidthScaling=false
    OptionalParentItem=(Type=Bone,Name="pelvis")
    RotationOffset=(X=-50,Y=0,Z=0)
   )
   n7 = &RigVMFunction_AlphaInterp(
    InterpSpeedDecreasing=5
    InterpSpeedIncreasing=5
    Scale=1
    bClampResult=False
    bInterpResult=False
    Value=&RigVMFunction_MathDoubleMul(A=@n6.OutputParam, B=RigVMVariableNode() {variable=isEmoting}.Value).Result
   )
   > &RigUnit_ModifyTransforms(
    ItemToModify="((Item=(Type=Bone,Name=\"base_belt_ft_01\"),Transform=(Rotation=(X=0,Y=0,Z=0,W=1),Translation=(X=0,Y=-3,Z=0),Scale3D=(X=1,Y=1,Z=1))),(Item=(Type=Bone,Name=\"base_belt_ft_02\"),Transform=(Rotation=(X=0,Y=0,Z=0,W=1),Translation=(X=0,Y=-4,Z=0),Scale3D=(X=1,Y=1,Z= ...(truncated)"
    WeightMinimum=0
    Weight=@n7.Result
   )
   n8 = &RigUnit_SphericalPoseReader(
    ActiveRegionScaleFactors=(PositiveWidth=1.5,NegativeWidth=1,PositiveHeight=2,NegativeHeight=2)
    ActiveRegionSize=0.15
    Debug=(bDrawDebug=false,bDrawLocalAxes=True,DebugScale=25,DebugSegments=20,DebugThickness=0.25)
    DriverAxis=(X=-1,Y=0,Z=0)
    DriverItem=(Type=Bone,Name="upperarm_r")
    FalloffRegionScaleFactors=(PositiveWidth=1,NegativeWidth=1,PositiveHeight=1,NegativeHeight=1)
    FalloffSize=0.4
    FlipHeightScaling=true
    FlipWidthScaling=true
    OptionalParentItem=(Type=Bone,Name="clavicle_r")
    RotationOffset=(X=-80,Y=20,Z=20)
    ExecutePin=@RigVMFunction_Sequence.C
   )
   > &RigUnit_ModifyTransforms(
    ItemToModify=((Item=(Type=Bone,Name="dyn_cuff_r_02_MSTR"),Transform=(Rotation=(X=-0.0449,Y=0.2549,Z=0.1677,W=0.9513),Translation=(X=0,Y=0,Z=0),Scale3D=(X=1,Y=1,Z=1))))
    WeightMinimum=0
    Weight=@n8.OutputParam
   )
   > &RigUnit_ModifyTransforms(
    ItemToModify=((Item=(Type=Bone,Name="dyn_cuff_r_04_MSTR"),Transform=(Rotation=(X=0.0449,Y=0.2549,Z=-0.1677,W=0.9513),Translation=(X=0,Y=0,Z=0),Scale3D=(X=1,Y=1,Z=1))))
    WeightMinimum=0
    Weight=@n8.OutputParam
   )
   > n9 = &RigUnit_SphericalPoseReader(
    ActiveRegionScaleFactors=(PositiveWidth=1,NegativeWidth=1,PositiveHeight=1,NegativeHeight=1)
    ActiveRegionSize=0.25
    Debug=(bDrawDebug=false,bDrawLocalAxes=true,DebugScale=25,DebugSegments=20,DebugThickness=0.25)
    DriverAxis=(X=1,Y=0,Z=0)
    DriverItem=(Type=Bone,Name="upperarm_l")
    FalloffRegionScaleFactors=(PositiveWidth=1,NegativeWidth=1,PositiveHeight=1,NegativeHeight=1)
    FalloffSize=0.35
    FlipWidthScaling=false
    OptionalParentItem=(Type=Bone,Name="clavicle_l")
    RotationOffset=(X=30,Y=0,Z=0)
   )
   > &RigUnit_ModifyTransforms(
    ItemToModify=((Item=(Type=Bone,Name="dyn_cuff_l_02"),Transform=(Rotation=(X=0.1094,Y=0.2346,Z=-0.4082,W=0.8754),Translation=(X=0,Y=0,Z=0),Scale3D=(X=1,Y=1,Z=1))))
    WeightMinimum=0
    Weight=@n9.OutputParam
   )
   > &RigUnit_ModifyTransforms(
    ItemToModify=((Item=(Type=Bone,Name="dyn_cuff_l_04"),Transform=(Rotation=(X=-0.056,Y=0.2091,Z=0.2527,W=0.943),Translation=(X=0,Y=0,Z=0),Scale3D=(X=1,Y=1,Z=1))))
    WeightMinimum=0
    Weight=@n9.OutputParam
   )
   > n10 = &RigUnit_SphericalPoseReader(
    ActiveRegionScaleFactors=(PositiveWidth=1,NegativeWidth=1,PositiveHeight=1,NegativeHeight=1)
    ActiveRegionSize=0.4
    Debug=(bDrawDebug=false,bDraw2D=False,bDrawLocalAxes=False,DebugScale=25,DebugSegments=20,DebugThickness=0.25)
    DriverAxis=(X=-1,Y=0,Z=0)
    DriverItem=(Type=Bone,Name="thigh_l")
    FalloffRegionScaleFactors=(PositiveWidth=1,NegativeWidth=1,PositiveHeight=1,NegativeHeight=1)
    FalloffSize=0.6
    FlipHeightScaling=False
    FlipWidthScaling=False
    OptionalParentItem=(Type=Bone,Name="pelvis")
    RotationOffset=(X=120,Y=0,Z=0)
   )
   Multiply = &RigVMFunction_MathDoubleMul(A=@n10.OutputParam, B=RigVMVariableNode() {variable=crouchwalking}.Value)
   AlphaInterp = &RigVMFunction_AlphaInterp(InterpSpeedDecreasing=5, InterpSpeedIncreasing=5, Scale=1, bClampResult=False, Value=@Multiply.Result)
   > &RigUnit_ModifyTransforms(
    ItemToModify="((Item=(Type=Bone,Name=\"dyn_skirt_fr_01_l_MSTR\"),Transform=(Rotation=(X=0.0368,Y=-0.421,Z=0.079,W=0.9029),Translation=(X=0,Y=0,Z=0),Scale3D=(X=1,Y=1,Z=1))),(Item=(Type=Bone,Name=\"dyn_skirt_fr_04_l\"),Transform=(Rotation=(X=0.056,Y=-0.6403,Z=0.0668,W=0.7631) ...(truncated)"
    WeightMinimum=0
    Weight=@AlphaInterp.Result
   )
   n11 = &RigVMFunction_AlphaInterp(
    InterpSpeedDecreasing=5
    InterpSpeedIncreasing=5
    Scale=1
    bClampResult=False
    Value=&RigVMFunction_MathDoubleMul(A=@n10.OutputParam, B=RigVMVariableNode() {variable=isMoving2dFloat}.Value).Result
   )
   > &RigUnit_ModifyTransforms(
    ItemToModify="((Item=(Type=Bone,Name=\"dyn_skirt_fr_01_l_MSTR\"),Transform=(Rotation=(X=0.0734,Y=-0.4162,Z=0.1574,W=0.8925),Translation=(X=0,Y=0,Z=0),Scale3D=(X=1,Y=1,Z=1))),(Item=(Type=Bone,Name=\"dyn_skirt_fr_04_l\"),Transform=(Rotation=(X=0,Y=0,Z=0,W=1),Translation=(X=0, ...(truncated)"
    WeightMinimum=0
    Weight=@n11.Result
   )
   &RigUnit_ParentConstraint(
    AdvancedSettings=(InterpolationType=Average,RotationOrderForFilter=XZY)
    Child=(Type=Bone,Name="dyn_chain_03")
    Parents=((Item=(Type=Bone,Name="spine_04"),Weight=0.5),(Item=(Type=Bone,Name="dyn_chain_02"),Weight=0.5))
    Weight=1
    ExecutePin=@RigVMFunction_Sequence.D
   )
   > &RigUnit_ParentConstraint(
    AdvancedSettings=(InterpolationType=Average,RotationOrderForFilter=XZY)
    Child=(Type=Bone,Name="dyn_chain_02")
    Parents=((Item=(Type=Bone,Name="spine_04"),Weight=0.5),(Item=(Type=Bone,Name="dyn_chain_02"),Weight=0.5))
    Weight=1
   )
   > n12 = &RigUnit_SphericalPoseReader(
    ActiveRegionScaleFactors=(PositiveWidth=1,NegativeWidth=1,PositiveHeight=1.6,NegativeHeight=0.5)
    ActiveRegionSize=0.15
    Debug=(bDrawDebug=false,bDrawLocalAxes=false,DebugScale=25,DebugSegments=20,DebugThickness=0.25)
    DriverAxis=(X=1,Y=0,Z=0)
    DriverItem=(Type=Bone,Name="upperarm_l")
    FalloffRegionScaleFactors=(PositiveWidth=1,NegativeWidth=1,PositiveHeight=1,NegativeHeight=0.5)
    FalloffSize=0.4
    FlipWidthScaling=false
    OptionalParentItem=(Type=Bone,Name="None")
    RotationOffset=(X=85,Y=0,Z=-30)
   )
   Subtract = &RigVMFunction_MathDoubleSub(A=1, B=RigVMVariableNode() {variable=isEmoting}.Value)
   Absolute = &RigVMFunction_MathDoubleAbs(Value=@Subtract.Result)
   n13 = &RigVMFunction_AlphaInterp(InterpSpeedDecreasing=5, InterpSpeedIncreasing=5, Scale=1, bClampResult=False, bInterpResult=False, Value=@Absolute.Result)
   > &RigUnit_ModifyTransforms(
    ItemToModify=((Item=(Type=Bone,Name="dyn_lapel_l"),Transform=(Rotation=(X=0,Y=0,Z=0.4226,W=0.9063),Translation=(X=0,Y=2,Z=0),Scale3D=(X=1,Y=1,Z=1))))
    Weight=&RigVMFunction_MathDoubleMul(B=@n13.Result, A=@n12.OutputParam).Result
   )
   > n14 = &RigUnit_SphericalPoseReader(
    ActiveRegionScaleFactors=(PositiveWidth=1,NegativeWidth=1,PositiveHeight=0.7,NegativeHeight=2)
    ActiveRegionSize=0.15
    Debug=(bDrawDebug=false,bDrawLocalAxes=True,DebugScale=25,DebugSegments=20,DebugThickness=0.25)
    DriverAxis=(X=-1,Y=0,Z=0)
    DriverItem=(Type=Bone,Name="upperarm_r")
    FalloffRegionScaleFactors=(PositiveWidth=1,NegativeWidth=1,PositiveHeight=0.7,NegativeHeight=1)
    FalloffSize=0.4
    FlipHeightScaling=true
    FlipWidthScaling=true
    OptionalParentItem=(Type=Bone,Name="clavicle_r")
    RotationOffset=(X=-80,Y=0,Z=20)
   )
   > &RigUnit_ModifyTransforms(
    ItemToModify=((Item=(Type=Bone,Name="dyn_lapel_r"),Transform=(Rotation=(X=-0.1485,Y=0.212,Z=-0.7912,W=0.554),Translation=(X=0,Y=-2,Z=0),Scale3D=(X=1,Y=1,Z=1))))
    Weight=@n14.OutputParam
   )
   > &RigUnit_ModifyTransforms(
    ItemToModify=((Item=(Type=Bone,Name="dyn_lapel_lower_r"),Transform=(Rotation=(X=0,Y=0,Z=-0.866,W=0.5),Translation=(X=0,Y=-2,Z=0),Scale3D=(X=1,Y=1,Z=1))))
    Weight=@n14.OutputParam
   )
   > n15 = &RigUnit_SphericalPoseReader(
    ActiveRegionScaleFactors=(PositiveWidth=0.4,NegativeWidth=3,PositiveHeight=1.5,NegativeHeight=1)
    ActiveRegionSize=0.1
    Debug=(bDrawDebug=true,bDrawLocalAxes=false,DebugScale=25,DebugSegments=20,DebugThickness=0.25)
    DriverAxis=(X=1,Y=0,Z=0)
    DriverItem=(Type=Bone,Name="upperarm_l")
    FalloffRegionScaleFactors=(PositiveWidth=0.2,NegativeWidth=2,PositiveHeight=0.8,NegativeHeight=1.5)
    FalloffSize=0.3
    FlipWidthScaling=false
    OptionalParentItem=(Type=Bone,Name="clavicle_l")
    RotationOffset=(X=85,Y=0,Z=-40)
   )
   > n16 = &RigUnit_SphericalPoseReader(
    ActiveRegionScaleFactors=(PositiveWidth=1,NegativeWidth=1,PositiveHeight=1,NegativeHeight=1)
    ActiveRegionSize=0.1
    Debug=(bDrawDebug=true,bDrawLocalAxes=false,DebugScale=25,DebugSegments=20,DebugThickness=0.25)
    DriverAxis=(X=1,Y=0,Z=0)
    DriverItem=(Type=Bone,Name="clavicle_l")
    FalloffRegionScaleFactors=(PositiveWidth=1,NegativeWidth=1,PositiveHeight=1,NegativeHeight=1)
    FalloffSize=0.2
    FlipWidthScaling=false
    OptionalParentItem=(Type=Bone,Name="spine_05")
    RotationOffset=(X=0,Y=-85,Z=40)
   )
   n17 = &RigVMFunction_AlphaInterp(
    InterpSpeedDecreasing=5
    InterpSpeedIncreasing=5
    Scale=1
    bClampResult=False
    bInterpResult=False
    Value=RigVMVariableNode() {variable=isEmoting}.Value
   )
   Maximum = &RigVMFunction_MathDoubleMax(A=@n15.OutputParam, B=@n16.OutputParam)
   n18 = &RigVMFunction_MathDoubleMul(B=@n17.Result, A=@Maximum.Result)
   > &RigUnit_ModifyTransforms(
    ItemToModify=((Item=(Type=Bone,Name="dyn_lapel_l"),Transform=(Rotation=(X=0,Y=0,Z=0.4226,W=0.9063),Translation=(X=0,Y=4,Z=0),Scale3D=(X=1,Y=1,Z=1))))
    Weight=@n18.Result
   )
   > &RigUnit_ModifyTransforms(
    ItemToModify=((Item=(Type=Bone,Name="dyn_cuff_l_02"),Transform=(Rotation=(X=-0.4162,Y=0.1574,Z=0.0734,W=0.8925),Translation=(X=0,Y=0,Z=-5),Scale3D=(X=1,Y=1,Z=1))))
    WeightMinimum=0
    Weight=@n18.Result
   )
   > &RigUnit_ModifyTransforms(
    ItemToModify=((Item=(Type=Bone,Name="dyn_cuff_l_04"),Transform=(Rotation=(X=-0.4568,Y=0.0034,Z=0.4568,W=0.7633),Translation=(X=0,Y=0,Z=0),Scale3D=(X=1,Y=1,Z=1))))
    WeightMinimum=0
    Weight=@n18.Result
   )
  ]}
]
hierarchy [
 root {loc=<0 0 0>, rot=<0 0 0 1>}
 attach {parent=root, loc=<0 0 0>, rot=<0 0 0 1>}
 pelvis {parent=root, loc=<0 -0.5301 98.0571>, rot=<0.0127 0.707 -0.0127 -0.707>}
 spine_01 {parent=pelvis, loc=<6.1208 -0.9328 0>, rot=<0 0 0.034 -0.9994>}
 spine_02 {parent=spine_01, loc=<6.5996 -0.2095 0>, rot=<0 0 0.0238 -0.9997>}
 spine_03 {parent=spine_02, loc=<6.4914 -0.1237 0>, rot=<0 0 0.0149 -0.9999>}
 spine_04 {parent=spine_03, loc=<6.9598 0.0794 0>, rot=<0 0 -0.0132 -0.9999>}
 spine_05 {parent=spine_04, loc=<6.9987 0 0>, rot=<0 0 -0.0015 -1>}
 clavicle_l {parent=spine_05, loc=<5.4831 -0.1053 -3.2946>, rot=<0.0465 -0.7289 -0.009 -0.6829>}
 upperarm_l {parent=clavicle_l, loc=<9.8658 -1.2136 -0.4105>, rot=<0.0164 -0.3245 -0.0269 -0.9454>}
 lowerarm_l {parent=upperarm_l, loc=<26.465 -0.8345 0.259>, rot=<0.0036 0.0161 -0.2298 -0.9731>}
 hand_l {parent=lowerarm_l, loc=<22.7355 -0.3203 -0.6105>, rot=<-0.6784 -0.0223 0.0735 -0.7307>}
 index_metacarpal_l {parent=hand_l, loc=<1.6535 0.0183 -2.96>, rot=<0.0158 -0.0124 -0.0173 -0.9996>}
 index_01_l {parent=index_metacarpal_l, loc=<8.3008 0.0357 -0.0524>, rot=<-0.0044 0.0485 0.0924 -0.9945>}
 index_02_l {parent=index_01_l, loc=<3.3587 -0.3004 0.0452>, rot=<-0.002 0.0027 0.3225 -0.9466>}
 index_03_l {parent=index_02_l, loc=<2.5578 -0.123 -0.0252>, rot=<-0.002 0.0067 0.2558 -0.9667>}
 middle_metacarpal_l {parent=hand_l, loc=<2.1856 -0.1725 -0.5984>, rot=<0.006 0.0168 -0.0329 -0.9993>}
 middle_01_l {parent=middle_metacarpal_l, loc=<7.3275 -0.2447 0.0129>, rot=<-0.0208 0.0042 0.0858 -0.9961>}
 middle_02_l {parent=middle_01_l, loc=<4.0123 0.1401 0.1175>, rot=<-0.0218 0.0258 0.3512 -0.9357>}
 middle_03_l {parent=middle_02_l, loc=<2.6077 0.0534 0.1373>, rot=<0.0082 0.0188 0.1708 -0.9851>}
 pinky_metacarpal_l {parent=hand_l, loc=<1.7616 -0.9753 3.3814>, rot=<-0.0397 0.0814 -0.069 -0.9935>}
 pinky_01_l {parent=pinky_metacarpal_l, loc=<7.1709 -0.4831 -0.1094>, rot=<-0.0654 -0.0035 0.2242 -0.9723>}
 pinky_02_l {parent=pinky_01_l, loc=<2.8274 0.0893 -0.1043>, rot=<0.0037 0.0011 0.2854 -0.9584>}
 pinky_03_l {parent=pinky_02_l, loc=<2.2046 0.0307 -0.0078>, rot=<-0.0002 0.0098 0.1445 -0.9895>}
 ring_metacarpal_l {parent=hand_l, loc=<2.2074 -0.3625 1.3638>, rot=<-0.0284 0.0722 -0.0458 -0.9959>}
 ring_01_l {parent=ring_metacarpal_l, loc=<7.0295 -0.6197 -0.1446>, rot=<-0.0489 -0.0252 0.1892 -0.9804>}
 ring_02_l {parent=ring_01_l, loc=<3.6174 0.4647 -0.1027>, rot=<0 0 0.2806 -0.9598>}
 ring_03_l {parent=ring_02_l, loc=<2.4152 0.114 0.0199>, rot=<0 0 0.1189 -0.9929>}
 thumb_01_l {parent=hand_l, loc=<1.3118 -0.3191 -2.459>, rot=<0.6546 -0.4042 0.0664 -0.6354>}
 thumb_02_l {parent=thumb_01_l, loc=<5.9336 -0.0448 0.0401>, rot=<-0.0037 0.0213 0.298 -0.9543>}
 thumb_03_l {parent=thumb_02_l, loc=<2.2223 0.3682 -0.0907>, rot=<0.0118 -0.013 0.0066 -0.9998>}
 weapon_l {parent=hand_l, loc=<-0.1974 -3.4111 0.3811>, rot=<0 0 0 1>}
 deform_wrist_dn_base_l {parent=hand_l, loc=<0 -1.5326 0>, rot=<0 0 0 -1>}
 deform_wrist_dn_l {parent=deform_wrist_dn_base_l, loc=<0 0 0>, rot=<0 0 0 1>}
 deform_wrist_up_base_l {parent=hand_l, loc=<0 2.3009 0>, rot=<0 0 0 -1>}
 deform_wrist_up_l {parent=deform_wrist_up_base_l, loc=<0 0 0>, rot=<0 0 0 1>}
 lowerarm_twist_01_l {parent=lowerarm_l, loc=<18.316 0 0>, rot=<0 0 0 1>}
 lowerarm_twist_02_l {parent=lowerarm_l, loc=<10.6707 0 0>, rot=<0 0 0 1>}
 dyn_wrist_strap_l_MSTR {parent=lowerarm_twist_02_l, loc=<2.6826 -4.0099 0.9184>, rot=<-0.4756 0.4929 -0.5226 -0.5077>}
 dyn_wrist_strap_l {parent=dyn_wrist_strap_l_MSTR, loc=<0 0 0>, rot=<0 0 0 1>}
 deform_elbow_in_base_l {parent=lowerarm_l, loc=<0.1834 3.6062 -0.2572>, rot=<0 0 0 -1>}
 deform_elbow_in_l {parent=deform_elbow_in_base_l, loc=<0 0 0>, rot=<0 0 0 1>}
 deform_elbow_out_base_l {parent=lowerarm_l, loc=<0 -4 0>, rot=<0 0 0 1>}
 deform_elbow_out_l {parent=deform_elbow_out_base_l, loc=<0 0 0>, rot=<0 0 0 1>}
 upperarm_twist_01_l {parent=upperarm_l, loc=<7.2135 -0.1134 0>, rot=<0 0 0 1>}
 dyn_cuff_base_l {parent=upperarm_twist_01_l, loc=<-3.0363 0 0>, rot=<0 0 0 1>}
 dyn_cuff_l_01 {parent=dyn_cuff_base_l, loc=<-0.4034 1.143 5.6335>, rot=<-0.0074 0.045 0.0016 -0.999>}
 dyn_cuff_l_03 {parent=dyn_cuff_base_l, loc=<-0.1337 -5.4568 0.1445>, rot=<-0.5628 0.0674 -0.0607 -0.8216>}
 dyn_cuff_l_02_MSTR {parent=dyn_cuff_base_l, loc=<0.5815 6.3233 -0.0655>, rot=<0.5322 0.0497 0.0592 -0.8431>}
 dyn_cuff_l_02 {parent=dyn_cuff_l_02_MSTR, loc=<0 0 0>, rot=<0 0 0 1>}
 dyn_cuff_l_04_MSTR {parent=dyn_cuff_base_l, loc=<0.5614 -0.0006 -5.7757>, rot=<0.9983 -0.0003 0.045 -0.0364>}
 dyn_cuff_l_04 {parent=dyn_cuff_l_04_MSTR, loc=<0 0 0>, rot=<0 0 0 1>}
 deform_upperarm_bk_base_l {parent=upperarm_twist_01_l, loc=<-6.4053 -4.3746 0.314>, rot=<0 0 0 -1>}
 deform_upperarm_bk_l {parent=deform_upperarm_bk_base_l, loc=<0 0 0>, rot=<0 0 0 1>}
 deform_upperarm_fr_base_l {parent=upperarm_twist_01_l, loc=<-5.6312 5.092 0.3354>, rot=<0 0 0 -1>}
 deform_upperarm_fr_l {parent=deform_upperarm_fr_base_l, loc=<0 0 0>, rot=<0 0 0 1>}
 deform_upperarm_up_base_l {parent=upperarm_twist_01_l, loc=<-8.891 0.4825 3.9517>, rot=<0 0 0 -1>}
 deform_upperarm_up_l {parent=deform_upperarm_up_base_l, loc=<0 0 0>, rot=<0 0 0 1>}
 upperarm_twist_02_l {parent=upperarm_l, loc=<19.7998 -0.1134 0>, rot=<0 0 0 1>}
 clavicle_r {parent=spine_05, loc=<5.4831 -0.1053 3.2946>, rot=<0.7289 0.0465 -0.6829 0.009>}
 upperarm_r {parent=clavicle_r, loc=<-9.8658 1.2136 0.4104>, rot=<0.0164 -0.3245 -0.0269 -0.9454>}
 lowerarm_r {parent=upperarm_r, loc=<-26.4651 0.8345 -0.259>, rot=<0.0036 0.0161 -0.2298 -0.9731>}
 hand_r {parent=lowerarm_r, loc=<-22.7356 0.3203 0.6106>, rot=<-0.6784 -0.0223 0.0735 -0.7307>}
 index_metacarpal_r {parent=hand_r, loc=<-1.6534 -0.0181 2.9601>, rot=<0.0158 -0.0124 -0.0173 -0.9996>}
 index_01_r {parent=index_metacarpal_r, loc=<-8.3008 -0.0357 0.0524>, rot=<-0.0044 0.0485 0.0924 -0.9945>}
 index_02_r {parent=index_01_r, loc=<-3.3589 0.3003 -0.0453>, rot=<-0.002 0.0027 0.3225 -0.9466>}
 index_03_r {parent=index_02_r, loc=<-2.5578 0.123 0.0252>, rot=<-0.002 0.0067 0.2558 -0.9667>}
 middle_metacarpal_r {parent=hand_r, loc=<-2.1856 0.1726 0.5984>, rot=<0.006 0.0168 -0.0329 -0.9993>}
 middle_01_r {parent=middle_metacarpal_r, loc=<-7.3273 0.2447 -0.0129>, rot=<-0.0208 0.0042 0.0858 -0.9961>}
 middle_02_r {parent=middle_01_r, loc=<-4.0124 -0.14 -0.1175>, rot=<-0.0218 0.0258 0.3512 -0.9357>}
 middle_03_r {parent=middle_02_r, loc=<-2.6077 -0.0533 -0.1373>, rot=<0.0082 0.0188 0.1708 -0.9851>}
 pinky_metacarpal_r {parent=hand_r, loc=<-1.7617 0.9753 -3.3814>, rot=<-0.0397 0.0814 -0.069 -0.9935>}
 pinky_01_r {parent=pinky_metacarpal_r, loc=<-7.1707 0.4832 0.1094>, rot=<-0.0654 -0.0035 0.2242 -0.9723>}
 pinky_02_r {parent=pinky_01_r, loc=<-2.8274 -0.0892 0.1043>, rot=<0.0037 0.0011 0.2854 -0.9584>}
 pinky_03_r {parent=pinky_02_r, loc=<-2.2045 -0.031 0.0078>, rot=<-0.0002 0.0098 0.1445 -0.9895>}
 ring_metacarpal_r {parent=hand_r, loc=<-2.2073 0.3627 -1.3638>, rot=<-0.0284 0.0722 -0.0458 -0.9959>}
 ring_01_r {parent=ring_metacarpal_r, loc=<-7.0297 0.6197 0.1446>, rot=<-0.0489 -0.0252 0.1892 -0.9804>}
 ring_02_r {parent=ring_01_r, loc=<-3.6171 -0.4648 0.1027>, rot=<0 0 0.2806 -0.9598>}
 ring_03_r {parent=ring_02_r, loc=<-2.4153 -0.1141 -0.0199>, rot=<0 0 0.1189 -0.9929>}
 thumb_01_r {parent=hand_r, loc=<-1.3119 0.3192 2.459>, rot=<0.6546 -0.4042 0.0664 -0.6354>}
 thumb_02_r {parent=thumb_01_r, loc=<-5.9335 0.0447 -0.0401>, rot=<-0.0037 0.0213 0.298 -0.9543>}
 thumb_03_r {parent=thumb_02_r, loc=<-2.2223 -0.3682 0.0907>, rot=<0.0118 -0.013 0.0066 -0.9998>}
 weapon_r {parent=hand_r, loc=<0.1975 3.4112 -0.3811>, rot=<0 0 0 -1>}
 deform_wrist_dn_base_r {parent=hand_r, loc=<0 1.5329 0>, rot=<0 0 0 -1>}
 deform_wrist_dn_r {parent=deform_wrist_dn_base_r, loc=<0 0 0>, rot=<0 0 0 1>}
 deform_wrist_up_base_r {parent=hand_r, loc=<-0.0001 -2.3006 -0.0001>, rot=<0 0 0 -1>}
 deform_wrist_up_r {parent=deform_wrist_up_base_r, loc=<0 0 0>, rot=<0 0 0 1>}
 lowerarm_twist_01_r {parent=lowerarm_r, loc=<-18.3161 0 0>, rot=<0 0 0 1>}
 lowerarm_twist_02_r {parent=lowerarm_r, loc=<-10.6707 0 0>, rot=<0 0 0 1>}
 dyn_wrist_strap_r_MSTR {parent=lowerarm_twist_02_r, loc=<-2.6825 4.0099 -0.9183>, rot=<-0.4756 0.4929 -0.5226 -0.5077>}
 dyn_wrist_strap_r {parent=dyn_wrist_strap_r_MSTR, loc=<0 0 0>, rot=<-0.033 0.1671 -0.3736 -0.9118>}
 deform_elbow_in_base_r {parent=lowerarm_r, loc=<-0.1834 -3.6062 0.2572>, rot=<0 0 0 -1>}
 deform_elbow_in_r {parent=deform_elbow_in_base_r, loc=<0 0 0>, rot=<0 0 0 1>}
 deform_elbow_out_base_r {parent=lowerarm_r, loc=<0 4 0>, rot=<0 0 0 1>}
 deform_elbow_out_r {parent=deform_elbow_out_base_r, loc=<0 0 0>, rot=<0 0 0 1>}
 upperarm_twist_01_r {parent=upperarm_r, loc=<-7.2135 0.1134 0>, rot=<0 0 0 1>}
 dyn_cuff_base_r {parent=upperarm_twist_01_r, loc=<3.544 0 0.0001>, rot=<0 -1 0 0>}
 dyn_cuff_r_03 {parent=dyn_cuff_base_r, loc=<0.1767 5.6474 -0.0802>, rot=<0.5628 0.0674 0.0607 -0.8216>}
 dyn_cuff_r_01 {parent=dyn_cuff_base_r, loc=<0.1865 -1.1579 5.1526>, rot=<0.0074 0.045 -0.0016 -0.999>}
 dyn_cuff_r_05_MSTR {parent=dyn_cuff_base_r, loc=<3.3314 0.1184 -0.01>, rot=<0.9983 0.0003 0.045 0.0364>}
 dyn_cuff_r_02_MSTR {parent=dyn_cuff_base_r, loc=<1.0112 -6.7002 -0.9549>, rot=<-0.5322 0.0497 -0.0592 -0.8431>}
 dyn_cuff_r_02 {parent=dyn_cuff_r_02_MSTR, loc=<0 0 0>, rot=<0 0 0 1>}
 dyn_cuff_r_04_MSTR {parent=dyn_cuff_base_r, loc=<0.8415 0.0297 -5.494>, rot=<0.9983 0.0003 0.045 0.0364>}
 dyn_cuff_r_04 {parent=dyn_cuff_r_04_MSTR, loc=<0 0 0>, rot=<0 0 0 1>}
 deform_upperarm_bk_base_r {parent=upperarm_twist_01_r, loc=<6.4052 4.3746 -0.3139>, rot=<0 0 0 -1>}
 deform_upperarm_bk_r {parent=deform_upperarm_bk_base_r, loc=<0 0 0>, rot=<0 0 0 1>}
 deform_upperarm_fr_base_r {parent=upperarm_twist_01_r, loc=<5.6311 -5.092 -0.3353>, rot=<0 0 0 -1>}
 deform_upperarm_fr_r {parent=deform_upperarm_fr_base_r, loc=<0 0 0>, rot=<0 0 0 1>}
 deform_upperarm_up_base_r {parent=upperarm_twist_01_r, loc=<8.8909 -0.4825 -3.9516>, rot=<0 0 0 -1>}
 deform_upperarm_up_r {parent=deform_upperarm_up_base_r, loc=<0 0 0>, rot=<0 0 0 1>}
 upperarm_twist_02_r {parent=upperarm_r, loc=<-19.7998 0.1134 0>, rot=<0 0 0 1>}
 neck_01 {parent=spine_05, loc=<11.3138 0.1529 0>, rot=<0 0 -0.2003 -0.9797>}
 neck_02 {parent=neck_01, loc=<4.5627 0 0>, rot=<0 0 0.0504 -0.9987>}
 head {parent=neck_02, loc=<5.3914 0 0>, rot=<0 0 0.115 -0.9934>}
 neck_01_LOD1 {parent=neck_01, loc=<0 0 0>, rot=<0 0 0 1>}
 dyn_clip_back_MSTR {parent=spine_05, loc=<5.8125 -7.5807 0.0002>, rot=<0.0342 -0.7099 0.7032 -0.0224>}
 dyn_clip_back_01 {parent=dyn_clip_back_MSTR, loc=<0 0 0>, rot=<-0.0069 0.2442 -0.0067 -0.9697>}
 dyn_clip_back_02 {parent=dyn_clip_back_01, loc=<1.5295 0 0>, rot=<0 -0.1899 0 -0.9818>}
 dyn_collar_00 {parent=spine_05, loc=<15.4661 -6.1347 -0.0095>, rot=<0.0284 -0.7065 -0.0284 -0.7065>}
 dyn_collar_01 {parent=dyn_collar_00, loc=<-0.0095 0.2822 5.5804>, rot=<0 0 0 1>}
 dyn_lapel_r {parent=spine_05, loc=<-1.2627 11.9481 6.7553>, rot=<0.0284 -0.7065 -0.0284 -0.7065>}
 dyn_lapel_r1 {parent=dyn_lapel_r, loc=<-4.255 -0.9313 3.7609>, rot=<0 0 0 1>}
 dyn_lapel_l {parent=spine_05, loc=<-1.2627 11.9481 -6.6888>, rot=<0.0284 -0.7065 -0.0284 -0.7065>}
 dyn_lapel_l1 {parent=dyn_lapel_l, loc=<3.7068 -0.4033 3.8783>, rot=<0 0 0 1>}
 attach_cape {parent=spine_05, loc=<9.3216 -8.4501 0>, rot=<0.0616 -0.7044 -0.0616 -0.7044>}
 cape {parent=attach_cape, type=Null, loc=<0 0 0>, rot=<0 0 0 1>}
 attach_backpack {parent=spine_05, loc=<0 -8.35 0>, rot=<-0.0093 -0.707 0.0093 -0.707>}
 Backpack_BR {parent=attach_backpack, type=Null, loc=<0 1.2502 0>, rot=<0 0 0 1>}
 base_chain {parent=spine_05, loc=<6.4927 7.8771 0.0251>, rot=<-0.0197 -0.7108 -0.7016 -0.0461>}
 dyn_chain_01 {parent=base_chain, loc=<0 0 0>, rot=<0 0 0 1>}
 dyn_chain_02 {parent=dyn_chain_01, loc=<1.2958 0 0>, rot=<0 0.1366 0 -0.9906>}
 dyn_chain_03 {parent=dyn_chain_02, loc=<4.3903 0.1747 -0.0124>, rot=<-0.0028 0.1303 -0.0197 -0.9913>}
 dyn_chain_04 {parent=dyn_chain_03, loc=<3.5882 0 0>, rot=<0 0 0 1>}
 deform_pec_aim_base_l {parent=spine_04, loc=<3.2456 9.3309 -1.2237>, rot=<0.0603 -0.4033 0.3318 -0.8507>}
 deform_pec_aim_l {parent=deform_pec_aim_base_l, loc=<0 0 0>, rot=<0 0 0 1>}
 deform_pec_base_l {parent=deform_pec_aim_l, loc=<9.3665 0 0>, rot=<0 0 0 1>}
 deform_pec_l {parent=deform_pec_base_l, loc=<0 0 0>, rot=<0 0 0 1>}
 deform_pec_aim_base_r {parent=spine_04, loc=<3.2452 9.3309 1.2237>, rot=<-0.4033 -0.0603 0.8507 0.3318>}
 deform_pec_aim_r {parent=deform_pec_aim_base_r, loc=<0 0 0>, rot=<0 0 0 1>}
 deform_pec_base_r {parent=deform_pec_aim_r, loc=<-9.3667 -0.0003 -0.0004>, rot=<0 0 0 1>}
 deform_pec_r {parent=deform_pec_base_r, loc=<0 0 0>, rot=<0 0 0 1>}
 dyn_lapel_lower_r {parent=spine_04, loc=<-0.4062 12.3691 5.8508>, rot=<0.0294 -0.7065 -0.0294 -0.7065>}
 dyn_lapel_lower_r1 {parent=dyn_lapel_lower_r, loc=<-3.4456 0.6264 -0.3913>, rot=<0 0 0 1>}
 dyn_clip_front_MSTR {parent=spine_03, loc=<-2.0058 10.4414 -0.0337>, rot=<0.0166 -0.7026 -0.0459 -0.7099>}
 dyn_clip_front_01 {parent=dyn_clip_front_MSTR, loc=<0 0 0>, rot=<0 0 0 1>}
 dyn_clip_front_02 {parent=dyn_clip_front_01, loc=<-0.017 -0.0273 -1.3106>, rot=<0 0 0 1>}
 dyn_jacketSide_l {parent=spine_03, loc=<2.1847 3.4272 -10.0089>, rot=<0.0906 -0.7333 0.6728 -0.0377>}
 dyn_jacketFront_r {parent=spine_03, loc=<4.1793 10.7098 7.1872>, rot=<0.0426 -0.4943 0.8578 -0.1341>}
 dyn_jacketBack {parent=spine_03, loc=<7.5243 -8.3606 0.0002>, rot=<-0.0425 -0.9991 0 0>}
 dyn_jacketFront_l {parent=spine_03, loc=<4.1795 10.7098 -7.4337>, rot=<0.117 -0.8588 0.4986 -0.0131>}
 dyn_jacketSideBack_r {parent=spine_03, loc=<4.0391 -5.4457 7.9666>, rot=<0.0254 -0.9233 0.3832 -0.0014>}
 dyn_jacketSideBack_l {parent=spine_03, loc=<4.0836 -5.8541 -8.5798>, rot=<0.0014 -0.3832 0.9233 -0.0254>}
 dyn_jacketSide_r {parent=spine_03, loc=<2.1843 3.4272 10.215>, rot=<0.0377 -0.6728 0.7333 -0.0906>}
 thigh_l {parent=pelvis, loc=<-1.0855 -0.2109 -8.7053>, rot=<0.0009 0.0355 0.0155 -0.9993>}
 calf_l {parent=thigh_l, loc=<-42.5284 -0.0019 -0.7297>, rot=<0.0006 -0.0161 -0.0391 -0.9991>}
 calf_twist_01_l {parent=calf_l, loc=<-35.8483 -1.9197 -0.5794>, rot=<0.0237 0.0095 -0.0035 -0.9997>}
 calf_twist_02_l {parent=calf_l, loc=<-19.1546 -1.7943 -0.2638>, rot=<0.0237 0.0095 -0.0035 -0.9997>}
 foot_l {parent=calf_l, loc=<-46.7402 -2.5014 -0.592>, rot=<0.0427 -0.0238 0.0353 -0.9982>}
 ball_l {parent=foot_l, loc=<-3.5658 10.8066 -0.1544>, rot=<0.0001 -0.0001 -0.7063 -0.7079>}
 dyn_laces_l {parent=foot_l, loc=<3.6151 -6.3888 0.2901>, rot=<-0.0357 -0.7045 -0.0284 -0.7082>}
 dyn_shoe_strap_l {parent=foot_l, loc=<0.8119 6.5819 -1.845>, rot=<-0.0341 0.708 0.0414 -0.7042>}
 deform_knee_in_base_l {parent=calf_l, loc=<0.8031 -5.0231 0.2456>, rot=<0 0 0 -1>}
 deform_knee_in_l {parent=deform_knee_in_base_l, loc=<0 0 0>, rot=<0 0 0 1>}
 deform_knee_out_base_l {parent=calf_l, loc=<0 4 0>, rot=<0 0 0 1>}
 deform_knee_out_l {parent=deform_knee_out_base_l, loc=<0 0 0>, rot=<0 0 0 1>}
 thigh_twist_01_l {parent=thigh_l, loc=<-27.952 0 0>, rot=<0.0013 0 0 -1>}
 deform_groin_base_l {parent=thigh_l, loc=<0.4801 5.6776 -0.2486>, rot=<0 0 0 -1>}
 deform_groin_l {parent=deform_groin_base_l, loc=<0 0 0>, rot=<0 0 0 1>}
 thigh_r {parent=pelvis, loc=<-1.0853 -0.2109 8.7053>, rot=<0.0355 -0.0009 0.9993 0.0155>}
 calf_r {parent=thigh_r, loc=<42.5286 0.0019 0.7297>, rot=<0.0006 -0.0161 -0.0391 -0.9991>}
 calf_twist_01_r {parent=calf_r, loc=<35.8485 1.9197 0.5794>, rot=<0.0237 0.0095 -0.0035 -0.9997>}
 calf_twist_02_r {parent=calf_r, loc=<19.1546 1.7943 0.2638>, rot=<0.0237 0.0095 -0.0035 -0.9997>}
 foot_r {parent=calf_r, loc=<46.7403 2.5014 0.592>, rot=<0.0427 -0.0238 0.0353 -0.9982>}
 ball_r {parent=foot_r, loc=<3.5658 -10.8066 0.1544>, rot=<0.0001 -0.0001 -0.7063 -0.7079>}
 dyn_laces_r {parent=foot_r, loc=<-3.6153 6.3888 -0.2901>, rot=<0.4979 -0.4997 0.5056 -0.4968>}
 dyn_shoe_strap_r {parent=foot_r, loc=<-0.8121 -6.5819 1.845>, rot=<0.7082 0.0284 -0.7045 -0.0357>}
 deform_knee_in_base_r {parent=calf_r, loc=<-0.8031 5.0231 -0.2457>, rot=<0 0 0 -1>}
 deform_knee_in_r {parent=deform_knee_in_base_r, loc=<0 0 0>, rot=<0 0 0 1>}
 deform_knee_out_base_r {parent=calf_r, loc=<0 -4 0>, rot=<0 0 0 1>}
 deform_knee_out_r {parent=deform_knee_out_base_r, loc=<0 0 0>, rot=<0 0 0 1>}
 thigh_twist_01_r {parent=thigh_r, loc=<27.952 0 0>, rot=<0.0013 0 0 -1>}
 dyn_skirt_fr_mid_r_01 {parent=thigh_twist_01_r, loc=<-23.2011 -7.1737 -1.138>, rot=<-0.708 0.0239 0.0035 -0.7058>}
 dyn_skirt_02_r {parent=thigh_twist_01_r, loc=<-22.7637 -4.2368 4.835>, rot=<-0.3704 0.0497 -0.0274 -0.9271>}
 dyn_skirt_side_02_r {parent=thigh_twist_01_r, loc=<-23.1531 2.3352 6.863>, rot=<-0.0001 0.0507 -0.0262 -0.9984>}
 dyn_skirt_fr_mid_r_02 {parent=thigh_twist_01_r, loc=<-14.4944 -7.5113 -0.8873>, rot=<-0.7079 -0.0015 0.0072 -0.7062>}
 dyn_skirt_fr_mid_r_03 {parent=dyn_skirt_fr_mid_r_02, loc=<8.5625 0 0>, rot=<0.0001 -0.0054 0.0259 -0.9996>}
 dyn_skirt_03_r {parent=thigh_twist_01_r, loc=<-14.4016 -4.1189 5.7821>, rot=<-0.3698 0.0335 -0.0358 -0.9278>}
 dyn_skirt_04_r {parent=dyn_skirt_03_r, loc=<8.5491 0 0>, rot=<-0.0002 -0.0671 0.0032 -0.9977>}
 dyn_skirt_side_03_r {parent=thigh_twist_01_r, loc=<-14.7157 2.78 7.7227>, rot=<0.0001 0.0341 -0.043 -0.9985>}
 dyn_skirt_side_04_r {parent=dyn_skirt_side_03_r, loc=<8.6269 0 0>, rot=<-0.0027 -0.066 0.0415 -0.997>}
 deform_groin_base_r {parent=thigh_r, loc=<-0.4799 -5.6776 0.2485>, rot=<0 0 0 -1>}
 deform_groin_r {parent=deform_groin_base_r, loc=<0 0 0>, rot=<0 0 0 1>}
 dyn_belt_bk_00 {parent=pelvis, loc=<1.8517 -11.1172 6.3087>, rot=<-0.3295 -0.5569 0.6697 -0.3644>}
 dyn_belt_bk_01 {parent=dyn_belt_bk_00, loc=<9.9638 0 0>, rot=<0.0587 -0.103 -0.0199 -0.9928>}
 dyn_belt_bk_02 {parent=dyn_belt_bk_01, loc=<5.3141 0 0>, rot=<0.1157 -0.2 0.0258 -0.9726>}
 dyn_belt_bk_03 {parent=dyn_belt_bk_02, loc=<6.1418 0 0>, rot=<0.0648 -0.1319 -0.0935 -0.9847>}
 dyn_belt_bk_04 {parent=dyn_belt_bk_03, loc=<6.2049 0 0>, rot=<0.0462 -0.1927 -0.1237 -0.9723>}
 dyn_belt_bk_05 {parent=dyn_belt_bk_04, loc=<3.0362 0 0>, rot=<0.0094 -0.1669 -0.0513 -0.9846>}
 dyn_belt_bk_06 {parent=dyn_belt_bk_05, loc=<4.063 0 0>, rot=<-0.0027 -0.2045 -0.0174 -0.9787>}
 base_side_piece {parent=dyn_belt_bk_04, loc=<5.2813 1.0392 1.9627>, rot=<0.1238 -0.0307 0.7801 -0.6126>}
 dyn_side_piece_01 {parent=base_side_piece, loc=<0 0 0>, rot=<0 0 0 1>}
 dyn_side_piece_02 {parent=dyn_side_piece_01, loc=<17.244 0 0>, rot=<0 0 0 1>}
 base_belt_bk_07 {parent=dyn_belt_bk_04, loc=<0 0 0>, rot=<0 0 0 1>}
 base_back_piece {parent=dyn_belt_bk_00, loc=<10.0955 -1.7593 1.389>, rot=<0.1247 0.0303 0.8885 -0.4406>}
 dyn_back_piece_01 {parent=base_back_piece, loc=<0 0 0>, rot=<0 0 0 1>}
 dyn_back_piece_02 {parent=dyn_back_piece_01, loc=<8.4076 0 0>, rot=<0 0 0 1>}
 dyn_belt_ft_MSTR {parent=pelvis, loc=<4.1902 -0.2376 14.0716>, rot=<-0.0127 -0.707 0.0127 -0.707>}
 dyn_belt_ft_base_00 {parent=dyn_belt_ft_MSTR, loc=<0 0 0>, rot=<0.6624 0.2475 -0.5253 -0.4733>}
 dyn_belt_ft_base_01 {parent=dyn_belt_ft_base_00, loc=<5.0776 0 0>, rot=<-0.1304 -0.2114 -0.0707 -0.9661>}
 dyn_belt_ft_00 {parent=dyn_belt_ft_base_01, loc=<5.9038 0 0>, rot=<-0.1118 -0.1681 0.038 -0.9787>}
 dyn_belt_ft_01 {parent=dyn_belt_ft_00, loc=<6.8104 0 0>, rot=<-0.0402 -0.072 0.0389 -0.9958>}
 dyn_belt_ft_02 {parent=dyn_belt_ft_01, loc=<7.5837 0 0>, rot=<-0.0862 -0.2022 0.0652 -0.9734>}
 dyn_belt_ft_03 {parent=dyn_belt_ft_02, loc=<6.5359 0 0>, rot=<-0.0098 -0.0496 0.1414 -0.9887>}
 dyn_belt_ft_04 {parent=dyn_belt_ft_03, loc=<5.8497 0 0>, rot=<-0.0161 -0.3764 0.0087 -0.9263>}
 dyn_belt_ft_05 {parent=dyn_belt_ft_04, loc=<3.6994 0 0>, rot=<-0.002 -0.1813 0.022 -0.9832>}
 dyn_belt_ft_06 {parent=dyn_belt_ft_05, loc=<4.0901 0 0>, rot=<0.0118 -0.1764 0.0544 -0.9827>}
 base_belt_ft_06 {parent=dyn_belt_ft_05, loc=<0 0 0>, rot=<0 0 0 1>}
 base_belt_ft_05 {parent=dyn_belt_ft_03, loc=<0 0 0>, rot=<0 0 0 1>}
 base_belt_ft_04 {parent=dyn_belt_ft_02, loc=<0 0 0>, rot=<0 0 0 1>}
 base_belt_ft_03 {parent=dyn_belt_ft_01, loc=<0 0 0>, rot=<0 0 0 1>}
 base_belt_ft_02 {parent=dyn_belt_ft_00, loc=<0 0 0>, rot=<0 0 0 1>}
 base_belt_00 {parent=dyn_belt_ft_00, loc=<1.7773 0.4211 1.0295>, rot=<0.2145 -0.0522 0.8966 0.3838>}
 dyn_belt_00 {parent=base_belt_00, loc=<0 0 0>, rot=<0 0 0 1>}
 dyn_belt_01 {parent=dyn_belt_00, loc=<3.1217 0 0>, rot=<0.0567 0.0121 0.3207 -0.9454>}
 dyn_belt_02 {parent=dyn_belt_01, loc=<4.6139 0 0>, rot=<0.0225 -0.0071 0.1233 -0.9921>}
 dyn_belt_03 {parent=dyn_belt_02, loc=<4.4439 0 0>, rot=<0.0179 -0.0054 0.1056 -0.9942>}
 dyn_belt_04 {parent=dyn_belt_03, loc=<8.156 0 0>, rot=<0.01 -0.0282 0.0741 -0.9968>}
 dyn_belt_05 {parent=dyn_belt_04, loc=<8.5403 0 0>, rot=<0.0019 -0.0295 0.0244 -0.9993>}
 dyn_belt_06 {parent=dyn_belt_05, loc=<8.2718 0 0>, rot=<0.0007 -0.0074 0.0164 -0.9998>}
 dyn_belt_07 {parent=dyn_belt_06, loc=<6.6398 0 0>, rot=<-0.503 0.4803 0.497 -0.5189>}
 base_belt_ft_01 {parent=dyn_belt_ft_base_01, loc=<0 0 0>, rot=<0 0 0 1>}
 dyn_skirt_fr_01_l_MSTR {parent=pelvis, loc=<0.1746 7.9391 0.4223>, rot=<0.0221 0.7062 0.7064 0.0414>}
 dyn_skirt_fr_01_l {parent=dyn_skirt_fr_01_l_MSTR, loc=<0 0 0>, rot=<0 0 0 1>}
 dyn_skirt_fr_02_l {parent=dyn_skirt_fr_01_l, loc=<5.4622 0 0>, rot=<-0.0004 -0.027 -0.0136 -0.9995>}
 dyn_skirt_fr_03_l {parent=dyn_skirt_fr_02_l, loc=<9.4106 0 0>, rot=<0 0 0 1>}
 dyn_skirt_fr_04_l {parent=dyn_skirt_fr_03_l, loc=<0 0 0>, rot=<0 0 0 1>}
 dyn_skirt_fr_05_l {parent=dyn_skirt_fr_04_l, loc=<8.2945 0 0>, rot=<0 0 0 1>}
 base_skirt_fr_02_l {parent=dyn_skirt_fr_02_l, loc=<0 0 0>, rot=<0 0 0 1>}
 dyn_skirt_01_r_side_MSTR {parent=pelvis, loc=<1.4631 -2.3403 14.0246>, rot=<0.1278 -0.005 0.9918 -0.003>}
 dyn_skirt_side_01_r {parent=dyn_skirt_01_r_side_MSTR, loc=<0 0 0>, rot=<0 0 0 1>}
 dyn_skirt_01_r_MSTR {parent=pelvis, loc=<1.2255 3.6192 12.4507>, rot=<0.1076 0.3693 0.923 0.0008>}
 dyn_skirt_01_r {parent=dyn_skirt_01_r_MSTR, loc=<0 0 0>, rot=<0 0 0 1>}
 dyn_skirt_bk_01_r_MSTR {parent=pelvis, loc=<2.0589 -7.7104 10.4295>, rot=<0.1034 -0.4274 0.8974 -0.0366>}
 dyn_skirt_bk_01_r {parent=dyn_skirt_bk_01_r_MSTR, loc=<0 0 0>, rot=<0 0 0 1>}
 dyn_skirt_bk_02_r {parent=dyn_skirt_bk_01_r, loc=<7.4888 0 0>, rot=<0.0168 -0.0388 -0.0035 -0.9991>}
 dyn_skirt_bk_03_r {parent=dyn_skirt_bk_02_r, loc=<8.8841 0 0>, rot=<0.0602 0.0172 -0.0113 -0.998>}
 dyn_skirt_bk_04_r {parent=dyn_skirt_bk_03_r, loc=<7.7391 0 0>, rot=<0 -0.1015 0 -0.9948>}
 dyn_skirt_bk_01_l_MSTR {parent=pelvis, loc=<1.3683 -10.6177 0.3096>, rot=<0.0242 -0.7851 0.6177 -0.0376>}
 dyn_skirt_bk_01_l {parent=dyn_skirt_bk_01_l_MSTR, loc=<0 0 0>, rot=<0 0 0 1>}
 dyn_skirt_bk_02_l {parent=dyn_skirt_bk_01_l, loc=<8.7894 0 0>, rot=<-0.0858 -0.0039 0.0103 -0.9963>}
 dyn_skirt_bk_03_l {parent=dyn_skirt_bk_02_l, loc=<6.5732 0 0>, rot=<0.0068 0.0016 -0.0008 -1>}
 dyn_skirt_bk_04_l {parent=dyn_skirt_bk_03_l, loc=<7.4421 0 0>, rot=<0 -0.0532 0 -0.9986>}
 dyn_skirt_fr_mid_r_MSTR {parent=pelvis, loc=<0.3753 6.6991 7.0002>, rot=<0.1076 0.3693 0.923 0.0008>}
 deform_glute_base_l {parent=pelvis, loc=<-7.3466 -9.0213 -9.1245>, rot=<0.0172 0.031 -0.4477 -0.8935>}
 deform_glute_l {parent=deform_glute_base_l, loc=<0 0 0>, rot=<0 0 0 1>}
 deform_glute_base_r {parent=pelvis, loc=<-7.3464 -9.0213 9.1245>, rot=<0.031 -0.0172 0.8935 -0.4477>}
 deform_glute_r {parent=deform_glute_base_r, loc=<0 0 0>, rot=<0 0 0 1>}
 thigh_l_null {parent=pelvis, type=Null, loc=<-1.0855 -0.2109 -8.7053>, rot=<0.0009 0.0355 0.0155 -0.9993>}
 ik_foot_root {parent=root, loc=<0 0 0>, rot=<0 0 0 1>}
 ik_foot_l {parent=ik_foot_root, loc=<14.8262 -7.3545 8.296>, rot=<0.0357 0.7045 0.0284 -0.7082>}
 ik_foot_r {parent=ik_foot_root, loc=<-14.8262 -7.3546 8.2958>, rot=<0.7082 0.0284 -0.7045 0.0357>}
 ik_hand_root {parent=root, loc=<0 0 0>, rot=<0 0 0 1>}
 ik_hand_gun {parent=ik_hand_root, loc=<-48.1744 3.3546 104.4381>, rot=<0.6276 0.1309 0.4183 -0.6434>}
 ik_hand_l {parent=ik_hand_gun, loc=<59.3272 67.6997 34.3524>, rot=<-0.6158 -0.7027 -0.3565 0>}
 ik_hand_r {parent=ik_hand_gun, loc=<0 0 0>, rot=<0 0 0 1>}
]
variables [
 DeltaRotation_outputWeight {type=double, default=0}
 crouchwalking {type=double, default=0}
 isEmoting {type=double, default=0}
 isMoving2dFloat {type=double, default=0}
 isTargetingFloat {type=double, default=0}
 targetingHeadupBool {type=bool, default=False}
]
