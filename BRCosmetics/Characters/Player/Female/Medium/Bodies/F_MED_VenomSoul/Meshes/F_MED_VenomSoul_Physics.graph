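# Physics asset for the VenomSoul female medium body, setting up secondary motion for its belts, skirt and jacket.
# Capsule bodies cover those dyn_ bones plus pelvis, spine_02, thighs, forearms and two small base pieces.
# Limited-angle constraints hang skirt and belt chains from the pelvis and jacket flaps from spine_02.
# Front and back belt chains lock together at their ends, while fully free constraints link thighs and some neighboring pieces.
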PhysicsAsset /BRCosmetics/Characters/Player/Female/Medium/Bodies/F_MED_VenomSoul/Meshes/F_MED_VenomSoul_Physics
bodies {
 base_chain [{shapeType=Capsule, length=0.505, radius=0.505, shapeName=base_chain_capsule, rotation=<rot 0 -90 0>}]
 base_side_piece [{shapeType=Capsule, length=0.505, radius=0.505, shapeName=base_side_piece_capsule, rotation=<rot 0 -90 0>}]
 dyn_belt_00 [{shapeType=Capsule, length=2.7602, radius=1, center=<1.65 0.41 0.75>, rotation=<rot 0 0 90>}]
 dyn_belt_01 [{shapeType=Capsule, length=3.8384, radius=1.9452, center=<1.8065 -0.1669 0.45>, rotation=<rot 0.0001 0.8887 -106.1507>}]
 dyn_belt_02 [{shapeType=Capsule, length=2.7043, radius=2.6946, center=<-0.9996 -0.5675 0.4>, rotation=<rot 89.4579 0.0448 95.5198>}]
 dyn_belt_03 [{shapeType=Capsule, length=10.9734, radius=2.8752, center=<3.4076 0 0.2534>, rotation=<rot 88.2743 0.0009 0.0009>}]
 dyn_belt_04 [{shapeType=Capsule, length=10.3327, radius=3.0263, center=<3 0 0.0636>, rotation=<rot 90 0 0>}]
 dyn_belt_05 [{shapeType=Capsule, length=6.0092, radius=2.9908, center=<3 0 0.0988>, rotation=<rot 90 0 0>}]
 dyn_belt_06 [{shapeType=Capsule, length=6, radius=3, center=<3 0 0>, rotation=<rot 90 0 0>}]
 dyn_belt_bk_01 [{shapeType=Capsule, length=4.8443, radius=2.3981, center=<2.9676 -0.2614 0.5272>, rotation=<rot 90 0 0>}]
 dyn_belt_bk_02 [{shapeType=Capsule, length=3.4281, radius=3.8143, center=<2.9676 -0.2614 0.5272>, rotation=<rot 90 0 0>}]
 dyn_belt_bk_03 [{shapeType=Capsule, length=4.2355, radius=3.0069, center=<2.9676 -0.2614 0.8394>, rotation=<rot 90 0 0>}]
 dyn_belt_bk_04 [{shapeType=Capsule, length=4.8443, radius=2.3981, center=<2.9676 -0.2614 0.5272>, rotation=<rot 90 0 0>}]
 dyn_belt_bk_05 [{shapeType=Capsule, length=4.8443, radius=2.3981, center=<2.9676 -0.2614 1.685>, rotation=<rot 90 0 0>}]
 dyn_belt_bk_06 [{shapeType=Capsule, length=4.041, radius=3.959, center=<0 0 4.5169>, rotation=<rot 90 0 0>}]
 dyn_belt_ft_00 [{shapeType=Capsule, length=4.8443, radius=2.3981, center=<2.9676 -0.2614 0.5272>, rotation=<rot 90 0 0>}]
 dyn_belt_ft_01 [{shapeType=Capsule, length=5.7517, radius=3.9124, center=<1.0562 2 4>, rotation=<rot 90 0 0>}]
 dyn_belt_ft_02 [{shapeType=Capsule, length=2.2671, radius=2.7329, center=<3.0507 0.1345 1.677>, rotation=<rot 82.4696 -179.9993 -179.9993>}]
 dyn_belt_ft_03 [{shapeType=Capsule, length=5.1867, radius=2.6662, center=<2.9524 0.6118 2.2024>, rotation=<rot 88.9748 -18.2599 -24.5464>}]
 dyn_belt_ft_04 [{shapeType=Capsule, length=3.2286, radius=3.2211, center=<2.8951 0.6124 1.7769>, rotation=<rot 85.1458 -179.9997 -179.9998>}]
 dyn_belt_ft_05 [{shapeType=Capsule, length=4.8314, radius=4.1686, center=<3 0 3.1344>, rotation=<rot 81.6673 0.0001 0>}]
 dyn_belt_ft_06 [{shapeType=Capsule, length=4.1924, radius=3.1279, center=<2.9611 0 1.905>, rotation=<rot 74.3433 0.0007 0.0004>}]
 dyn_belt_ft_base_00 [{shapeType=Capsule, length=4.3975, radius=2.845, center=<2.9676 -0.2614 0.5272>, rotation=<rot 90 0 0>}]
 dyn_belt_ft_base_01 [{shapeType=Capsule, length=4.6775, radius=2.565, center=<2.9676 -0.2614 0.5272>, rotation=<rot 90 0 0>}]
 dyn_jacketBack [{
   shapeType=Capsule
   length=4.0357
   radius=3.4674
   shapeName=dyn_jacketBack_capsule
   center=<7.2 -2 0.056>
   rotation=<rot 4.6507 -89.9999 2.4345>
  }]
 dyn_jacketFront_l [{
   shapeType=Capsule
   length=3.8534
   radius=3.1402
   shapeName=dyn_jacketFront_l_capsule
   center=<5.1 2.6 0.285>
   rotation=<rot -3.7159 -89.6226 -176.1149>
  }]
 dyn_jacketFront_r [{
   shapeType=Capsule
   length=3.8534
   radius=3.1402
   shapeName=dyn_jacketFront_r_capsule
   center=<5.2 -2.64 -0.285>
   rotation=<rot -3.7159 -89.6226 3.8851>
  }]
 dyn_jacketSideBack_l [{
   shapeType=Capsule
   length=6.0358
   radius=3.2816
   shapeName=dyn_jacketSideBack_l_capsule
   center=<4 0.89 -0.83>
   rotation=<rot 7.3478 -89.9999 -179.8515>
  }]
 dyn_jacketSideBack_r [{
   shapeType=Capsule
   length=6.0358
   radius=3.2816
   shapeName=dyn_jacketSideBack_r_capsule
   center=<4 -1.76 0.57>
   rotation=<rot -0.5938 -90.0204 0.1473>
  }]
 dyn_jacketSide_l [{
   shapeType=Capsule
   length=3.1322
   radius=3.2156
   shapeName=dyn_jacketSide_l_capsule
   center=<3.1 3.2 -0.51>
   rotation=<rot 4.2262 -89.8804 -175.2611>
  }]
 dyn_jacketSide_r [{
   shapeType=Capsule
   length=3.1322
   radius=3.2156
   shapeName=dyn_jacketSide_r_capsule
   center=<3.1 -3.2 0.51>
   rotation=<rot 4.2262 -89.8804 4.7389>
  }]
 dyn_skirt_bk_01_l [{shapeType=Capsule, length=5, radius=6, center=<6 0 4>, rotation=<rot 0 0 90>}]
 dyn_skirt_bk_01_r [{shapeType=Capsule, length=5, radius=6, center=<6 0 4>, rotation=<rot 0 0 90>}]
 dyn_skirt_bk_02_l [{shapeType=Capsule, length=11.1233, radius=5.8472, center=<6 0 4>, rotation=<rot 0 0 90>}]
 dyn_skirt_bk_02_r [{shapeType=Capsule, length=5, radius=6, center=<6 0 4>, rotation=<rot 0 0 90>}]
 dyn_skirt_bk_03_l [{shapeType=Capsule, length=5, radius=6, center=<6 0 4>, rotation=<rot 4.3328 0 90>}]
 dyn_skirt_bk_03_r [{shapeType=Capsule, length=5, radius=6, center=<6 0 4>, rotation=<rot 0 0 90>}]
 dyn_skirt_fr_01_l [{shapeType=Capsule, length=5, radius=6, center=<6 0 4>, rotation=<rot 0 0 90>}]
 dyn_skirt_fr_02_l [{shapeType=Capsule, length=1, radius=6, center=<6 0 4>, rotation=<rot 0 0 90>}]
 dyn_skirt_fr_03_l [{shapeType=Capsule, length=4.3663, radius=5.1516, center=<4.9845 0.1381 2>, rotation=<rot 0 0 90>}]
 dyn_skirt_fr_mid_r_02 [{
   shapeType=Capsule
   length=10.0922
   radius=5.4721
   shapeName=dyn_skirt_fr_mid_r_02_capsule
   center=<6.4275 0.2499 -4.6171>
   rotation=<rot 0.0491 0 -77.6611>
  }]
 lowerarm_l [{shapeType=Capsule, length=0.505, radius=0.505, shapeName=lowerarm_l_capsule, rotation=<rot 0 -90 0>}]
 lowerarm_r [{shapeType=Capsule, length=0.505, radius=0.505, shapeName=lowerarm_r_capsule, rotation=<rot 0 -90 0>}]
 pelvis [
  {shapeType=Capsule, length=20.6524, radius=8.2251, center=<-2.3579 -0.2117 0.7001>, rotation=<rot 1.1562 -179.9862 174.1997>}
  {shapeType=Capsule, length=4.3091, radius=3.6909, shapeName=pelvis_capsule, center=<3.7383 7.4263 0>}
 ]
 spine_02 [{shapeType=Capsule, length=0.8411, radius=9.0033, center=<3.5945 1.9057 0.1991>, rotation=<rot -1.2787 -90 0.7891>}]
 thigh_twist_01_l [{shapeType=Capsule, length=33.6868, radius=7.6896, center=<11.5546 -0.2247 0.4>, rotation=<rot 0.0877 -93.0872 -90.6824>}]
 thigh_twist_01_r [
  {shapeType=Capsule, length=19.647, radius=7.6748, center=<-14.9229 0.6592 0.018>, rotation=<rot -0.0011 89.4153 89.8962>}
  {shapeType=Capsule, length=22.1151, radius=5.969, shapeName=thigh_twist_01_r_capsule, center=<6.3405 1.0654 0>, rotation=<rot 90 9.7667 0>}
 ]
}
constraints [
 {
  bone1Name=thigh_twist_01_l
  bone2Name=pelvis
  swing1LimitDegrees=45
  swing1Motion=Free
  swing2LimitDegrees=45
  swing2Motion=Free
  twistLimitDegrees=45
  twistMotion=Free
 }
 {
  bone1Name=thigh_twist_01_r
  bone2Name=pelvis
  swing1LimitDegrees=45
  swing1Motion=Free
  swing2LimitDegrees=45
  swing2Motion=Free
  twistLimitDegrees=45
  twistMotion=Free
 }
 {bone1Name=dyn_skirt_bk_01_r, bone2Name=pelvis, swing1LimitDegrees=10, swing2LimitDegrees=20, twistLimitDegrees=1}
 {bone1Name=dyn_skirt_bk_01_l, bone2Name=pelvis, swing1LimitDegrees=15, swing2LimitDegrees=20, twistLimitDegrees=0}
 {bone1Name=dyn_skirt_fr_01_l, bone2Name=pelvis, swing1LimitDegrees=5, swing2LimitDegrees=10, twistLimitDegrees=0}
 {bone1Name=dyn_belt_ft_base_00, bone2Name=pelvis, swing1LimitDegrees=20, swing2LimitDegrees=5, twistLimitDegrees=20}
 {bone1Name=dyn_belt_bk_01, bone2Name=pelvis, swing1LimitDegrees=20, swing2LimitDegrees=20, twistLimitDegrees=20}
 {bone1Name=dyn_belt_bk_02, bone2Name=dyn_belt_bk_01, swing1LimitDegrees=20, swing2LimitDegrees=20, twistLimitDegrees=5}
 {bone1Name=dyn_belt_bk_03, bone2Name=dyn_belt_bk_02, swing1LimitDegrees=20, swing2LimitDegrees=20, twistLimitDegrees=5}
 {
  bone1Name=dyn_belt_bk_04
  bone2Name=dyn_belt_bk_03
  swing1LimitDegrees=20
  swing1Motion=Locked
  swing2LimitDegrees=20
  twistLimitDegrees=20
  twistMotion=Locked
 }
 {
  bone1Name=dyn_belt_bk_05
  bone2Name=dyn_belt_bk_04
  swing1LimitDegrees=45
  swing1Motion=Locked
  swing2LimitDegrees=45
  swing2Motion=Locked
  twistLimitDegrees=45
  twistMotion=Locked
 }
 {
  bone1Name=dyn_belt_bk_06
  bone2Name=dyn_belt_bk_05
  swing1LimitDegrees=45
  swing1Motion=Locked
  swing2LimitDegrees=45
  swing2Motion=Locked
  twistLimitDegrees=45
  twistMotion=Locked
 }
 {
  bone1Name=dyn_belt_bk_05
  bone2Name=dyn_belt_ft_06
  swing1LimitDegrees=1
  swing1Motion=Locked
  swing2LimitDegrees=1
  swing2Motion=Locked
  twistLimitDegrees=1
  twistMotion=Locked
 }
 {
  bone1Name=dyn_belt_bk_06
  bone2Name=dyn_belt_ft_05
  swing1LimitDegrees=1
  swing1Motion=Locked
  swing2LimitDegrees=1
  swing2Motion=Locked
  twistLimitDegrees=1
  twistMotion=Locked
 }
 {bone1Name=dyn_belt_ft_base_01, bone2Name=dyn_belt_ft_base_00, swing1LimitDegrees=20, swing2LimitDegrees=5, twistLimitDegrees=20}
 {bone1Name=dyn_belt_ft_00, bone2Name=dyn_belt_ft_base_01, swing1LimitDegrees=20, swing2LimitDegrees=10, twistLimitDegrees=20}
 {bone1Name=dyn_belt_ft_01, bone2Name=dyn_belt_ft_00, swing1LimitDegrees=20, swing2LimitDegrees=20, twistLimitDegrees=20}
 {bone1Name=dyn_belt_00, bone2Name=dyn_belt_ft_00, swing1LimitDegrees=5, swing2LimitDegrees=5, twistLimitDegrees=1}
 {bone1Name=dyn_belt_01, bone2Name=dyn_belt_00, swing1LimitDegrees=20, swing2LimitDegrees=30, twistLimitDegrees=0}
 {bone1Name=dyn_belt_02, bone2Name=dyn_belt_01, swing1LimitDegrees=20, swing2LimitDegrees=40, twistLimitDegrees=1}
 {bone1Name=dyn_belt_03, bone2Name=dyn_belt_02, swing1LimitDegrees=20, swing2LimitDegrees=40, twistLimitDegrees=1}
 {bone1Name=dyn_belt_04, bone2Name=dyn_belt_03, swing1LimitDegrees=20, swing2LimitDegrees=35, twistLimitDegrees=1}
 {bone1Name=dyn_belt_05, bone2Name=dyn_belt_04, swing1LimitDegrees=20, swing2LimitDegrees=40, twistLimitDegrees=1}
 {bone1Name=dyn_belt_06, bone2Name=dyn_belt_05, swing1LimitDegrees=20, swing2LimitDegrees=50, twistLimitDegrees=1}
 {bone1Name=dyn_belt_ft_02, bone2Name=dyn_belt_ft_01, swing1LimitDegrees=20, swing2LimitDegrees=20, twistLimitDegrees=20}
 {bone1Name=dyn_belt_ft_03, bone2Name=dyn_belt_ft_02, swing1LimitDegrees=20, swing2LimitDegrees=20, twistLimitDegrees=20}
 {
  bone1Name=dyn_belt_ft_04
  bone2Name=dyn_belt_ft_03
  swing1LimitDegrees=25
  swing1Motion=Locked
  swing2LimitDegrees=5
  swing2Motion=Locked
  twistLimitDegrees=1
  twistMotion=Locked
 }
 {
  bone1Name=dyn_belt_ft_05
  bone2Name=dyn_belt_ft_04
  swing1LimitDegrees=1
  swing1Motion=Locked
  swing2LimitDegrees=1
  swing2Motion=Locked
  twistLimitDegrees=5
  twistMotion=Locked
 }
 {
  bone1Name=dyn_belt_ft_06
  bone2Name=dyn_belt_ft_05
  swing1LimitDegrees=1
  swing1Motion=Locked
  swing2LimitDegrees=1
  swing2Motion=Locked
  twistLimitDegrees=1
  twistMotion=Locked
 }
 {
  bone1Name=thigh_twist_01_l
  bone2Name=dyn_belt_ft_06
  swing1LimitDegrees=45
  swing1Motion=Free
  swing2LimitDegrees=45
  swing2Motion=Free
  twistLimitDegrees=45
  twistMotion=Free
 }
 {bone1Name=dyn_skirt_bk_02_l, bone2Name=dyn_skirt_bk_01_l, swing1LimitDegrees=10, swing2LimitDegrees=20, twistLimitDegrees=0}
 {bone1Name=dyn_skirt_bk_02_r, bone2Name=dyn_skirt_bk_01_r, swing1LimitDegrees=10, swing2LimitDegrees=20, twistLimitDegrees=1}
 {bone1Name=dyn_skirt_bk_03_r, bone2Name=dyn_skirt_bk_02_r, swing1LimitDegrees=10, swing2LimitDegrees=20, twistLimitDegrees=1}
 {bone1Name=dyn_skirt_fr_02_l, bone2Name=dyn_skirt_fr_01_l, swing1LimitDegrees=5, swing2LimitDegrees=10, twistLimitDegrees=0}
 {bone1Name=dyn_jacketBack, bone2Name=spine_02, swing1LimitDegrees=30, swing2LimitDegrees=5, twistLimitDegrees=0}
 {bone1Name=dyn_jacketFront_l, bone2Name=spine_02, swing1LimitDegrees=30, swing2LimitDegrees=5, twistLimitDegrees=0}
 {bone1Name=dyn_jacketFront_r, bone2Name=spine_02, swing1LimitDegrees=30, swing2LimitDegrees=5, twistLimitDegrees=0}
 {bone1Name=dyn_jacketSideBack_l, bone2Name=spine_02, swing1LimitDegrees=30, swing2LimitDegrees=5, twistLimitDegrees=0}
 {bone1Name=dyn_jacketSideBack_r, bone2Name=spine_02, swing1LimitDegrees=30, swing2LimitDegrees=5, twistLimitDegrees=0}
 {bone1Name=dyn_jacketSide_l, bone2Name=spine_02, swing1LimitDegrees=30, swing2LimitDegrees=5, twistLimitDegrees=0}
 {bone1Name=dyn_jacketSide_r, bone2Name=spine_02, swing1LimitDegrees=30, swing2LimitDegrees=5, twistLimitDegrees=0}
 {
  bone1Name=dyn_jacketSideBack_l
  bone2Name=dyn_jacketBack
  swing1LimitDegrees=45
  swing1Motion=Free
  swing2LimitDegrees=45
  swing2Motion=Free
  twistLimitDegrees=45
  twistMotion=Free
 }
 {
  bone1Name=dyn_jacketSideBack_r
  bone2Name=dyn_jacketBack
  swing1LimitDegrees=45
  swing1Motion=Free
  swing2LimitDegrees=45
  swing2Motion=Free
  twistLimitDegrees=45
  twistMotion=Free
 }
 {bone1Name=dyn_skirt_fr_03_l, bone2Name=dyn_skirt_fr_02_l, swing1LimitDegrees=15, swing2LimitDegrees=30, twistLimitDegrees=0}
 {bone1Name=dyn_skirt_bk_03_l, bone2Name=dyn_skirt_bk_02_l, swing1LimitDegrees=5, swing2LimitDegrees=30, twistLimitDegrees=1}
 {
  bone1Name=dyn_skirt_fr_mid_r_02
  bone2Name=dyn_skirt_fr_03_l
  swing1LimitDegrees=45
  swing1Motion=Free
  swing2LimitDegrees=45
  swing2Motion=Free
  twistLimitDegrees=45
  twistMotion=Free
 }
]
