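# Physics asset for the ChessPieces part of the female medium VenomSoul body, covering a dynamic back piece and the body around it.
# Capsule bodies sit on the base back piece, its dynamic child bone, the pelvis and the left thigh.
# The dynamic bone has narrow swing limits and zero twist against the base, which links to the pelvis with free swing and twist.
# The left thigh body has no constraint.
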PhysicsAsset /BRCosmetics/Characters/Player/Female/Medium/Bodies/F_MED_VenomSoul/Meshes/F_MED_VenomSoul_Physics_ChessPieces
bodies {
 base_back_piece [{shapeType=Capsule, length=0.505, radius=0.505, shapeName=base_back_piece_capsule, rotation=<rot 0 -90 0>}]
 dyn_back_piece_01 [{
   shapeType=Capsule
   length=9.2708
   radius=5.2326
   shapeName=dyn_back_piece_01_capsule
   center=<-0.65 -0.78 3.5>
   rotation=<rot 0 97.502 -88.6104>
  }]
 pelvis [{shapeType=Capsule, length=14.6385, radius=6.9533, shapeName=pelvis_capsule, center=<0.5224 -1.106 0.4668>, rotation=<rot 0.1866 -90 -2.4067>}]
 thigh_l [{
   shapeType=Capsule
   length=31.0974
   radius=7.0303
   shapeName=thigh_l_capsule
   center=<-16.1927 1.4902 0.5504>
   rotation=<rot 0.5054 -82.7809 -91.5202>
  }]
}
constraints [
 {bone1Name=dyn_back_piece_01, bone2Name=base_back_piece, swing1LimitDegrees=10, swing2LimitDegrees=20, twistLimitDegrees=0}
 {
  bone1Name=base_back_piece
  bone2Name=pelvis
  swing1LimitDegrees=45
  swing1Motion=Free
  swing2LimitDegrees=45
  swing2Motion=Free
  twistLimitDegrees=45
  twistMotion=Free
 }
]
