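# Animation Blueprint for the F_MED_VenomSoul female character body, adding secondary motion to skirt, belts, straps, cuffs, clips and chain.
# It copies the parent mesh pose into a Control Rig driven by targeting, movement and emote state.
# Constraints, additive bone offsets, three RigidBody solvers (one skydive-only) and AnimDynamics follow.
# The EventGraph refreshes state floats, derives thigh-driven belt correction and sets gravity from a parameter asset.
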AnimBlueprint /BRCosmetics/Characters/Player/Female/Medium/Bodies/F_MED_VenomSoul/Meshes/F_MED_VenomSoul_AnimBP
target_skeleton=&/F_MED_VenomSoul_Skeleton
graphs [
 AnimGraph {nodes [
   n1 = Control Rig!(
    Source=Copy Pose From Mesh() {bUseAttachedParent=true}
    Alpha
    targetingHeadupBool:bool=Get targetingHeadUpBool() -> (targetingHeadUpBool:bool)
    isTargetingFloat:double=Get targetingfloat() -> (targetingfloat:double)
    crouchwalking:double=Get crouchWalking() -> (crouchWalking:double)
    isMoving2dFloat:double=Get IsMoving2dFloat() -> (IsMoving2dFloat:double)
    isEmoting:double=Get isEmotingFloat() -> (isEmotingFloat:double)
   ) -> (Pose) {
    alphaBoolBlend {blendOption=Linear}
    controlRigAssetReference {blueprintRigClass=&/F_MED_VenomSoul_CtrlRig.F_MED_VenomSoul_CtrlRig_C}
   }
   n2 = "Constraint - dyn_clip_front_MSTR (PPPPT)"!(
    ComponentPose=Local To Component(LocalPose=@n1)
    Alpha=1
    ConstraintWeights_0=0.0723
    ConstraintWeights_1:float=0.1633
    ConstraintWeights_2:float=0.2784
    ConstraintWeights_3:float=0.486
    ConstraintWeights_4:float=0
   ) -> (Pose) {
    alphaBoolBlend {blendOption=Linear}
    boneToModify {boneName=dyn_clip_front_MSTR}
    constraintSetup [
     {targetBone {boneName=spine_01}}
     {targetBone {boneName=spine_02}}
     {targetBone {boneName=spine_03}}
     {targetBone {boneName=spine_04}}
     {transformType=Translation, targetBone {boneName=None}}
    ]
    constraintWeights [0,1,1,1,1]
   }
   n3 = "Constraint - dyn_belt_ft_03 (P)"(ComponentPose=@n2, Alpha=0, ConstraintWeights_0=1) {
    alphaBoolBlend {blendOption=Linear}
    boneToModify {boneName=dyn_belt_ft_03}
    constraintSetup [{targetBone {boneName=thigh_twist_01_l}}]
    constraintWeights [1]
   }
   n4 = "Constraint - dyn_belt_bk_03 (P)"(ComponentPose=@n3, Alpha=0, ConstraintWeights_0=1) {
    alphaBoolBlend {blendOption=Linear}
    boneToModify {boneName=dyn_belt_bk_03}
    constraintSetup [{targetBone {boneName=thigh_twist_01_l}}]
    constraintWeights [1]
   }
   n5 = Anim Dynamics - Bone: dyn_cuff_l_01!(ComponentPose=@n4, ExternalForce="0,0,0") -> (Pose) {
    alpha=0.5
    angularDampingOverride=0.999
    angularSpringConstant=15000
    bAngularSpring=true
    bGravityOverrideInSimSpace=true
    bOverrideAngularDamping=true
    bUseGravityOverride=true
    simulationSpace=RootRelative
    alphaBoolBlend {blendOption=Linear}
    boundBone {boneName=dyn_cuff_l_01}
    componentAppliedLinearAccClamp=<5000 5000 5000>
    componentLinearAccScale=<0.2 0.2 0.55>
    componentLinearVelScale=<0.2 0.2 0.3>
    gravityOverride {z=-980}
    physicsBodyDefinitions [{
      sphereCollisionRadius=0
      boundBone {boneName=dyn_cuff_l_01}
      boxExtents=<30 30 30>
      constraintSetup {angularLimitsMax=<0 15 0>, angularLimitsMin=<0 -50 0>, angularTarget=<1 0 0>}
      localJointOffset=<4 0 0>
     }]
    relativeSpaceBone {boneName=root}
   }
   n6 = Anim Dynamics - Bone: dyn_cuff_r_01!(ComponentPose=@n5, ExternalForce="0,0,0") -> (Pose) {
    alpha=0.5
    angularDampingOverride=0.999
    angularSpringConstant=15000
    bAngularSpring=true
    bGravityOverrideInSimSpace=true
    bOverrideAngularDamping=true
    bUseGravityOverride=true
    simulationSpace=RootRelative
    alphaBoolBlend {blendOption=Linear}
    boundBone {boneName=dyn_cuff_r_01}
    componentAppliedLinearAccClamp=<5000 5000 5000>
    componentLinearAccScale=<0.2 0.2 0.55>
    componentLinearVelScale=<0.2 0.2 0.3>
    gravityOverride {z=-980}
    physicsBodyDefinitions [{
      sphereCollisionRadius=0
      boundBone {boneName=dyn_cuff_r_01}
      boxExtents=<30 30 30>
      constraintSetup {angularLimitsMax=<0 15 0>, angularLimitsMin=<0 -50 0>, angularTarget=<1 0 0>}
      localJointOffset=<4 0 0>
     }]
    relativeSpaceBone {boneName=root}
   }
   n7 = Anim Dynamics - Bone: dyn_cuff_l_04!(ComponentPose=@n6, ExternalForce="0,0,0") -> (Pose) {
    alpha=0.5
    angularDampingOverride=0.999
    angularSpringConstant=15000
    bAngularSpring=true
    bGravityOverrideInSimSpace=true
    bOverrideAngularDamping=true
    bUseGravityOverride=true
    simulationSpace=RootRelative
    alphaBoolBlend {blendOption=Linear}
    boundBone {boneName=dyn_cuff_l_04}
    componentAppliedLinearAccClamp=<5000 5000 5000>
    componentLinearAccScale=<0.2 0.2 0.55>
    componentLinearVelScale=<0.2 0.2 0.3>
    gravityOverride {z=-980}
    physicsBodyDefinitions [{
      sphereCollisionRadius=0
      boundBone {boneName=dyn_cuff_l_04}
      boxExtents=<30 30 30>
      constraintSetup {angularLimitsMax=<0 15 0>, angularLimitsMin=<0 -50 0>, angularTarget=<1 0 0>}
      localJointOffset=<4 0 0>
     }]
    relativeSpaceBone {boneName=root}
   }
   n8 = Anim Dynamics - Bone: dyn_cuff_r_04!(ComponentPose=@n7, ExternalForce="0,0,0") -> (Pose) {
    alpha=0.5
    angularDampingOverride=0.999
    angularSpringConstant=15000
    bAngularSpring=true
    bGravityOverrideInSimSpace=true
    bOverrideAngularDamping=true
    bUseGravityOverride=true
    simulationSpace=RootRelative
    alphaBoolBlend {blendOption=Linear}
    boundBone {boneName=dyn_cuff_r_04}
    componentAppliedLinearAccClamp=<5000 5000 5000>
    componentLinearAccScale=<0.2 0.2 0.55>
    componentLinearVelScale=<0.2 0.2 0.3>
    gravityOverride {z=-980}
    physicsBodyDefinitions [{
      sphereCollisionRadius=0
      boundBone {boneName=dyn_cuff_r_04}
      boxExtents=<30 30 30>
      constraintSetup {angularLimitsMax=<0 15 0>, angularLimitsMin=<0 -50 0>, angularTarget=<1 0 0>}
      localJointOffset=<4 0 0>
     }]
    relativeSpaceBone {boneName=root}
   }
   n9 = Anim Dynamics - Bone: dyn_cuff_l_02!(ComponentPose=@n8, ExternalForce="0,0,0") -> (Pose) {
    alpha=0.5
    angularDampingOverride=0.999
    angularSpringConstant=15000
    bAngularSpring=true
    bGravityOverrideInSimSpace=true
    bOverrideAngularDamping=true
    bUseGravityOverride=true
    simulationSpace=RootRelative
    alphaBoolBlend {blendOption=Linear}
    boundBone {boneName=dyn_cuff_l_02}
    componentAppliedLinearAccClamp=<5000 5000 5000>
    componentLinearAccScale=<0.2 0.2 0.55>
    componentLinearVelScale=<0.2 0.2 0.3>
    gravityOverride {z=-980}
    physicsBodyDefinitions [{
      sphereCollisionRadius=0
      boundBone {boneName=dyn_cuff_l_02}
      boxExtents=<30 30 30>
      constraintSetup {angularLimitsMax=<0 15 0>, angularLimitsMin=<0 -50 0>, angularTarget=<1 0 0>}
      localJointOffset=<4 0 0>
     }]
    relativeSpaceBone {boneName=root}
   }
   n10 = Anim Dynamics - Bone: dyn_cuff_r_02!(ComponentPose=@n9, ExternalForce="0,0,0") -> (Pose) {
    alpha=0.5
    angularDampingOverride=0.999
    angularSpringConstant=15000
    bAngularSpring=true
    bGravityOverrideInSimSpace=true
    bOverrideAngularDamping=true
    bUseGravityOverride=true
    simulationSpace=RootRelative
    alphaBoolBlend {blendOption=Linear}
    boundBone {boneName=dyn_cuff_r_02}
    componentAppliedLinearAccClamp=<5000 5000 5000>
    componentLinearAccScale=<0.2 0.2 0.55>
    componentLinearVelScale=<0.2 0.2 0.3>
    gravityOverride {z=-980}
    physicsBodyDefinitions [{
      sphereCollisionRadius=0
      boundBone {boneName=dyn_cuff_r_02}
      boxExtents=<30 30 30>
      constraintSetup {angularLimitsMax=<0 15 0>, angularLimitsMin=<0 -50 0>, angularTarget=<1 0 0>}
      localJointOffset=<4 0 0>
     }]
    relativeSpaceBone {boneName=root}
   }
   n11 = Anim Dynamics - Bone: dyn_cuff_l_03!(ComponentPose=@n10, ExternalForce="0,0,0") -> (Pose) {
    alpha=0.5
    angularDampingOverride=0.999
    angularSpringConstant=15000
    bAngularSpring=true
    bGravityOverrideInSimSpace=true
    bOverrideAngularDamping=true
    bUseGravityOverride=true
    simulationSpace=RootRelative
    alphaBoolBlend {blendOption=Linear}
    boundBone {boneName=dyn_cuff_l_03}
    componentAppliedLinearAccClamp=<5000 5000 5000>
    componentLinearAccScale=<0.2 0.2 0.55>
    componentLinearVelScale=<0.2 0.2 0.3>
    gravityOverride {z=-980}
    physicsBodyDefinitions [{
      sphereCollisionRadius=0
      boundBone {boneName=dyn_cuff_l_03}
      boxExtents=<30 30 30>
      constraintSetup {angularLimitsMax=<0 15 0>, angularLimitsMin=<0 -50 0>, angularTarget=<1 0 0>}
      localJointOffset=<4 0 0>
     }]
    relativeSpaceBone {boneName=root}
   }
   n12 = Anim Dynamics - Bone: dyn_cuff_r_03!(ComponentPose=@n11, ExternalForce="0,0,0") -> (Pose) {
    alpha=0.5
    angularDampingOverride=0.999
    angularSpringConstant=15000
    bAngularSpring=true
    bGravityOverrideInSimSpace=true
    bOverrideAngularDamping=true
    bUseGravityOverride=true
    simulationSpace=RootRelative
    alphaBoolBlend {blendOption=Linear}
    boundBone {boneName=dyn_cuff_r_03}
    componentAppliedLinearAccClamp=<5000 5000 5000>
    componentLinearAccScale=<0.2 0.2 0.55>
    componentLinearVelScale=<0.2 0.2 0.3>
    gravityOverride {z=-980}
    physicsBodyDefinitions [{
      sphereCollisionRadius=0
      boundBone {boneName=dyn_cuff_r_03}
      boxExtents=<30 30 30>
      constraintSetup {angularLimitsMax=<0 15 0>, angularLimitsMin=<0 -50 0>, angularTarget=<1 0 0>}
      localJointOffset=<4 0 0>
     }]
    relativeSpaceBone {boneName=root}
   }
   n13 = "Transform (Modify) Bone - Bone: dyn_clip_back_MSTR"(ComponentPose=@n12, Alpha=1, Translation="0,0,-5", Rotation="0,0,0", Scale="0,0,0") {
    rotationMode=BMM_Additive
    rotationSpace=BCS_BoneSpace
    scaleSpace=BCS_BoneSpace
    translationMode=BMM_Additive
    translationSpace=BCS_BoneSpace
    alphaBoolBlend {blendOption=Linear}
    boneToModify {boneName=dyn_clip_back_MSTR}
    scale=<0 0 0>
    translation {z=-5}
   }
   n14 = Get thigh_rot_r()
   n15 = "Transform (Modify) Bone - Bone: dyn_belt_ft_base_01"(ComponentPose=@n13, Alpha=@n14, Translation="0,0,0", Rotation="0,-20,0", Scale="0,0,0") {
    rotationMode=BMM_Additive
    rotationSpace=BCS_BoneSpace
    scaleSpace=BCS_BoneSpace
    translationMode=BMM_Additive
    translationSpace=BCS_BoneSpace
    alphaBoolBlend {blendOption=Linear}
    boneToModify {boneName=dyn_belt_ft_base_01}
    rotation {yaw=-20}
    scale=<0 0 0>
   }
   n16 = "Transform (Modify) Bone - Bone: dyn_belt_ft_base_00"(ComponentPose=@n15, Alpha=@n14, Translation="0,0,0", Rotation="0,-20,0", Scale="0,0,0") {
    rotationMode=BMM_Additive
    rotationSpace=BCS_BoneSpace
    scaleSpace=BCS_BoneSpace
    translationMode=BMM_Additive
    translationSpace=BCS_BoneSpace
    alphaBoolBlend {blendOption=Linear}
    boneToModify {boneName=dyn_belt_ft_base_00}
    rotation {yaw=-20}
    scale=<0 0 0>
   }
   n17 = "Constraint - dyn_skirt_fr_02_l (P)"(ComponentPose=@n16, Alpha=0.5, ConstraintWeights_0=1) {
    alphaBoolBlend {blendOption=Linear}
    boneToModify {boneName=dyn_skirt_fr_02_l}
    constraintSetup [{targetBone {boneName=thigh_twist_01_r}}]
    constraintWeights [1]
   }
   n18 = Get GravityOverride()
   n19 = RigidBody!(
    ComponentPose=@n17
    Alpha
    OverrideWorldGravity:Vector=@n18
    ExternalForce
    SimSpaceSettings:Sim Space Settings Structure=(WorldAlpha=1,VelocityScaleZ=0.6,DampingAlpha=1,MaxLinearVelocity=700,MaxAngularVelocity=8,MaxLinearAcceleration=3500,MaxAngularAcceleration=40,ExternalLinearDragV=(X=0,Y=0,Z=0),ExternalLinearVelocity=(X=0,Y=0,Z=0),ExternalAngularVelocity=(X=0,Y=0,Z=0))
   ) -> (Pose) {
    bClampLinearTranslationLimitToRefPose=true
    bOverrideWorldGravity=true
    overridePhysicsAsset=&/F_MED_VenomSoul_Physics
    simulationSpace=BaseBoneSpace
    alphaBoolBlend {blendOption=Linear}
    baseBoneRef {boneName=pelvis}
    overrideWorldGravity {z=-1080}
    simSpaceSettings {maxAngularAcceleration=40, maxAngularVelocity=8, maxLinearAcceleration=3500, maxLinearVelocity=700, velocityScaleZ=0.6, worldAlpha=1}
   }
   n20 = RigidBody(
    ComponentPose=@n19
    OverrideWorldGravity=@n18
    SimSpaceSettings=(WorldAlpha=0,VelocityScaleZ=1,DampingAlpha=1,MaxLinearVelocity=10000,MaxAngularVelocity=10000,MaxLinearAcceleration=10000,MaxAngularAcceleration=10000,ExternalLinearDragV=(X=0,Y=0,Z=0),ExternalLinearVelocity=(X=0,Y=0,Z=0),ExternalAngularVelocity=(X=0,Y=0,Z=0))
   ) {
    bClampLinearTranslationLimitToRefPose=true
    bOverrideWorldGravity=true
    overridePhysicsAsset=&/F_MED_VenomSoul_Physics_ChessPieces
    simulationSpace=BaseBoneSpace
    alphaBoolBlend {blendOption=Linear}
    baseBoneRef {boneName=pelvis}
    overrideWorldGravity {z=-1080}
   }
   n21 = RigidBody!(
    ComponentPose=@n20
    bAlphaBoolEnabled:bool=Get bIsSkyDiving()
    OverrideWorldGravity=@n18
    ExternalForce
    SimSpaceSettings=(WorldAlpha=1,VelocityScaleZ=0.6,DampingAlpha=1,MaxLinearVelocity=700,MaxAngularVelocity=8,MaxLinearAcceleration=3500,MaxAngularAcceleration=40,ExternalLinearDragV=(X=0,Y=0,Z=0),ExternalLinearVelocity=(X=0,Y=0,Z=0),ExternalAngularVelocity=(X=0,Y=0,Z=0))
   ) -> (Pose) {
    alphaInputType=Bool
    bClampLinearTranslationLimitToRefPose=true
    bOverrideWorldGravity=true
    overridePhysicsAsset=&/F_MED_VenomSoul_Skydive_Physics
    simulationSpace=BaseBoneSpace
    alphaBoolBlend {blendOption=Linear}
    baseBoneRef {boneName=pelvis}
    overrideWorldGravity {z=-1080}
    simSpaceSettings {maxAngularAcceleration=40, maxAngularVelocity=8, maxLinearAcceleration=3500, maxLinearVelocity=700, velocityScaleZ=0.6, worldAlpha=1}
   }
   n22 = Anim Dynamics - Bone: dyn_skirt_side_03_r!(ComponentPose=@n21, Alpha=1, GravityOverride:Vector="0,0,-980", ExternalForce="0,0,0") -> (Pose) {
    angularDampingOverride=0.998
    angularSpringConstant=25000
    bAngularSpring=true
    bOverrideAngularDamping=true
    bUseGravityOverride=true
    gravityScale=0.35
    simulationSpace=RootRelative
    boundBone {boneName=dyn_skirt_side_03_r}
    componentLinearAccScale=<1 1 1>
    componentLinearVelScale=<1 1 1>
    gravityOverride {z=-980}
    physicsBodyDefinitions [{
      boundBone {boneName=dyn_skirt_side_03_r}
      boxExtents=<40 40 40>
      constraintSetup {angularLimitsMax=<0 20 0>, angularLimitsMin=<0 -80 0>, angularTarget=<1 0 0>}
      localJointOffset=<4 0 0>
     }]
   }
   n23 = Anim Dynamics - Bone: dyn_skirt_fr_mid_r_02!(ComponentPose=@n22, Alpha=1, GravityOverride:Vector="0,0,-980", ExternalForce="0,0,0") -> (Pose) {
    angularDampingOverride=0.998
    angularSpringConstant=25000
    bAngularSpring=true
    bOverrideAngularDamping=true
    bUseGravityOverride=true
    gravityScale=0.35
    simulationSpace=RootRelative
    boundBone {boneName=dyn_skirt_fr_mid_r_02}
    componentLinearAccScale=<1 1 1>
    componentLinearVelScale=<1 1 1>
    gravityOverride {z=-980}
    physicsBodyDefinitions [{
      boundBone {boneName=dyn_skirt_fr_mid_r_02}
      boxExtents=<40 40 40>
      constraintSetup {angularLimitsMax=<0 0 0>, angularLimitsMin=<0 -90 0>, angularTarget=<1 0 0>}
      localJointOffset=<4 0 0>
     }]
   }
   n24 = Anim Dynamics - Bone: dyn_skirt_03_r!(ComponentPose=@n23, Alpha=1, GravityOverride:Vector="0,0,-980", ExternalForce="0,0,0") -> (Pose) {
    angularDampingOverride=0.998
    angularSpringConstant=25000
    bAngularSpring=true
    bOverrideAngularDamping=true
    bUseGravityOverride=true
    gravityScale=0.35
    simulationSpace=RootRelative
    boundBone {boneName=dyn_skirt_03_r}
    componentLinearAccScale=<1 1 1>
    componentLinearVelScale=<1 1 1>
    gravityOverride {z=-980}
    physicsBodyDefinitions [{
      boundBone {boneName=dyn_skirt_03_r}
      boxExtents=<40 40 40>
      constraintSetup {angularLimitsMax=<0 10 0>, angularLimitsMin=<0 -90 0>, angularTarget=<1 0 0>}
      localJointOffset=<4 0 0>
     }]
   }
   n25 = Anim Dynamics - Bone: dyn_shoe_strap_l(ComponentPose=@n24, Alpha=1, ExternalForce="0,0,0") {
    angularDampingOverride=0.7
    angularSpringConstant=1000
    bAngularSpring=true
    bGravityOverrideInSimSpace=true
    bUseGravityOverride=true
    simulationSpace=BoneRelative
    alphaBoolBlend {blendOption=Linear}
    boundBone {boneName=dyn_shoe_strap_l}
    componentAppliedLinearAccClamp=<5000 5000 5000>
    componentLinearAccScale=<0.1 0.1 0.2>
    componentLinearVelScale=<0.05 0.05 0.1>
    gravityOverride {z=-200}
    physicsBodyDefinitions [{
      sphereCollisionRadius=0
      boundBone {boneName=dyn_shoe_strap_l}
      boxExtents=<1 20 20>
      constraintSetup {angularLimitsMax=<0 20 0>, angularLimitsMin=<-0.1 -80 -0.1>, angularTarget=<1 0 0>}
      localJointOffset=<-1 0 0>
     }]
    relativeSpaceBone {boneName=root}
   }
   n26 = Anim Dynamics - Bone: dyn_shoe_strap_r(ComponentPose=@n25, Alpha=1, ExternalForce="0,0,0") {
    angularDampingOverride=0.7
    angularSpringConstant=1000
    bAngularSpring=true
    bGravityOverrideInSimSpace=true
    bUseGravityOverride=true
    simulationSpace=BoneRelative
    alphaBoolBlend {blendOption=Linear}
    boundBone {boneName=dyn_shoe_strap_r}
    componentAppliedLinearAccClamp=<5000 5000 5000>
    componentLinearAccScale=<0.1 0.1 0.2>
    componentLinearVelScale=<0.05 0.05 0.1>
    gravityOverride {z=-200}
    physicsBodyDefinitions [{
      sphereCollisionRadius=0
      boundBone {boneName=dyn_shoe_strap_r}
      boxExtents=<1 20 20>
      constraintSetup {angularLimitsMax=<0 20 0>, angularLimitsMin=<-0.1 -80 -0.1>, angularTarget=<1 0 0>}
      localJointOffset=<-1 0 0>
     }]
    relativeSpaceBone {boneName=root}
   }
   n27 = Anim Dynamics - Bone: dyn_clip_back_01(ComponentPose=@n26, Alpha=1, ExternalForce="0,0,0") {
    angularDampingOverride=0.8
    angularSpringConstant=1000
    bAngularSpring=true
    bGravityOverrideInSimSpace=true
    bUseGravityOverride=true
    simulationSpace=BoneRelative
    alphaBoolBlend {blendOption=Linear}
    boundBone {boneName=dyn_clip_back_01}
    componentAppliedLinearAccClamp=<5000 5000 5000>
    componentLinearAccScale {z=0.8}
    componentLinearVelScale {z=0.4}
    gravityOverride {z=-1000}
    physicsBodyDefinitions [{
      sphereCollisionRadius=0
      boundBone {boneName=dyn_clip_back_01}
      boxExtents=<0.5 30 30>
      constraintSetup {angularLimitsMax=<0 0 0>, angularLimitsMin=<-0.1 -80 -0.1>, angularTarget=<1 0 0>}
      localJointOffset=<1 0 0>
     }]
    relativeSpaceBone {boneName=root}
   }
   n28 = Anim Dynamics - Bone: dyn_wrist_strap_l(ComponentPose=@n27, Alpha=1, ExternalForce="0,0,0") {
    angularDampingOverride=0.9999
    angularSpringConstant=10000
    bAngularSpring=true
    bOverrideAngularDamping=true
    bUseGravityOverride=true
    gravityScale=0.35
    simulationSpace=RootRelative
    boundBone {boneName=dyn_wrist_strap_l}
    componentLinearAccScale=<1 1 1>
    componentLinearVelScale=<1 1 1>
    gravityOverride {z=-980}
    physicsBodyDefinitions [{
      boundBone {boneName=dyn_wrist_strap_l}
      boxExtents=<20 20 20>
      constraintSetup {angularLimitsMax=<0 25 0>, angularLimitsMin=<0 -25 0>, angularTarget=<1 0 0>}
      localJointOffset=<4 0 0>
     }]
   }
   n29 = Anim Dynamics - Bone: dyn_wrist_strap_r(ComponentPose=@n28, Alpha=1, ExternalForce="0,0,0") {
    angularDampingOverride=0.9999
    angularSpringConstant=10000
    bAngularSpring=true
    bOverrideAngularDamping=true
    bUseGravityOverride=true
    gravityScale=0.35
    simulationSpace=RootRelative
    boundBone {boneName=dyn_wrist_strap_r}
    componentLinearAccScale=<1 1 1>
    componentLinearVelScale=<1 1 1>
    gravityOverride {z=-980}
    physicsBodyDefinitions [{
      boundBone {boneName=dyn_wrist_strap_r}
      boxExtents=<20 20 20>
      constraintSetup {angularLimitsMax=<0 25 0>, angularLimitsMin=<0 -25 0>, angularTarget=<1 0 0>}
      localJointOffset=<4 0 0>
     }]
   }
   n30 = Anim Dynamics - Bone: dyn_chain_02!(ComponentPose=@n29, Alpha=1, GravityOverride:Vector="0,0,-980", ExternalForce="0,0,0") -> (Pose) {
    angularDampingOverride=0.999
    angularSpringConstant=15000
    bAngularSpring=true
    bOverrideAngularDamping=true
    bUseGravityOverride=true
    bUseSphericalLimits=true
    gravityScale=0.35
    simulationSpace=RootRelative
    boundBone {boneName=dyn_chain_02}
    componentLinearAccScale=<1 1 1>
    componentLinearVelScale=<1 1 1>
    gravityOverride {z=-980}
    physicsBodyDefinitions [{
      boundBone {boneName=dyn_chain_02}
      boxExtents=<30 30 30>
      constraintSetup {angularLimitsMax=<0 0 0>, angularLimitsMin=<0 -30 0>, angularTarget=<1 0 0>}
      localJointOffset=<4 0 0>
     }]
    sphericalLimits [{limitRadius=9.1095, drivingBone {boneName=spine_05}, sphereLocalOffset=<-2.2698 1.134 0.185>}]
   }
   n31 = Anim Dynamics - Bone: dyn_chain_03!(ComponentPose=@n30, Alpha=1, GravityOverride:Vector="0,0,-980", ExternalForce="0,0,0") -> (Pose) {
    angularDampingOverride=0.999
    angularSpringConstant=15000
    bAngularSpring=true
    bOverrideAngularDamping=true
    bUseGravityOverride=true
    bUseSphericalLimits=true
    gravityScale=0.35
    simulationSpace=RootRelative
    boundBone {boneName=dyn_chain_03}
    componentLinearAccScale=<1 1 1>
    componentLinearVelScale=<1 1 1>
    gravityOverride {z=-980}
    physicsBodyDefinitions [{
      boundBone {boneName=dyn_chain_03}
      boxExtents=<30 30 30>
      constraintSetup {angularLimitsMax=<0 0 0>, angularLimitsMin=<0 -40 0>, angularTarget=<1 0 0>}
      localJointOffset=<4 0 0>
     }]
    sphericalLimits [{limitRadius=9.223, drivingBone {boneName=spine_05}, sphereLocalOffset=<-5.732 3.4266 0>}]
   }
   n32 = Anim Dynamics - Bone: dyn_side_piece_01!(ComponentPose=@n31, Alpha=1, GravityOverride:Vector="0,0,-980", ExternalForce="0,0,0") -> (Pose) {
    angularDampingOverride=0.999
    angularSpringConstant=35000
    bAngularSpring=true
    bOverrideAngularDamping=true
    bUseGravityOverride=true
    gravityScale=0.35
    simulationSpace=RootRelative
    boundBone {boneName=dyn_side_piece_01}
    componentLinearAccScale=<1 1 1>
    componentLinearVelScale=<1 1 1>
    gravityOverride {z=-980}
    physicsBodyDefinitions [{
      boundBone {boneName=dyn_side_piece_01}
      boxExtents=<60 60 60>
      constraintSetup {angularLimitsMax=<0 15 15>, angularLimitsMin=<0 -15 -15>, angularTarget=<1 0 0>}
      localJointOffset=<4 0 0>
     }]
   }
   n33 = "Transform (Modify) Bone - Bone: dyn_belt_ft_base_01"(ComponentPose=@n32, Alpha=1, Translation="0.000,2,0.000", Rotation="0,0,0", Scale="0,0,0") {
    rotationMode=BMM_Additive
    translationMode=BMM_Additive
    alphaBoolBlend {blendOption=Linear}
    boneToModify {boneName=dyn_belt_ft_base_01}
    scale=<0 0 0>
    translation {y=2}
   }
   Output Pose(Result=Component To Local(ComponentPose=@n33)) {layerGroup=None}
   Get FrontEndAndSkyDive() -> (FrontEndAndSkyDive:double)
   Get isInFrontEndFloat() -> (isInFrontEndFloat:double)
  ]}
 EventGraph {nodes [
   Event OnBeginCrouching()
   > Set Crouch(Crouch:double=0) -> (Output_Get:double)
   > Set Crouch_mask(Crouch_mask:double=1) -> (Output_Get:double)
   Event OnEndCrouching()
   > Set Crouch(Crouch=0.3)
   > Set Crouch_mask(Crouch_mask=0)
   Event BlueprintUpdateAnimation()
   > n1 = Sequence!(execute) -> (then_0, then_1, then_2:Exec, then_3:Exec, then_4:Exec, then_5:Exec, then_6:Exec, then_7:Exec, then_8:Exec)
   Set Skel_Mesh(execute=@n1.then_1, Skel_Mesh=GetOwnerSkeletalMeshComponent())
   n2 = GetOwnerSkeletalMeshComponent()
   Set thigh_rot_r(
    execute=@n1.then_2
    thigh_rot_r=GetAlphaFromReferenceTransformDelta(SkeletalMeshComponent=@n2, BoneName=thigh_r, BaseName=pelvis, InputRange=(X=0,Y=-90), TwistAxis="0,0,1")
   )
   n3 = GetAlphaFromReferenceTransformDelta(SkeletalMeshComponent=@n2, BoneName=head, BaseName=spine_05, DeltaComponentType=EulerY, InputRange=(X=5,Y=25), TwistAxis="0,0,1")
   > Set targetingHeadUpBool(targetingHeadUpBool:bool=AND Boolean(A=Get bIsTargeting(), B="float > float"(A=@n3, B=0.01))) -> (Output_Get:bool)
   > Set crouchBool(crouchBool:bool=Get bIsCrouching()) -> (Output_Get:bool)
   > Set targetingfloat(targetingfloat:double="To Float (Boolean)"(InBool=Get bIsTargeting())) -> (Output_Get:double)
   n4 = Get bIsMoving2D()
   n5 = Get bIsCrouching()
   > Set crouchWalking(crouchWalking:double="To Float (Boolean)"(InBool=AND Boolean(A=@n5, B=@n4))) -> (Output_Get:double)
   n6 = NOT Boolean(A=@n5)
   > Set IsMoving2dFloat(IsMoving2dFloat:double="To Float (Boolean)"(InBool=AND Boolean(A=@n6, B=@n4))) -> (Output_Get:double)
   > Set isEmotingFloat(isEmotingFloat:double="To Float (Boolean)"(InBool=Get bIsPlayingEmote())) -> (Output_Get:double)
   n7 = Branch(execute=@n1.then_3, Condition=Get bIsInFrontEnd())
   > Set isInFrontEndFloat(isInFrontEndFloat:double=0.5) -> (Output_Get:double)
   > Set isSkyDivingFloat(isSkyDivingFloat:double=1) -> (Output_Get:double)
   n8 = Branch(execute=@n7.else, Condition=Get bIsSkyDiving())
   > Set isSkyDivingFloat(isSkyDivingFloat=0)
   Set isInFrontEndFloat(execute=@n8.else, isInFrontEndFloat=1)
   > Set isSkyDivingFloat(isSkyDivingFloat=1)
   n9 = "Gravity Override State Function (Parameter Asset) (Nativized)"(
    execute=@n1.then_0
    CharacterPartAnimInstance=Self-Reference()
    GameTimeInSeconds=GetGameTimeInSeconds()
    Parameters=&/F_MED_VenomSoul_Parameters
   )
   > Set GravityOverride(GravityOverride=@n9)
   > Set RBAN Settings Out(RBAN Settings Out=@n9.OutRigidBodyAnimNodeParameters)
   Set FrontEndAndSkyDive(
    execute=@n1.then_4
    FrontEndAndSkyDive:double=float * float(A=Get isSkyDivingFloat() -> (isSkyDivingFloat:double), B=Get isInFrontEndFloat()) {is=title K2Node_PromotableOperator}
   ) -> (Output_Get:double)
  ]}
]
variables [
 Crouch {type=number}
 Crouch_mask {type=number}
 crouchBool {type=boolean}
 crouchWalking {type=number}
 FrontEndAndSkyDive {type=number}
 GravityOverride {type=Vector}
 isEmotingFloat {type=number}
 isInFrontEndFloat {type=number}
 IsMoving2dFloat {type=number}
 isSkyDivingFloat {type=number}
 RBAN Settings Out {type=FortRigidBodyAnimNodeParameters}
 Skel_Mesh {type=SkeletalMeshComponent}
 targetingfloat {type=number}
 targetingHeadUpBool {type=boolean}
 thigh_rot_r {type=number}
]
