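# Control Rig for the VenomSoul female face accessory that procedurally corrects collar, hair and lapel bones from body pose.
# A four-way sequence runs SphericalPoseReaders on upper arms, left clavicle and head, feeding their outputs as ModifyTransforms weights.
# Hair also gets a constant offset, and paired left-arm readers combine through Max for the left collar.
# The isEmoting variable switches the left lapel between two differently driven corrections; right-side corrections ignore it.
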ControlRigBlueprint /BRCosmetics/Characters/Player/Female/Medium/Bodies/F_MED_VenomSoul/Meshes/Parts/F_MED_VenomSoul_FaceAcc_CtrlRig
graphs [
 RigVMFunctionLibrary {links [], nodes []}
 RigVMFunction_Sequence_1_ContainedGraph {nodes [
   n1 = RigVMFunctionEntryNode()
   RigVMFunction_Sequence = &RigVMFunction_Sequence(ExecuteContext=@n1.ExecuteContext)
   n2 = &RigVMFunction_Sequence(ExecuteContext=@RigVMFunction_Sequence.B)
   n3 = &RigVMFunction_Sequence(ExecuteContext=@n2.B)
   RigVMFunctionReturnNode(A=@RigVMFunction_Sequence.A, B=@n2.A, C=@n3.A, D=@n3.B)
  ]}
 RigVMModel {nodes [
   RigUnit_BeginExecution = &RigUnit_BeginExecution()
   RigVMFunction_Sequence = RigVMAggregateNode(ExecuteContext=@RigUnit_BeginExecution.ExecutePin)
   n1 = &RigUnit_SphericalPoseReader(
    ActiveRegionScaleFactors=(PositiveWidth=1,NegativeWidth=1,PositiveHeight=1,NegativeHeight=1)
    ActiveRegionSize=0.19
    Debug=(bDrawDebug=false,bDrawLocalAxes=true,DebugScale=25,DebugSegments=20,DebugThickness=0.25)
    DriverAxis=(X=1,Y=0,Z=0)
    DriverItem=(Type=Bone,Name="upperarm_l")
    FalloffRegionScaleFactors=(PositiveWidth=1,NegativeWidth=1,PositiveHeight=1,NegativeHeight=1)
    FalloffSize=0.25
    FlipWidthScaling=false
    OptionalParentItem=(Type=Bone,Name="clavicle_l")
    RotationOffset=(X=40,Y=-10,Z=0)
    ExecutePin=@RigVMFunction_Sequence.A
   )
   n2 = &RigUnit_SphericalPoseReader(
    ActiveRegionScaleFactors=(PositiveWidth=1,NegativeWidth=1,PositiveHeight=1,NegativeHeight=1)
    ActiveRegionSize=0.09
    Debug=(bDrawDebug=false,bDrawLocalAxes=true,DebugScale=25,DebugSegments=20,DebugThickness=0.25)
    DriverAxis=(X=1,Y=0,Z=0)
    DriverItem=(Type=Bone,Name="upperarm_l")
    FalloffRegionScaleFactors=(PositiveWidth=1,NegativeWidth=1,PositiveHeight=1,NegativeHeight=1)
    FalloffSize=0.15
    FlipWidthScaling=false
    OptionalParentItem=(Type=Bone,Name="clavicle_l")
    RotationOffset=(X=-20,Y=-40,Z=0)
    ExecutePin=@RigVMFunction_Sequence.B
   )
   Maximum = &RigVMFunction_MathDoubleMax(A=@n1.OutputParam, B=@n2.OutputParam)
   &RigUnit_ModifyTransforms(
    ItemToModify=((Item=(Type=Bone,Name="dyn_collar_01_l"),Transform=(Rotation=(X=0,Y=0,Z=0.8192,W=0.5736),Translation=(X=0,Y=0,Z=0),Scale3D=(X=1,Y=1,Z=1))))
    WeightMinimum=0
    Weight=@Maximum.Result
    ExecutePin=@n1.ExecutePin
   )
   > n3 = &RigUnit_SphericalPoseReader(
    ActiveRegionScaleFactors=(PositiveWidth=1,NegativeWidth=1,PositiveHeight=1,NegativeHeight=1)
    ActiveRegionSize=0.15
    Debug=(bDrawDebug=false,bDrawLocalAxes=True,DebugScale=25,DebugSegments=20,DebugThickness=0.25)
    DriverAxis=(X=-1,Y=0,Z=0)
    DriverItem=(Type=Bone,Name="upperarm_r")
    FalloffRegionScaleFactors=(PositiveWidth=1,NegativeWidth=1,PositiveHeight=1,NegativeHeight=1)
    FalloffSize=0.3
    FlipHeightScaling=true
    FlipWidthScaling=true
    OptionalParentItem=(Type=Bone,Name="clavicle_r")
    RotationOffset=(X=-120,Y=30,Z=0)
   )
   > &RigUnit_ModifyTransforms(
    ItemToModify=((Item=(Type=Bone,Name="dyn_collar_01_r"),Transform=(Rotation=(X=0,Y=0,Z=-0.8192,W=0.5736),Translation=(X=0,Y=0,Z=0),Scale3D=(X=1,Y=1,Z=1))))
    WeightMinimum=0
    Weight=@n3.OutputParam
   )
   &RigUnit_ModifyTransforms(
    ItemToModify="((Item=(Type=Bone,Name=\"dyn_hair_fr_l_01\"),Transform=(Rotation=(X=0,Y=-0.0872,Z=0,W=0.9962),Translation=(X=0,Y=0,Z=0),Scale3D=(X=1,Y=1,Z=1))),(Item=(Type=Bone,Name=\"dyn_hair_l_00\"),Transform=(Rotation=(X=0,Y=0,Z=-0.0436,W=0.999),Translation=(X=0,Y=0,Z=0),S ...(truncated)"
    Weight=1
    WeightMinimum=0
    ExecutePin=@RigVMFunction_Sequence.C
   )
   > n4 = &RigUnit_SphericalPoseReader(
    ActiveRegionScaleFactors=(PositiveWidth=1,NegativeWidth=1.5,PositiveHeight=1,NegativeHeight=1.8)
    ActiveRegionSize=0.1
    Debug=(bDrawDebug=true,bDrawLocalAxes=false,DebugScale=25,DebugSegments=20,DebugThickness=0.25)
    DriverAxis=(X=0,Y=1,Z=0)
    DriverItem=(Type=Bone,Name="head")
    FalloffRegionScaleFactors=(PositiveWidth=1,NegativeWidth=2,PositiveHeight=1,NegativeHeight=3)
    FalloffSize=0.2
    FlipWidthScaling=false
    OptionalParentItem=(Type=Bone,Name="None")
    RotationOffset=(X=130,Y=0,Z=0)
   )
   > &RigUnit_ModifyTransforms(
    ItemToModify="((Item=(Type=Bone,Name=\"dyn_hair_fr_l_00\"),Transform=(Rotation=(X=0,Y=0.0436,Z=0,W=0.999),Translation=(X=0,Y=0,Z=0),Scale3D=(X=1,Y=1,Z=1))),(Item=(Type=Bone,Name=\"dyn_hair_l_00\"),Transform=(Rotation=(X=0,Y=0.0436,Z=0,W=0.999),Translation=(X=0,Y=0,Z=0),Scal ...(truncated)"
    WeightMinimum=0
    Weight=@n4.OutputParam
   )
   n5 = &RigUnit_SphericalPoseReader(
    ActiveRegionScaleFactors=(PositiveWidth=1,NegativeWidth=1,PositiveHeight=1.6,NegativeHeight=0.5)
    ActiveRegionSize=0.15
    Debug=(bDrawDebug=false,bDrawLocalAxes=false,DebugScale=25,DebugSegments=20,DebugThickness=0.25)
    DriverAxis=(X=1,Y=0,Z=0)
    DriverItem=(Type=Bone,Name="upperarm_l")
    FalloffRegionScaleFactors=(PositiveWidth=1,NegativeWidth=1,PositiveHeight=1,NegativeHeight=0.5)
    FalloffSize=0.4
    FlipWidthScaling=false
    OptionalParentItem=(Type=Bone,Name="None")
    RotationOffset=(X=85,Y=0,Z=-30)
    ExecutePin=@RigVMFunction_Sequence.D
   )
   Subtract = &RigVMFunction_MathDoubleSub(A=1, B=RigVMVariableNode() {variable=isEmoting}.Value)
   Absolute = &RigVMFunction_MathDoubleAbs(Value=@Subtract.Result)
   n6 = &RigVMFunction_AlphaInterp(InterpSpeedDecreasing=5, InterpSpeedIncreasing=5, Scale=1, bClampResult=False, bInterpResult=False, Value=@Absolute.Result)
   > &RigUnit_ModifyTransforms(
    ItemToModify=((Item=(Type=Bone,Name="dyn_lapel_l"),Transform=(Rotation=(X=0,Y=0,Z=0.4226,W=0.9063),Translation=(X=0,Y=2,Z=0),Scale3D=(X=1,Y=1,Z=1))))
    Weight=&RigVMFunction_MathDoubleMul(B=@n6.Result, A=@n5.OutputParam).Result
   )
   > n7 = &RigUnit_SphericalPoseReader(
    ActiveRegionScaleFactors=(PositiveWidth=1,NegativeWidth=1,PositiveHeight=0.7,NegativeHeight=2)
    ActiveRegionSize=0.15
    Debug=(bDrawDebug=false,bDrawLocalAxes=True,DebugScale=25,DebugSegments=20,DebugThickness=0.25)
    DriverAxis=(X=-1,Y=0,Z=0)
    DriverItem=(Type=Bone,Name="upperarm_r")
    FalloffRegionScaleFactors=(PositiveWidth=1,NegativeWidth=1,PositiveHeight=0.7,NegativeHeight=1)
    FalloffSize=0.4
    FlipHeightScaling=true
    FlipWidthScaling=true
    OptionalParentItem=(Type=Bone,Name="clavicle_r")
    RotationOffset=(X=-80,Y=0,Z=20)
   )
   > &RigUnit_ModifyTransforms(
    ItemToModify=((Item=(Type=Bone,Name="dyn_lapel_r"),Transform=(Rotation=(X=-0.1485,Y=0.212,Z=-0.7912,W=0.554),Translation=(X=0,Y=-2,Z=0),Scale3D=(X=1,Y=1,Z=1))))
    Weight=@n7.OutputParam
   )
   > &RigUnit_ModifyTransforms(
    ItemToModify=((Item=(Type=Bone,Name="dyn_lapel_lower_r"),Transform=(Rotation=(X=0,Y=0,Z=-0.866,W=0.5),Translation=(X=0,Y=-2,Z=0),Scale3D=(X=1,Y=1,Z=1))))
    Weight=@n7.OutputParam
   )
   > n8 = &RigUnit_SphericalPoseReader(
    ActiveRegionScaleFactors=(PositiveWidth=0.4,NegativeWidth=3,PositiveHeight=1.5,NegativeHeight=1)
    ActiveRegionSize=0.1
    Debug=(bDrawDebug=false,bDrawLocalAxes=false,DebugScale=25,DebugSegments=20,DebugThickness=0.25)
    DriverAxis=(X=1,Y=0,Z=0)
    DriverItem=(Type=Bone,Name="upperarm_l")
    FalloffRegionScaleFactors=(PositiveWidth=0.2,NegativeWidth=2,PositiveHeight=0.8,NegativeHeight=1.5)
    FalloffSize=0.3
    FlipWidthScaling=false
    OptionalParentItem=(Type=Bone,Name="clavicle_l")
    RotationOffset=(X=85,Y=0,Z=-40)
   )
   > n9 = &RigUnit_SphericalPoseReader(
    ActiveRegionScaleFactors=(PositiveWidth=1,NegativeWidth=1,PositiveHeight=1,NegativeHeight=1)
    ActiveRegionSize=0.1
    Debug=(bDrawDebug=false,bDrawLocalAxes=false,DebugScale=25,DebugSegments=20,DebugThickness=0.25)
    DriverAxis=(X=1,Y=0,Z=0)
    DriverItem=(Type=Bone,Name="clavicle_l")
    FalloffRegionScaleFactors=(PositiveWidth=1,NegativeWidth=1,PositiveHeight=1,NegativeHeight=1)
    FalloffSize=0.2
    FlipWidthScaling=false
    OptionalParentItem=(Type=Bone,Name="spine_05")
    RotationOffset=(X=0,Y=-85,Z=40)
   )
   n10 = &RigVMFunction_AlphaInterp(
    InterpSpeedDecreasing=5
    InterpSpeedIncreasing=5
    Scale=1
    bClampResult=False
    bInterpResult=False
    Value=RigVMVariableNode() {variable=isEmoting}.Value
   )
   > &RigUnit_ModifyTransforms(
    ItemToModify=((Item=(Type=Bone,Name="dyn_lapel_l"),Transform=(Rotation=(X=0,Y=0,Z=0.4226,W=0.9063),Translation=(X=0,Y=4,Z=0),Scale3D=(X=1,Y=1,Z=1))))
    Weight=&RigVMFunction_MathDoubleMul(B=@n10.Result, A=&RigVMFunction_MathDoubleMax(A=@n8.OutputParam, B=@n9.OutputParam).Result).Result
   )
  ]}
]
hierarchy [
 root {loc=<0 0 0>, rot=<0 0 0 1>}
 attach {parent=root, loc=<0 0 0>, rot=<0 0 0 1>}
 pelvis {parent=root, loc=<0 -0.5301 98.0571>, rot=<0.0127 0.707 -0.0127 -0.707>}
 spine_01 {parent=pelvis, loc=<6.1208 -0.9328 0>, rot=<0 0 0.034 -0.9994>}
 spine_02 {parent=spine_01, loc=<6.5996 -0.2095 0>, rot=<0 0 0.0238 -0.9997>}
 spine_03 {parent=spine_02, loc=<6.4914 -0.1237 0>, rot=<0 0 0.0149 -0.9999>}
 spine_04 {parent=spine_03, loc=<6.9598 0.0794 0>, rot=<0 0 -0.0132 -0.9999>}
 spine_05 {parent=spine_04, loc=<6.9987 0 0>, rot=<0 0 -0.0015 -1>}
 clavicle_l {parent=spine_05, loc=<5.4831 -0.1053 -3.2946>, rot=<0.0465 -0.7289 -0.009 -0.6829>}
 upperarm_l {parent=clavicle_l, loc=<9.8658 -1.2136 -0.4105>, rot=<0.0164 -0.3245 -0.0269 -0.9454>}
 lowerarm_l {parent=upperarm_l, loc=<26.465 -0.8345 0.259>, rot=<0.0036 0.0161 -0.2298 -0.9731>}
 hand_l {parent=lowerarm_l, loc=<22.7355 -0.3203 -0.6105>, rot=<-0.6784 -0.0223 0.0735 -0.7307>}
 index_metacarpal_l {parent=hand_l, loc=<1.6535 0.0183 -2.96>, rot=<0.0158 -0.0124 -0.0173 -0.9996>}
 index_01_l {parent=index_metacarpal_l, loc=<8.3008 0.0357 -0.0524>, rot=<-0.0044 0.0485 0.0924 -0.9945>}
 index_02_l {parent=index_01_l, loc=<3.3587 -0.3004 0.0452>, rot=<-0.002 0.0027 0.3225 -0.9466>}
 index_03_l {parent=index_02_l, loc=<2.5578 -0.123 -0.0252>, rot=<-0.002 0.0067 0.2558 -0.9667>}
 middle_metacarpal_l {parent=hand_l, loc=<2.1856 -0.1725 -0.5984>, rot=<0.006 0.0168 -0.0329 -0.9993>}
 middle_01_l {parent=middle_metacarpal_l, loc=<7.3275 -0.2447 0.0129>, rot=<-0.0208 0.0042 0.0858 -0.9961>}
 middle_02_l {parent=middle_01_l, loc=<4.0123 0.1401 0.1175>, rot=<-0.0218 0.0258 0.3512 -0.9357>}
 middle_03_l {parent=middle_02_l, loc=<2.6077 0.0534 0.1373>, rot=<0.0082 0.0188 0.1708 -0.9851>}
 pinky_metacarpal_l {parent=hand_l, loc=<1.7616 -0.9753 3.3814>, rot=<-0.0397 0.0814 -0.069 -0.9935>}
 pinky_01_l {parent=pinky_metacarpal_l, loc=<7.1709 -0.4831 -0.1094>, rot=<-0.0654 -0.0035 0.2242 -0.9723>}
 pinky_02_l {parent=pinky_01_l, loc=<2.8274 0.0893 -0.1043>, rot=<0.0037 0.0011 0.2854 -0.9584>}
 pinky_03_l {parent=pinky_02_l, loc=<2.2046 0.0307 -0.0078>, rot=<-0.0002 0.0098 0.1445 -0.9895>}
 ring_metacarpal_l {parent=hand_l, loc=<2.2074 -0.3625 1.3638>, rot=<-0.0284 0.0722 -0.0458 -0.9959>}
 ring_01_l {parent=ring_metacarpal_l, loc=<7.0295 -0.6197 -0.1446>, rot=<-0.0489 -0.0252 0.1892 -0.9804>}
 ring_02_l {parent=ring_01_l, loc=<3.6174 0.4647 -0.1027>, rot=<0 0 0.2806 -0.9598>}
 ring_03_l {parent=ring_02_l, loc=<2.4152 0.114 0.0199>, rot=<0 0 0.1189 -0.9929>}
 thumb_01_l {parent=hand_l, loc=<1.3118 -0.3191 -2.459>, rot=<0.6546 -0.4042 0.0664 -0.6354>}
 thumb_02_l {parent=thumb_01_l, loc=<5.9336 -0.0448 0.0401>, rot=<-0.0037 0.0213 0.298 -0.9543>}
 thumb_03_l {parent=thumb_02_l, loc=<2.2223 0.3682 -0.0907>, rot=<0.0118 -0.013 0.0066 -0.9998>}
 weapon_l {parent=hand_l, loc=<-0.1974 -3.4111 0.3811>, rot=<0 0 0 1>}
 deform_wrist_dn_base_l {parent=hand_l, loc=<0 -1.5326 0>, rot=<0 0 0 -1>}
 deform_wrist_dn_l {parent=deform_wrist_dn_base_l, loc=<0 0 0>, rot=<0 0 0 1>}
 deform_wrist_up_base_l {parent=hand_l, loc=<0 2.3009 0>, rot=<0 0 0 -1>}
 deform_wrist_up_l {parent=deform_wrist_up_base_l, loc=<0 0 0>, rot=<0 0 0 1>}
 lowerarm_twist_01_l {parent=lowerarm_l, loc=<18.316 0 0>, rot=<0 0 0 1>}
 lowerarm_twist_02_l {parent=lowerarm_l, loc=<10.6707 0 0>, rot=<0 0 0 1>}
 dyn_wrist_strap_l_MSTR {parent=lowerarm_twist_02_l, loc=<2.6826 -4.0099 0.9184>, rot=<-0.4756 0.4929 -0.5226 -0.5077>}
 dyn_wrist_strap_l {parent=dyn_wrist_strap_l_MSTR, loc=<0 0 0>, rot=<0 0 0 1>}
 deform_elbow_in_base_l {parent=lowerarm_l, loc=<0.1834 3.6062 -0.2572>, rot=<0 0 0 -1>}
 deform_elbow_in_l {parent=deform_elbow_in_base_l, loc=<0 0 0>, rot=<0 0 0 1>}
 deform_elbow_out_base_l {parent=lowerarm_l, loc=<0 -4 0>, rot=<0 0 0 1>}
 deform_elbow_out_l {parent=deform_elbow_out_base_l, loc=<0 0 0>, rot=<0 0 0 1>}
 upperarm_twist_01_l {parent=upperarm_l, loc=<7.2135 -0.1134 0>, rot=<0 0 0 1>}
 dyn_cuff_base_l {parent=upperarm_twist_01_l, loc=<-3.0363 0 0>, rot=<0 0 0 1>}
 dyn_cuff_l_01 {parent=dyn_cuff_base_l, loc=<-0.4034 1.143 5.6335>, rot=<-0.0074 0.045 0.0016 -0.999>}
 dyn_cuff_l_03 {parent=dyn_cuff_base_l, loc=<-0.1337 -5.4568 0.1445>, rot=<-0.5628 0.0674 -0.0607 -0.8216>}
 dyn_cuff_l_02_MSTR {parent=dyn_cuff_base_l, loc=<0.5815 6.3233 -0.0655>, rot=<0.5322 0.0497 0.0592 -0.8431>}
 dyn_cuff_l_02 {parent=dyn_cuff_l_02_MSTR, loc=<0 0 0>, rot=<0 0 0 1>}
 dyn_cuff_l_04_MSTR {parent=dyn_cuff_base_l, loc=<0.5614 -0.0006 -5.7757>, rot=<0.9983 -0.0003 0.045 -0.0364>}
 dyn_cuff_l_04 {parent=dyn_cuff_l_04_MSTR, loc=<0 0 0>, rot=<0 0 0 1>}
 deform_upperarm_bk_base_l {parent=upperarm_twist_01_l, loc=<-6.4053 -4.3746 0.314>, rot=<0 0 0 -1>}
 deform_upperarm_bk_l {parent=deform_upperarm_bk_base_l, loc=<0 0 0>, rot=<0 0 0 1>}
 deform_upperarm_fr_base_l {parent=upperarm_twist_01_l, loc=<-5.6312 5.092 0.3354>, rot=<0 0 0 -1>}
 deform_upperarm_fr_l {parent=deform_upperarm_fr_base_l, loc=<0 0 0>, rot=<0 0 0 1>}
 deform_upperarm_up_base_l {parent=upperarm_twist_01_l, loc=<-8.891 0.4825 3.9517>, rot=<0 0 0 -1>}
 deform_upperarm_up_l {parent=deform_upperarm_up_base_l, loc=<0 0 0>, rot=<0 0 0 1>}
 upperarm_twist_02_l {parent=upperarm_l, loc=<19.7998 -0.1134 0>, rot=<0 0 0 1>}
 clavicle_r {parent=spine_05, loc=<5.4831 -0.1053 3.2946>, rot=<0.7289 0.0465 -0.6829 0.009>}
 upperarm_r {parent=clavicle_r, loc=<-9.8658 1.2136 0.4104>, rot=<0.0164 -0.3245 -0.0269 -0.9454>}
 lowerarm_r {parent=upperarm_r, loc=<-26.4651 0.8345 -0.259>, rot=<0.0036 0.0161 -0.2298 -0.9731>}
 hand_r {parent=lowerarm_r, loc=<-22.7356 0.3203 0.6106>, rot=<-0.6784 -0.0223 0.0735 -0.7307>}
 index_metacarpal_r {parent=hand_r, loc=<-1.6534 -0.0181 2.9601>, rot=<0.0158 -0.0124 -0.0173 -0.9996>}
 index_01_r {parent=index_metacarpal_r, loc=<-8.3008 -0.0357 0.0524>, rot=<-0.0044 0.0485 0.0924 -0.9945>}
 index_02_r {parent=index_01_r, loc=<-3.3589 0.3003 -0.0453>, rot=<-0.002 0.0027 0.3225 -0.9466>}
 index_03_r {parent=index_02_r, loc=<-2.5578 0.123 0.0252>, rot=<-0.002 0.0067 0.2558 -0.9667>}
 middle_metacarpal_r {parent=hand_r, loc=<-2.1856 0.1726 0.5984>, rot=<0.006 0.0168 -0.0329 -0.9993>}
 middle_01_r {parent=middle_metacarpal_r, loc=<-7.3273 0.2447 -0.0129>, rot=<-0.0208 0.0042 0.0858 -0.9961>}
 middle_02_r {parent=middle_01_r, loc=<-4.0124 -0.14 -0.1175>, rot=<-0.0218 0.0258 0.3512 -0.9357>}
 middle_03_r {parent=middle_02_r, loc=<-2.6077 -0.0533 -0.1373>, rot=<0.0082 0.0188 0.1708 -0.9851>}
 pinky_metacarpal_r {parent=hand_r, loc=<-1.7617 0.9753 -3.3814>, rot=<-0.0397 0.0814 -0.069 -0.9935>}
 pinky_01_r {parent=pinky_metacarpal_r, loc=<-7.1707 0.4832 0.1094>, rot=<-0.0654 -0.0035 0.2242 -0.9723>}
 pinky_02_r {parent=pinky_01_r, loc=<-2.8274 -0.0892 0.1043>, rot=<0.0037 0.0011 0.2854 -0.9584>}
 pinky_03_r {parent=pinky_02_r, loc=<-2.2045 -0.031 0.0078>, rot=<-0.0002 0.0098 0.1445 -0.9895>}
 ring_metacarpal_r {parent=hand_r, loc=<-2.2073 0.3627 -1.3638>, rot=<-0.0284 0.0722 -0.0458 -0.9959>}
 ring_01_r {parent=ring_metacarpal_r, loc=<-7.0297 0.6197 0.1446>, rot=<-0.0489 -0.0252 0.1892 -0.9804>}
 ring_02_r {parent=ring_01_r, loc=<-3.6171 -0.4648 0.1027>, rot=<0 0 0.2806 -0.9598>}
 ring_03_r {parent=ring_02_r, loc=<-2.4153 -0.1141 -0.0199>, rot=<0 0 0.1189 -0.9929>}
 thumb_01_r {parent=hand_r, loc=<-1.3119 0.3192 2.459>, rot=<0.6546 -0.4042 0.0664 -0.6354>}
 thumb_02_r {parent=thumb_01_r, loc=<-5.9335 0.0447 -0.0401>, rot=<-0.0037 0.0213 0.298 -0.9543>}
 thumb_03_r {parent=thumb_02_r, loc=<-2.2223 -0.3682 0.0907>, rot=<0.0118 -0.013 0.0066 -0.9998>}
 weapon_r {parent=hand_r, loc=<0.1975 3.4112 -0.3811>, rot=<0 0 0 -1>}
 deform_wrist_dn_base_r {parent=hand_r, loc=<0 1.5329 0>, rot=<0 0 0 -1>}
 deform_wrist_dn_r {parent=deform_wrist_dn_base_r, loc=<0 0 0>, rot=<0 0 0 1>}
 deform_wrist_up_base_r {parent=hand_r, loc=<-0.0001 -2.3006 -0.0001>, rot=<0 0 0 -1>}
 deform_wrist_up_r {parent=deform_wrist_up_base_r, loc=<0 0 0>, rot=<0 0 0 1>}
 lowerarm_twist_01_r {parent=lowerarm_r, loc=<-18.3161 0 0>, rot=<0 0 0 1>}
 lowerarm_twist_02_r {parent=lowerarm_r, loc=<-10.6707 0 0>, rot=<0 0 0 1>}
 dyn_wrist_strap_r_MSTR {parent=lowerarm_twist_02_r, loc=<-2.6825 4.0099 -0.9183>, rot=<-0.4756 0.4929 -0.5226 -0.5077>}
 dyn_wrist_strap_r {parent=dyn_wrist_strap_r_MSTR, loc=<0 0 0>, rot=<-0.033 0.1671 -0.3736 -0.9118>}
 deform_elbow_in_base_r {parent=lowerarm_r, loc=<-0.1834 -3.6062 0.2572>, rot=<0 0 0 -1>}
 deform_elbow_in_r {parent=deform_elbow_in_base_r, loc=<0 0 0>, rot=<0 0 0 1>}
 deform_elbow_out_base_r {parent=lowerarm_r, loc=<0 4 0>, rot=<0 0 0 1>}
 deform_elbow_out_r {parent=deform_elbow_out_base_r, loc=<0 0 0>, rot=<0 0 0 1>}
 upperarm_twist_01_r {parent=upperarm_r, loc=<-7.2135 0.1134 0>, rot=<0 0 0 1>}
 dyn_cuff_base_r {parent=upperarm_twist_01_r, loc=<3.544 0 0.0001>, rot=<0 -1 0 0>}
 dyn_cuff_r_03 {parent=dyn_cuff_base_r, loc=<0.1767 5.6474 -0.0802>, rot=<0.5628 0.0674 0.0607 -0.8216>}
 dyn_cuff_r_01 {parent=dyn_cuff_base_r, loc=<0.1865 -1.1579 5.1526>, rot=<0.0074 0.045 -0.0016 -0.999>}
 dyn_cuff_r_05_MSTR {parent=dyn_cuff_base_r, loc=<3.3314 0.1184 -0.01>, rot=<0.9983 0.0003 0.045 0.0364>}
 dyn_cuff_r_02_MSTR {parent=dyn_cuff_base_r, loc=<1.0112 -6.7002 -0.9549>, rot=<-0.5322 0.0497 -0.0592 -0.8431>}
 dyn_cuff_r_02 {parent=dyn_cuff_r_02_MSTR, loc=<0 0 0>, rot=<0 0 0 1>}
 dyn_cuff_r_04_MSTR {parent=dyn_cuff_base_r, loc=<0.8415 0.0297 -5.494>, rot=<0.9983 0.0003 0.045 0.0364>}
 dyn_cuff_r_04 {parent=dyn_cuff_r_04_MSTR, loc=<0 0 0>, rot=<0 0 0 1>}
 deform_upperarm_bk_base_r {parent=upperarm_twist_01_r, loc=<6.4052 4.3746 -0.3139>, rot=<0 0 0 -1>}
 deform_upperarm_bk_r {parent=deform_upperarm_bk_base_r, loc=<0 0 0>, rot=<0 0 0 1>}
 deform_upperarm_fr_base_r {parent=upperarm_twist_01_r, loc=<5.6311 -5.092 -0.3353>, rot=<0 0 0 -1>}
 deform_upperarm_fr_r {parent=deform_upperarm_fr_base_r, loc=<0 0 0>, rot=<0 0 0 1>}
 deform_upperarm_up_base_r {parent=upperarm_twist_01_r, loc=<8.8909 -0.4825 -3.9516>, rot=<0 0 0 -1>}
 deform_upperarm_up_r {parent=deform_upperarm_up_base_r, loc=<0 0 0>, rot=<0 0 0 1>}
 upperarm_twist_02_r {parent=upperarm_r, loc=<-19.7998 0.1134 0>, rot=<0 0 0 1>}
 neck_01 {parent=spine_05, loc=<11.3138 0.1529 0>, rot=<0 0 -0.2003 -0.9797>}
 neck_02 {parent=neck_01, loc=<4.5627 0 0>, rot=<0 0 0.0504 -0.9987>}
 head {parent=neck_02, loc=<5.3914 0 0>, rot=<0 0 0.115 -0.9934>}
 dyn_hair_bk_low_00 {parent=head, loc=<9.1859 -6.2057 -3.9489>, rot=<0 0 1 -0.0041>}
 dyn_hair_bk_low_02 {parent=head, loc=<5.9255 -6.8282 -0.5239>, rot=<0 0 1 -0.0041>}
 dyn_hair_bk_low_08 {parent=head, loc=<5.4134 -5.8889 -4.4008>, rot=<0 0 1 -0.0041>}
 dyn_hair_bk_low_06 {parent=head, loc=<3.9203 -2.5864 -6.6868>, rot=<0 0 1 -0.0041>}
 dyn_hair_bk_low_10 {parent=head, loc=<5.3052 -4.6086 7.0684>, rot=<0 0 1 -0.0041>}
 dyn_hair_bk_low_04 {parent=head, loc=<5.6912 -7.9148 2.8771>, rot=<0 0 1 -0.0041>}
 dyn_hair_r_11 {parent=head, loc=<11.1326 -0.3286 5.8846>, rot=<0.1102 0.0205 0.977 -0.1816>}
 dyn_hair_r_12 {parent=dyn_hair_r_11, loc=<3.6652 0 0>, rot=<0.0606 0.1038 0.1462 -0.9819>}
 dyn_hair_r_13 {parent=dyn_hair_r_12, loc=<3.3844 0 0>, rot=<-0.005 -0.2176 0.0226 -0.9758>}
 dyn_hair_r_15 {parent=dyn_hair_r_11, loc=<2.6721 1.4001 0.6071>, rot=<0.02 -0.1103 0.1775 -0.9777>}
 dyn_hair_bangs_l_00 {parent=head, loc=<8.3014 9.4657 -1.6>, rot=<0.1135 0.6979 0.699 0.1066>}
 dyn_hair_bangs_l_01 {parent=dyn_hair_bangs_l_00, loc=<2.4383 0.0007 -0.2258>, rot=<-0.0008 -0.1597 0.0049 -0.9871>}
 dyn_hair_bangs_r_00 {parent=head, loc=<9.3286 10.212 -0.2505>, rot=<0.2084 0.6757 0.6775 0.2025>}
 dyn_hair_bangs_r_01 {parent=dyn_hair_bangs_r_00, loc=<2.5194 0.0032 -0.5376>, rot=<-0.0013 -0.2945 0.0042 -0.9557>}
 dyn_hair_l_00 {parent=head, loc=<9.1965 3.0085 -7.0485>, rot=<0.064 -0.9892 -0.0085 -0.1319>}
 dyn_hair_l_01 {parent=dyn_hair_l_00, loc=<4.1293 0.5163 -1.1186>, rot=<-0.0085 -0.1279 -0.0615 -0.9898>}
 dyn_hair_l_02 {parent=dyn_hair_l_01, loc=<4.4679 0 0>, rot=<0 -0.004 0.0017 -1>}
 dyn_hair_l_04 {parent=dyn_hair_l_01, loc=<1.6592 -1.3223 0.7971>, rot=<0 -0.004 0.0017 -1>}
 dyn_hair_bk_r_01 {parent=head, loc=<11.7603 -3.6043 3.7687>, rot=<0.2815 -0.6487 0.6991 -0.106>}
 dyn_hair_bk_r_02 {parent=dyn_hair_bk_r_01, loc=<3.2684 0 0>, rot=<-0.1399 0.2295 0.274 -0.9234>}
 dyn_hair_bk_r_03 {parent=dyn_hair_bk_r_02, loc=<1.5007 0 0>, rot=<-0.1612 0.2211 -0.1964 -0.9416>}
 dyn_hair_fr_00 {parent=head, loc=<9.9234 9.4226 3.9997>, rot=<0.073 0.7033 0.7064 0.0316>}
 dyn_hair_fr_01 {parent=dyn_hair_fr_00, loc=<2.8789 -0.2826 -0.3663>, rot=<-0.0039 -0.0869 0.0136 -0.9961>}
 dyn_hair_fr_02 {parent=dyn_hair_fr_01, loc=<3.3747 0 0>, rot=<0.0001 0.0094 0.0204 -0.9997>}
 dyn_hair_fr_03 {parent=dyn_hair_fr_02, loc=<3.0725 0 0>, rot=<0 -0.0118 -0.0103 -0.9999>}
 dyn_hair_fr_05 {parent=dyn_hair_fr_02, loc=<2.8983 2.8823 -1.5382>, rot=<0 -0.0005 -0.0046 -1>}
 dyn_hair_r_01 {parent=head, loc=<10.1561 4.1956 5.9325>, rot=<0.1185 0.007 0.9912 -0.0587>}
 dyn_hair_r_02 {parent=dyn_hair_r_01, loc=<3.463 0 0>, rot=<0.0102 -0.0256 -0.0012 -0.9996>}
 dyn_hair_r_03 {parent=dyn_hair_r_02, loc=<4.0527 0 0>, rot=<-0.0054 -0.0982 0.0548 -0.9936>}
 dyn_hair_r_05 {parent=dyn_hair_r_02, loc=<2.074 -1.5982 1.2444>, rot=<-0.0052 -0.0936 0.0548 -0.9941>}
 dyn_hair_r_07 {parent=dyn_hair_r_01, loc=<3.3208 0.442 1.8066>, rot=<0.0065 -0.1185 0.0546 -0.9914>}
 dyn_hair_fr_l_00 {parent=head, loc=<9.7588 6.4683 -5.7916>, rot=<0.0421 -0.8508 -0.4807 -0.208>}
 dyn_hair_fr_l_01 {parent=dyn_hair_fr_l_00, loc=<3.0234 0.838 -1.0576>, rot=<0.2643 -0.1156 -0.1595 -0.9441>}
 dyn_hair_fr_l_02 {parent=dyn_hair_fr_l_01, loc=<4.3719 0 0>, rot=<0 -0.001 -0.0005 -1>}
 dyn_hair_bk_00 {parent=head, loc=<11.8659 -4.149 0>, rot=<0.2184 -0.6725 0.6722 -0.2194>}
 dyn_hair_bk_01 {parent=dyn_hair_bk_00, loc=<4.1035 0 0>, rot=<-0.6848 0.1608 -0.1666 -0.691>}
 dyn_hair_bk_02 {parent=dyn_hair_bk_01, loc=<1.7142 0 0>, rot=<0.6031 -0.365 0.3672 -0.6068>}
 dyn_hair_bk_l_00 {parent=head, loc=<11.1215 -1.9077 -4.538>, rot=<0.2018 -0.9614 -0.0384 -0.183>}
 dyn_hair_bk_l_01 {parent=dyn_hair_bk_l_00, loc=<2.4711 -0.5376 0.4775>, rot=<0.183 0.1741 0.2903 -0.923>}
 dyn_hair_bk_l_02 {parent=dyn_hair_bk_l_01, loc=<2.6783 1.7034 -0.5345>, rot=<-0.3794 -0.4322 -0.4 -0.7136>}
 dyn_hood_bk {parent=head, loc=<8.7774 -14.2838 0.9518>, rot=<0.5635 -0.4272 0.4272 -0.5635>}
 dyn_hood_ft {parent=head, loc=<6.3341 12.6025 0.1431>, rot=<0.0029 -0.7071 -0.0029 -0.7071>}
 FACIAL_C_FacialRoot {parent=head, loc=<2.1524 -0.746 -0.0002>, rot=<0.4879 -0.5118 -0.4879 -0.5118>}
 FACIAL_C_Forehead {parent=FACIAL_C_FacialRoot, loc=<-0.0004 -2.8701 8.8395>, rot=<-0.0451 0 0 -0.999>}
 FACIAL_L_ForeheadIn {parent=FACIAL_C_FacialRoot, loc=<1.6826 -2.8695 8.6757>, rot=<-0.0624 -0.0696 0.0042 -0.9956>}
 FACIAL_R_ForeheadIn {parent=FACIAL_C_FacialRoot, loc=<-1.6831 -2.8707 8.6738>, rot=<-0.0624 0.0696 -0.0041 -0.9956>}
 FACIAL_L_ForeheadMid {parent=FACIAL_C_FacialRoot, loc=<3.12 -3.1924 8.1399>, rot=<-0.0726 -0.1756 -0.0209 -0.9816>}
 FACIAL_R_ForeheadMid {parent=FACIAL_C_FacialRoot, loc=<-3.1204 -3.1931 8.1379>, rot=<-0.0726 0.1756 0.0209 -0.9816>}
 FACIAL_L_ForeheadOut {parent=FACIAL_C_FacialRoot, loc=<4.3844 -3.1967 7.2008>, rot=<-0.123 -0.2555 -0.0796 -0.9557>}
 FACIAL_R_ForeheadOut {parent=FACIAL_C_FacialRoot, loc=<-4.3843 -3.1965 7.2008>, rot=<-0.123 0.2555 0.0796 -0.9557>}
 FACIAL_L_EyesackUpper {parent=FACIAL_C_FacialRoot, loc=<3.2076 -2.1775 8.6355>, rot=<0.0472 -0.0841 -0.0646 -0.9932>}
 FACIAL_R_EyesackUpper {parent=FACIAL_C_FacialRoot, loc=<-3.2071 -2.1774 8.6364>, rot=<0.0472 0.0841 0.0647 -0.9932>}
 FACIAL_L_EyelidUpperFurrow {parent=FACIAL_C_FacialRoot, loc=<3.0625 -1.9275 8.5569>, rot=<-0.0317 -0.1507 -0.0799 -0.9848>}
 FACIAL_R_EyelidUpperFurrow {parent=FACIAL_C_FacialRoot, loc=<-3.0589 -1.9269 8.5566>, rot=<-0.0317 0.1507 0.0799 -0.9848>}
 FACIAL_L_EyelidUpperA {parent=FACIAL_C_FacialRoot, loc=<2.9159 -0.9895 6.8192>, rot=<-0.1495 0 0 -0.9888>}
 FACIAL_R_EyelidUpperA {parent=FACIAL_C_FacialRoot, loc=<-2.9159 -0.9889 6.8207>, rot=<-0.1495 0 0 -0.9888>}
 FACIAL_L_Eye {parent=FACIAL_C_FacialRoot, loc=<2.9078 -0.9985 6.8117>, rot=<-0.0017 0 0 -1>}
 FACIAL_R_Eye {parent=FACIAL_C_FacialRoot, loc=<-2.9097 -0.9976 6.8131>, rot=<-0.0017 0 0 -1>}
 FACIAL_L_EyelidLowerA {parent=FACIAL_C_FacialRoot, loc=<2.9159 -0.9895 6.8192>, rot=<0.2824 0 0 -0.9593>}
 FACIAL_R_EyelidLowerA {parent=FACIAL_C_FacialRoot, loc=<-2.9159 -0.9889 6.8207>, rot=<0.2824 0 0 -0.9593>}
 FACIAL_L_EyeCornerInner {parent=FACIAL_C_FacialRoot, loc=<1.6578 -0.3091 8.0611>, rot=<-0.0017 0 0 -1>}
 FACIAL_R_EyeCornerInner {parent=FACIAL_C_FacialRoot, loc=<-1.6471 -0.3107 8.0589>, rot=<-0.0017 0 0 -1>}
 FACIAL_L_EyeCornerOuter {parent=FACIAL_C_FacialRoot, loc=<4.5586 -0.8124 7.3728>, rot=<-0.0016 -0.3826 -0.0006 -0.9239>}
 FACIAL_R_EyeCornerOuter {parent=FACIAL_C_FacialRoot, loc=<-4.5612 -0.8167 7.3773>, rot=<-0.0016 0.3826 0.0007 -0.9239>}
 FACIAL_L_EyesackLower {parent=FACIAL_C_FacialRoot, loc=<3.4068 0.5948 7.3073>, rot=<-0.0239 -0.1397 -0.0225 -0.9897>}
 FACIAL_R_EyesackLower {parent=FACIAL_C_FacialRoot, loc=<-3.3997 0.591 7.2935>, rot=<-0.0239 0.1397 0.0225 -0.9897>}
 FACIAL_L_CheekInner {parent=FACIAL_C_FacialRoot, loc=<2.1127 1.3498 7.7459>, rot=<-0.1248 -0.0877 -0.1972 -0.9684>}
 FACIAL_R_CheekInner {parent=FACIAL_C_FacialRoot, loc=<-2.1075 1.3442 7.7489>, rot=<-0.1248 0.0877 0.1973 -0.9684>}
 FACIAL_L_CheekOuter {parent=FACIAL_C_FacialRoot, loc=<3.8394 1.3831 6.0533>, rot=<0.0537 -0.3771 0.0997 -0.9192>}
 FACIAL_R_CheekOuter {parent=FACIAL_C_FacialRoot, loc=<-3.8376 1.3837 6.0537>, rot=<0.0536 0.3771 -0.0997 -0.9192>}
 FACIAL_L_NasolabialBulge {parent=FACIAL_C_FacialRoot, loc=<3.0902 4.6034 6.0008>, rot=<0.0564 -0.3774 0.0232 -0.9241>}
 FACIAL_R_NasolabialBulge {parent=FACIAL_C_FacialRoot, loc=<-3.0906 4.6011 5.9983>, rot=<0.0564 0.3774 -0.0232 -0.9241>}
 FACIAL_L_NasolabialFurrow {parent=FACIAL_C_FacialRoot, loc=<2.698 3.3331 8.328>, rot=<-0.0348 -0.1252 -0.3638 -0.9224>}
 FACIAL_R_NasolabialFurrow {parent=FACIAL_C_FacialRoot, loc=<-2.6997 3.3302 8.327>, rot=<-0.0348 0.1252 0.3639 -0.9223>}
 FACIAL_L_CheekLower {parent=FACIAL_C_FacialRoot, loc=<3.5652 4.4176 5.389>, rot=<0.0331 -0.427 0.0571 -0.9019>}
 FACIAL_R_CheekLower {parent=FACIAL_C_FacialRoot, loc=<-3.5653 4.4142 5.3866>, rot=<0.0331 0.427 -0.057 -0.9019>}
 FACIAL_C_Nose {parent=FACIAL_C_FacialRoot, loc=<-0.0002 0.755 9.2317>, rot=<-0.132 0 0 -0.9913>}
 FACIAL_C_NoseLower {parent=FACIAL_C_Nose, loc=<-0.0002 2.8844 -0.4547>, rot=<0.1305 0 0 -0.9914>}
 FACIAL_L_Nostril {parent=FACIAL_C_Nose, loc=<0.8246 2.2148 -0.3466>, rot=<0.0173 0.1305 0.0023 -0.9913>}
 FACIAL_R_Nostril {parent=FACIAL_C_Nose, loc=<-0.8227 2.215 -0.3502>, rot=<0.0173 -0.1305 -0.0023 -0.9913>}
 FACIAL_C_NoseTip {parent=FACIAL_C_Nose, loc=<-0.0007 2.1799 0.434>, rot=<0.0872 0 0 -0.9962>}
 FACIAL_C_MouthUpper {parent=FACIAL_C_FacialRoot, loc=<0.0003 4.769 7.003>, rot=<-0.0016 0 0 -1>}
 FACIAL_C_LipUpper {parent=FACIAL_C_MouthUpper, loc=<-0.0026 -0.1719 2.1255>, rot=<-0.0872 0 0 -0.9962>}
 FACIAL_L_LipUpper {parent=FACIAL_C_MouthUpper, loc=<0.9905 -0.1229 1.9244>, rot=<-0.0778 -0.1601 -0.0379 -0.9833>}
 FACIAL_R_LipUpper {parent=FACIAL_C_MouthUpper, loc=<-0.9956 -0.124 1.9223>, rot=<-0.0778 0.1601 0.0379 -0.9833>}
 FACIAL_L_LipUpperOuter {parent=FACIAL_C_MouthUpper, loc=<1.7424 0.0082 1.379>, rot=<-0.0984 -0.2293 -0.0983 -0.9634>}
 FACIAL_R_LipUpperOuter {parent=FACIAL_C_MouthUpper, loc=<-1.7468 0.0082 1.3747>, rot=<-0.0984 0.2293 0.0983 -0.9634>}
 FACIAL_L_LipCorner {parent=FACIAL_C_MouthUpper, loc=<2.2285 0.3949 0.6838>, rot=<0 -0.2078 0 -0.9782>}
 FACIAL_R_LipCorner {parent=FACIAL_C_MouthUpper, loc=<-2.2305 0.3967 0.6794>, rot=<0 0.2078 0 -0.9782>}
 FACIAL_L_Masseter {parent=FACIAL_C_FacialRoot, loc=<5.8328 1.8444 3.8429>, rot=<0.0677 0.1129 0.0081 -0.9913>}
 FACIAL_R_Masseter {parent=FACIAL_C_FacialRoot, loc=<-5.8335 1.8448 3.8414>, rot=<0.0677 -0.1129 -0.0081 -0.9913>}
 FACIAL_C_UnderChin {parent=FACIAL_C_FacialRoot, loc=<0.0003 8.6805 3.8713>, rot=<0.4984 0 0 -0.8669>}
 FACIAL_L_UnderChin {parent=FACIAL_C_FacialRoot, loc=<2.66 7.567 2.7881>, rot=<0.4011 -0.355 0.0817 -0.8405>}
 FACIAL_R_UnderChin {parent=FACIAL_C_FacialRoot, loc=<-2.6607 7.5677 2.786>, rot=<0.4011 0.355 -0.0817 -0.8405>}
 FACIAL_C_TeethUpper {parent=FACIAL_C_FacialRoot, loc=<-0.0005 5.2307 8.052>, rot=<-0.0016 0 0 -1>}
 FACIAL_C_LowerLipRotation {parent=FACIAL_C_FacialRoot, loc=<-0.0012 2.45 2.0912>, rot=<0.3405 0 0 -0.9403>}
 FACIAL_C_MouthLower {parent=FACIAL_C_LowerLipRotation, loc=<0.0015 -1.3637 5.2577>, rot=<-0.342 0 0 -0.9397>}
 FACIAL_C_LipLower {parent=FACIAL_C_MouthLower, loc=<-0.0004 1.192 1.8441>, rot=<0.0872 0 0 -0.9962>}
 FACIAL_L_LipLower {parent=FACIAL_C_MouthLower, loc=<0.8612 1.1505 1.6285>, rot=<0.0927 -0.1294 0.0272 -0.9869>}
 FACIAL_R_LipLower {parent=FACIAL_C_MouthLower, loc=<-0.8629 1.1509 1.6276>, rot=<0.0927 0.1294 -0.0272 -0.9869>}
 FACIAL_L_LipLowerOuter {parent=FACIAL_C_MouthLower, loc=<1.6385 0.7882 1.125>, rot=<0.0913 -0.2321 0.0318 -0.9679>}
 FACIAL_R_LipLowerOuter {parent=FACIAL_C_MouthLower, loc=<-1.6415 0.7881 1.1239>, rot=<0.0913 0.2321 -0.0318 -0.9679>}
 FACIAL_C_TeethLower {parent=FACIAL_C_LowerLipRotation, loc=<0.0007 -1.6101 6.307>, rot=<-0.342 0 0 -0.9397>}
 FACIAL_C_Tongue1 {parent=FACIAL_C_TeethLower, loc=<0.0018 0.1189 -4.8682>, rot=<-0.2588 0 0 -0.9659>}
 FACIAL_C_Tongue2 {parent=FACIAL_C_Tongue1, loc=<-0.0006 0.3979 1.3735>, rot=<0.2588 0 0 -0.9659>}
 FACIAL_C_Tongue3 {parent=FACIAL_C_Tongue2, loc=<-0.0002 0.1865 1.2865>, rot=<0.1305 0 0 -0.9914>}
 FACIAL_C_Tongue4 {parent=FACIAL_C_Tongue3, loc=<0.0003 -0.1865 1.1174>, rot=<0.0436 0 0 -0.999>}
 FACIAL_C_TongueUpper3 {parent=FACIAL_C_Tongue3, loc=<0.0004 -0.329 -0.1771>, rot=<0 0 0 -1>}
 FACIAL_C_TongueLower3 {parent=FACIAL_C_Tongue3, loc=<-0.0001 0.7984 0.0501>, rot=<0 0 0 -1>}
 FACIAL_L_TongueSide3 {parent=FACIAL_C_Tongue3, loc=<1.4948 0.015 0.1147>, rot=<0 0 0 -1>}
 FACIAL_R_TongueSide3 {parent=FACIAL_C_Tongue3, loc=<-1.4954 0.0151 0.112>, rot=<0 0 0 -1>}
 FACIAL_C_TongueUpper2 {parent=FACIAL_C_Tongue2, loc=<-0.0001 -0.5474 -0.1596>, rot=<0 0 0 1>}
 FACIAL_L_TongueSide2 {parent=FACIAL_C_Tongue2, loc=<1.7743 0.0475 0.1975>, rot=<0 0 0 1>}
 FACIAL_R_TongueSide2 {parent=FACIAL_C_Tongue2, loc=<-1.775 0.0466 0.1953>, rot=<0 0 0 1>}
 FACIAL_C_Jaw {parent=FACIAL_C_FacialRoot, loc=<-0.0012 2.45 2.0912>, rot=<0.3405 0 0 -0.9403>}
 FACIAL_C_Jawline {parent=FACIAL_C_Jaw, loc=<0.0007 2.279 7.2327>, rot=<0.2586 0 0 -0.966>}
 FACIAL_L_ChinSide {parent=FACIAL_C_Jaw, loc=<2.5173 0.3862 6.7876>, rot=<-0.2208 -0.3738 0.0499 -0.8995>}
 FACIAL_R_ChinSide {parent=FACIAL_C_Jaw, loc=<-2.5142 0.3856 6.7883>, rot=<-0.2208 0.3738 -0.0499 -0.8995>}
 FACIAL_C_Chin {parent=FACIAL_C_Jaw, loc=<0.0012 0.8738 7.4354>, rot=<-0.2591 0 0 -0.9659>}
 FACIAL_L_Ear {parent=FACIAL_C_FacialRoot, loc=<6.1353 0.0215 0.6714>, rot=<-0.028 -0.0006 -0.0261 -0.9993>}
 FACIAL_R_Ear {parent=FACIAL_C_FacialRoot, loc=<-6.1346 0.0225 0.6681>, rot=<-0.028 0.0006 0.0262 -0.9993>}
 FACIAL_L_Jawline {parent=FACIAL_C_FacialRoot, loc=<3.9565 5.7054 4.2488>, rot=<0.2077 0.1578 -0.0544 -0.9638>}
 FACIAL_R_Jawline {parent=FACIAL_C_FacialRoot, loc=<-3.9589 5.7063 4.2476>, rot=<0.2077 -0.1578 0.0544 -0.9638>}
 FACIAL_C_NoseBridge {parent=FACIAL_C_FacialRoot, loc=<-0.0003 -1.3317 9.5475>, rot=<0.0422 0 0 -0.9991>}
 FACIAL_L_EyelidUpperB {parent=FACIAL_C_FacialRoot, loc=<2.9159 -0.9895 6.8192>, rot=<-0.2266 0 0 -0.974>}
 FACIAL_L_EyelidLowerB {parent=FACIAL_C_FacialRoot, loc=<2.9159 -0.9895 6.8192>, rot=<0.3157 0 0 -0.9489>}
 FACIAL_R_EyelidLowerB {parent=FACIAL_C_FacialRoot, loc=<-2.9159 -0.9889 6.8207>, rot=<0.324 0 0 -0.9461>}
 FACIAL_R_EyelidUpperB {parent=FACIAL_C_FacialRoot, loc=<-2.9159 -0.9889 6.8207>, rot=<-0.2266 0 0 -0.974>}
 neck_01_LOD1 {parent=neck_01, loc=<0 0 0>, rot=<0 0 0 1>}
 dyn_clip_back_MSTR {parent=spine_05, loc=<5.8125 -7.5807 0.0002>, rot=<0.0342 -0.7099 0.7032 -0.0224>}
 dyn_clip_back_01 {parent=dyn_clip_back_MSTR, loc=<0 0 0>, rot=<-0.0069 0.2442 -0.0067 -0.9697>}
 dyn_clip_back_02 {parent=dyn_clip_back_01, loc=<1.5295 0 0>, rot=<0 -0.1899 0 -0.9818>}
 dyn_collar_00 {parent=spine_05, loc=<15.4661 -6.1347 -0.0095>, rot=<0.0284 -0.7065 -0.0284 -0.7065>}
 dyn_collar_01 {parent=dyn_collar_00, loc=<-0.0095 0.2822 5.5804>, rot=<0 0 0 1>}
 dyn_lapel_r {parent=spine_05, loc=<-1.2627 11.9481 6.7553>, rot=<0.0364 -0.7062 -0.0203 -0.7068>}
 dyn_lapel_r1 {parent=dyn_lapel_r, loc=<-4.255 -0.9313 3.7609>, rot=<0 0 0 1>}
 dyn_lapel_l {parent=spine_05, loc=<-1.2627 11.9481 -6.6888>, rot=<0.0364 -0.7062 -0.0203 -0.7068>}
 dyn_lapel_l1 {parent=dyn_lapel_l, loc=<3.7068 -0.4033 3.8783>, rot=<0 0 0 1>}
 attach_cape {parent=spine_05, loc=<9.3216 -8.4501 0>, rot=<0.0616 -0.7044 -0.0616 -0.7044>}
 attach_backpack {parent=spine_05, loc=<0 -8.35 0>, rot=<-0.0093 -0.707 0.0093 -0.707>}
 base_chain {parent=spine_05, loc=<6.4927 7.8771 0.0251>, rot=<-0.0197 -0.7108 -0.7016 -0.0461>}
 dyn_chain_01 {parent=base_chain, loc=<0 0 0>, rot=<0 0 0 1>}
 dyn_chain_02 {parent=dyn_chain_01, loc=<1.2958 0 0>, rot=<0 0.1366 0 -0.9906>}
 dyn_chain_03 {parent=dyn_chain_02, loc=<4.3903 0.1747 -0.0124>, rot=<-0.0028 0.1303 -0.0197 -0.9913>}
 dyn_chain_04 {parent=dyn_chain_03, loc=<3.5882 0 0>, rot=<0 0 0 1>}
 base_collar_r {parent=spine_05, loc=<13.9062 8.3178 8.9637>, rot=<0.7016 0.0883 0.7064 0.0318>}
 dyn_collar_01_r {parent=base_collar_r, loc=<0 0 0>, rot=<0 0 -0.0221 -0.9998>}
 dyn_Collar_02_r {parent=dyn_collar_01_r, loc=<3.7213 0 0>, rot=<0 0 0 1>}
 base_collar_l {parent=spine_05, loc=<13.906 8.3177 -8.9637>, rot=<0.0284 -0.7065 -0.0284 -0.7065>}
 dyn_collar_01_l {parent=base_collar_l, loc=<0 0 0>, rot=<0 0 0 1>}
 dyn_Collar_02_l {parent=dyn_collar_01_l, loc=<3.7212 0 0>, rot=<0 0 0 1>}
 deform_pec_aim_base_l {parent=spine_04, loc=<3.2456 9.3309 -1.2237>, rot=<0.0603 -0.4033 0.3318 -0.8507>}
 deform_pec_aim_l {parent=deform_pec_aim_base_l, loc=<0 0 0>, rot=<0 0 0 1>}
 deform_pec_base_l {parent=deform_pec_aim_l, loc=<9.3665 0 0>, rot=<0 0 0 1>}
 deform_pec_l {parent=deform_pec_base_l, loc=<0 0 0>, rot=<0 0 0 1>}
 deform_pec_aim_base_r {parent=spine_04, loc=<3.2452 9.3309 1.2237>, rot=<-0.4033 -0.0603 0.8507 0.3318>}
 deform_pec_aim_r {parent=deform_pec_aim_base_r, loc=<0 0 0>, rot=<0 0 0 1>}
 deform_pec_base_r {parent=deform_pec_aim_r, loc=<-9.3667 -0.0003 -0.0004>, rot=<0 0 0 1>}
 deform_pec_r {parent=deform_pec_base_r, loc=<0 0 0>, rot=<0 0 0 1>}
 dyn_lapel_lower_r {parent=spine_04, loc=<-0.4062 12.3691 5.8508>, rot=<0.0294 -0.7065 -0.0294 -0.7065>}
 dyn_lapel_lower_r1 {parent=dyn_lapel_lower_r, loc=<-3.4456 0.6264 -0.3913>, rot=<0 0 0 1>}
 dyn_clip_front_MSTR {parent=spine_03, loc=<-2.0058 10.4414 -0.0337>, rot=<0.0166 -0.7026 -0.0459 -0.7099>}
 dyn_clip_front_01 {parent=dyn_clip_front_MSTR, loc=<0 0 0>, rot=<0 0 0 1>}
 dyn_clip_front_02 {parent=dyn_clip_front_01, loc=<-0.017 -0.0273 -1.3106>, rot=<0 0 0 1>}
 dyn_jacketFront_r {parent=spine_03, loc=<4.1793 10.7098 7.1872>, rot=<0.0426 -0.4943 0.8578 -0.1341>}
 dyn_jacketBack {parent=spine_03, loc=<7.5243 -8.3606 0.0002>, rot=<-0.0425 -0.9991 0 0>}
 dyn_jacketSideBack_r {parent=spine_03, loc=<4.0391 -5.4457 7.9666>, rot=<0.0254 -0.9233 0.3832 -0.0014>}
 dyn_jacketSide_r {parent=spine_03, loc=<2.1843 3.4272 10.215>, rot=<0.0377 -0.6728 0.7333 -0.0906>}
 dyn_jacketSide_l {parent=spine_03, loc=<2.1847 3.4272 -10.0089>, rot=<0.0906 -0.7333 0.6728 -0.0377>}
 dyn_jacketFront_l {parent=spine_03, loc=<4.1795 10.7098 -7.4337>, rot=<0.117 -0.8588 0.4986 -0.0131>}
 dyn_jacketSideBack_l {parent=spine_03, loc=<4.0836 -5.8541 -8.5798>, rot=<0.0014 -0.3832 0.9233 -0.0254>}
 thigh_l {parent=pelvis, loc=<-1.0855 -0.2109 -8.7053>, rot=<0.0009 0.0355 0.0155 -0.9993>}
 calf_l {parent=thigh_l, loc=<-42.5284 -0.0019 -0.7297>, rot=<0.0006 -0.0161 -0.0391 -0.9991>}
 calf_twist_01_l {parent=calf_l, loc=<-35.8483 -1.9197 -0.5794>, rot=<0.0237 0.0095 -0.0035 -0.9997>}
 calf_twist_02_l {parent=calf_l, loc=<-19.1546 -1.7943 -0.2638>, rot=<0.0237 0.0095 -0.0035 -0.9997>}
 foot_l {parent=calf_l, loc=<-46.7402 -2.5014 -0.592>, rot=<0.0427 -0.0238 0.0353 -0.9982>}
 ball_l {parent=foot_l, loc=<-3.5658 10.8066 -0.1544>, rot=<0.0001 -0.0001 -0.7063 -0.7079>}
 dyn_laces_l {parent=foot_l, loc=<3.6151 -6.3888 0.2901>, rot=<-0.0357 -0.7045 -0.0284 -0.7082>}
 dyn_shoe_strap_l {parent=foot_l, loc=<0.8119 6.5819 -1.845>, rot=<-0.0341 0.708 0.0414 -0.7042>}
 deform_knee_in_base_l {parent=calf_l, loc=<0.8031 -5.0231 0.2456>, rot=<0 0 0 -1>}
 deform_knee_in_l {parent=deform_knee_in_base_l, loc=<0 0 0>, rot=<0 0 0 1>}
 deform_knee_out_base_l {parent=calf_l, loc=<0 4 0>, rot=<0 0 0 1>}
 deform_knee_out_l {parent=deform_knee_out_base_l, loc=<0 0 0>, rot=<0 0 0 1>}
 thigh_twist_01_l {parent=thigh_l, loc=<-27.952 0 0>, rot=<0.0013 0 0 -1>}
 deform_groin_base_l {parent=thigh_l, loc=<0.4801 5.6776 -0.2486>, rot=<0 0 0 -1>}
 deform_groin_l {parent=deform_groin_base_l, loc=<0 0 0>, rot=<0 0 0 1>}
 thigh_r {parent=pelvis, loc=<-1.0853 -0.2109 8.7053>, rot=<0.0355 -0.0009 0.9993 0.0155>}
 calf_r {parent=thigh_r, loc=<42.5286 0.0019 0.7297>, rot=<0.0006 -0.0161 -0.0391 -0.9991>}
 calf_twist_01_r {parent=calf_r, loc=<35.8485 1.9197 0.5794>, rot=<0.0237 0.0095 -0.0035 -0.9997>}
 calf_twist_02_r {parent=calf_r, loc=<19.1546 1.7943 0.2638>, rot=<0.0237 0.0095 -0.0035 -0.9997>}
 foot_r {parent=calf_r, loc=<46.7403 2.5014 0.592>, rot=<0.0427 -0.0238 0.0353 -0.9982>}
 ball_r {parent=foot_r, loc=<3.5658 -10.8066 0.1544>, rot=<0.0001 -0.0001 -0.7063 -0.7079>}
 dyn_laces_r {parent=foot_r, loc=<-3.6153 6.3888 -0.2901>, rot=<0.4979 -0.4997 0.5056 -0.4968>}
 dyn_shoe_strap_r {parent=foot_r, loc=<-0.8121 -6.5819 1.845>, rot=<0.7082 0.0284 -0.7045 -0.0357>}
 deform_knee_in_base_r {parent=calf_r, loc=<-0.8031 5.0231 -0.2457>, rot=<0 0 0 -1>}
 deform_knee_in_r {parent=deform_knee_in_base_r, loc=<0 0 0>, rot=<0 0 0 1>}
 deform_knee_out_base_r {parent=calf_r, loc=<0 -4 0>, rot=<0 0 0 1>}
 deform_knee_out_r {parent=deform_knee_out_base_r, loc=<0 0 0>, rot=<0 0 0 1>}
 thigh_twist_01_r {parent=thigh_r, loc=<27.952 0 0>, rot=<0.0013 0 0 -1>}
 dyn_skirt_fr_mid_r_01 {parent=thigh_twist_01_r, loc=<-23.2011 -7.1737 -1.138>, rot=<-0.708 0.0239 0.0035 -0.7058>}
 dyn_skirt_02_r {parent=thigh_twist_01_r, loc=<-22.7637 -4.2368 4.835>, rot=<-0.3704 0.0497 -0.0274 -0.9271>}
 dyn_skirt_side_02_r {parent=thigh_twist_01_r, loc=<-23.1531 2.3352 6.863>, rot=<-0.0001 0.0507 -0.0262 -0.9984>}
 dyn_skirt_fr_mid_r_02 {parent=thigh_twist_01_r, loc=<-14.4944 -7.5113 -0.8873>, rot=<-0.7079 -0.0015 0.0072 -0.7062>}
 dyn_skirt_fr_mid_r_03 {parent=dyn_skirt_fr_mid_r_02, loc=<8.5625 0 0>, rot=<0.0001 -0.0054 0.0259 -0.9996>}
 dyn_skirt_03_r {parent=thigh_twist_01_r, loc=<-14.4016 -4.1189 5.7821>, rot=<-0.3698 0.0335 -0.0358 -0.9278>}
 dyn_skirt_04_r {parent=dyn_skirt_03_r, loc=<8.5491 0 0>, rot=<-0.0002 -0.0671 0.0032 -0.9977>}
 dyn_skirt_side_03_r {parent=thigh_twist_01_r, loc=<-14.7157 2.78 7.7227>, rot=<0.0001 0.0341 -0.043 -0.9985>}
 dyn_skirt_side_04_r {parent=dyn_skirt_side_03_r, loc=<8.6269 0 0>, rot=<-0.0027 -0.066 0.0415 -0.997>}
 deform_groin_base_r {parent=thigh_r, loc=<-0.4799 -5.6776 0.2485>, rot=<0 0 0 -1>}
 deform_groin_r {parent=deform_groin_base_r, loc=<0 0 0>, rot=<0 0 0 1>}
 dyn_belt_bk_00 {parent=pelvis, loc=<1.8517 -11.1172 6.3087>, rot=<-0.3295 -0.5569 0.6697 -0.3644>}
 dyn_belt_bk_01 {parent=dyn_belt_bk_00, loc=<9.9638 0 0>, rot=<0.0587 -0.103 -0.0199 -0.9928>}
 dyn_belt_bk_02 {parent=dyn_belt_bk_01, loc=<5.3141 0 0>, rot=<0.1157 -0.2 0.0258 -0.9726>}
 dyn_belt_bk_03 {parent=dyn_belt_bk_02, loc=<6.1418 0 0>, rot=<0.0648 -0.1319 -0.0935 -0.9847>}
 dyn_belt_bk_04 {parent=dyn_belt_bk_03, loc=<6.2049 0 0>, rot=<0.0462 -0.1927 -0.1237 -0.9723>}
 dyn_belt_bk_05 {parent=dyn_belt_bk_04, loc=<3.0362 0 0>, rot=<0.0094 -0.1669 -0.0513 -0.9846>}
 dyn_belt_bk_06 {parent=dyn_belt_bk_05, loc=<4.063 0 0>, rot=<-0.0027 -0.2045 -0.0174 -0.9787>}
 base_side_piece {parent=dyn_belt_bk_04, loc=<5.2813 1.0392 1.9627>, rot=<0.1238 -0.0307 0.7801 -0.6126>}
 dyn_side_piece_01 {parent=base_side_piece, loc=<0 0 0>, rot=<0 0 0 1>}
 dyn_side_piece_02 {parent=dyn_side_piece_01, loc=<17.244 0 0>, rot=<0 0 0 1>}
 base_belt_bk_07 {parent=dyn_belt_bk_04, loc=<0 0 0>, rot=<0 0 0 1>}
 base_back_piece {parent=dyn_belt_bk_00, loc=<10.0955 -1.7593 1.389>, rot=<0.1247 0.0303 0.8885 -0.4406>}
 dyn_back_piece_01 {parent=base_back_piece, loc=<0 0 0>, rot=<0 0 0 1>}
 dyn_back_piece_02 {parent=dyn_back_piece_01, loc=<8.4076 0 0>, rot=<0 0 0 1>}
 dyn_belt_ft_MSTR {parent=pelvis, loc=<4.1902 -0.2376 14.0716>, rot=<-0.0127 -0.707 0.0127 -0.707>}
 dyn_belt_ft_base_00 {parent=dyn_belt_ft_MSTR, loc=<0 0 0>, rot=<0.6624 0.2475 -0.5253 -0.4733>}
 dyn_belt_ft_base_01 {parent=dyn_belt_ft_base_00, loc=<5.0776 0 0>, rot=<-0.1304 -0.2114 -0.0707 -0.9661>}
 dyn_belt_ft_00 {parent=dyn_belt_ft_base_01, loc=<5.9038 0 0>, rot=<-0.1118 -0.1681 0.038 -0.9787>}
 dyn_belt_ft_01 {parent=dyn_belt_ft_00, loc=<6.8104 0 0>, rot=<-0.0402 -0.072 0.0389 -0.9958>}
 dyn_belt_ft_02 {parent=dyn_belt_ft_01, loc=<7.5837 0 0>, rot=<-0.0862 -0.2022 0.0652 -0.9734>}
 dyn_belt_ft_03 {parent=dyn_belt_ft_02, loc=<6.5359 0 0>, rot=<-0.0098 -0.0496 0.1414 -0.9887>}
 dyn_belt_ft_04 {parent=dyn_belt_ft_03, loc=<5.8497 0 0>, rot=<-0.0161 -0.3764 0.0087 -0.9263>}
 dyn_belt_ft_05 {parent=dyn_belt_ft_04, loc=<3.6994 0 0>, rot=<-0.002 -0.1813 0.022 -0.9832>}
 dyn_belt_ft_06 {parent=dyn_belt_ft_05, loc=<4.0901 0 0>, rot=<0.0118 -0.1764 0.0544 -0.9827>}
 base_belt_ft_06 {parent=dyn_belt_ft_05, loc=<0 0 0>, rot=<0 0 0 1>}
 base_belt_ft_05 {parent=dyn_belt_ft_03, loc=<0 0 0>, rot=<0 0 0 1>}
 base_belt_ft_04 {parent=dyn_belt_ft_02, loc=<0 0 0>, rot=<0 0 0 1>}
 base_belt_ft_03 {parent=dyn_belt_ft_01, loc=<0 0 0>, rot=<0 0 0 1>}
 base_belt_00 {parent=dyn_belt_ft_00, loc=<1.7773 0.4211 1.0295>, rot=<0.2145 -0.0522 0.8966 0.3838>}
 dyn_belt_00 {parent=base_belt_00, loc=<0 0 0>, rot=<0 0 0 1>}
 dyn_belt_01 {parent=dyn_belt_00, loc=<3.1217 0 0>, rot=<0.0567 0.0121 0.3207 -0.9454>}
 dyn_belt_02 {parent=dyn_belt_01, loc=<4.6139 0 0>, rot=<0.0225 -0.0071 0.1233 -0.9921>}
 dyn_belt_03 {parent=dyn_belt_02, loc=<4.4439 0 0>, rot=<0.0179 -0.0054 0.1056 -0.9942>}
 dyn_belt_04 {parent=dyn_belt_03, loc=<8.156 0 0>, rot=<0.01 -0.0282 0.0741 -0.9968>}
 dyn_belt_05 {parent=dyn_belt_04, loc=<8.5403 0 0>, rot=<0.0019 -0.0295 0.0244 -0.9993>}
 dyn_belt_06 {parent=dyn_belt_05, loc=<8.2718 0 0>, rot=<0.0007 -0.0074 0.0164 -0.9998>}
 dyn_belt_07 {parent=dyn_belt_06, loc=<6.6398 0 0>, rot=<-0.503 0.4803 0.497 -0.5189>}
 base_belt_ft_02 {parent=dyn_belt_ft_00, loc=<0 0 0>, rot=<0 0 0 1>}
 base_belt_ft_01 {parent=dyn_belt_ft_base_01, loc=<0 0 0>, rot=<0 0 0 1>}
 dyn_skirt_fr_01_l_MSTR {parent=pelvis, loc=<0.1746 7.9391 0.4223>, rot=<0.0221 0.7062 0.7064 0.0414>}
 dyn_skirt_fr_01_l {parent=dyn_skirt_fr_01_l_MSTR, loc=<0 0 0>, rot=<0 0 0 1>}
 dyn_skirt_fr_02_l {parent=dyn_skirt_fr_01_l, loc=<5.4622 0 0>, rot=<-0.0004 -0.027 -0.0136 -0.9995>}
 dyn_skirt_fr_03_l {parent=dyn_skirt_fr_02_l, loc=<9.4106 0 0>, rot=<0 0 0 1>}
 dyn_skirt_fr_04_l {parent=dyn_skirt_fr_03_l, loc=<0 0 0>, rot=<0 0 0 1>}
 dyn_skirt_fr_05_l {parent=dyn_skirt_fr_04_l, loc=<8.2945 0 0>, rot=<0 0 0 1>}
 base_skirt_fr_02_l {parent=dyn_skirt_fr_02_l, loc=<0 0 0>, rot=<0 0 0 1>}
 dyn_skirt_01_r_side_MSTR {parent=pelvis, loc=<1.4631 -2.3403 14.0246>, rot=<0.1278 -0.005 0.9918 -0.003>}
 dyn_skirt_side_01_r {parent=dyn_skirt_01_r_side_MSTR, loc=<0 0 0>, rot=<0 0 0 1>}
 dyn_skirt_01_r_MSTR {parent=pelvis, loc=<1.2255 3.6192 12.4507>, rot=<0.1076 0.3693 0.923 0.0008>}
 dyn_skirt_01_r {parent=dyn_skirt_01_r_MSTR, loc=<0 0 0>, rot=<0 0 0 1>}
 dyn_skirt_bk_01_r_MSTR {parent=pelvis, loc=<2.0589 -7.7104 10.4295>, rot=<0.1034 -0.4274 0.8974 -0.0366>}
 dyn_skirt_bk_01_r {parent=dyn_skirt_bk_01_r_MSTR, loc=<0 0 0>, rot=<0 0 0 1>}
 dyn_skirt_bk_02_r {parent=dyn_skirt_bk_01_r, loc=<7.4888 0 0>, rot=<0.0168 -0.0388 -0.0035 -0.9991>}
 dyn_skirt_bk_03_r {parent=dyn_skirt_bk_02_r, loc=<8.8841 0 0>, rot=<0.0602 0.0172 -0.0113 -0.998>}
 dyn_skirt_bk_04_r {parent=dyn_skirt_bk_03_r, loc=<7.7391 0 0>, rot=<0 -0.1015 0 -0.9948>}
 dyn_skirt_bk_01_l_MSTR {parent=pelvis, loc=<1.3683 -10.6177 0.3096>, rot=<0.0242 -0.7851 0.6177 -0.0376>}
 dyn_skirt_bk_01_l {parent=dyn_skirt_bk_01_l_MSTR, loc=<0 0 0>, rot=<0 0 0 1>}
 dyn_skirt_bk_02_l {parent=dyn_skirt_bk_01_l, loc=<8.7894 0 0>, rot=<-0.0858 -0.0039 0.0103 -0.9963>}
 dyn_skirt_bk_03_l {parent=dyn_skirt_bk_02_l, loc=<6.5732 0 0>, rot=<0.0068 0.0016 -0.0008 -1>}
 dyn_skirt_bk_04_l {parent=dyn_skirt_bk_03_l, loc=<7.4421 0 0>, rot=<0 -0.0532 0 -0.9986>}
 dyn_skirt_fr_mid_r_MSTR {parent=pelvis, loc=<0.3753 6.6991 7.0002>, rot=<0.1076 0.3693 0.923 0.0008>}
 deform_glute_base_l {parent=pelvis, loc=<-7.3466 -9.0213 -9.1245>, rot=<0.0172 0.031 -0.4477 -0.8935>}
 deform_glute_l {parent=deform_glute_base_l, loc=<0 0 0>, rot=<0 0 0 1>}
 deform_glute_base_r {parent=pelvis, loc=<-7.3464 -9.0213 9.1245>, rot=<0.031 -0.0172 0.8935 -0.4477>}
 deform_glute_r {parent=deform_glute_base_r, loc=<0 0 0>, rot=<0 0 0 1>}
 ik_foot_root {parent=root, loc=<0 0 0>, rot=<0 0 0 1>}
 ik_foot_l {parent=ik_foot_root, loc=<14.8262 -7.3545 8.296>, rot=<0.0357 0.7045 0.0284 -0.7082>}
 ik_foot_r {parent=ik_foot_root, loc=<-14.8262 -7.3546 8.2958>, rot=<0.7082 0.0284 -0.7045 0.0357>}
 ik_hand_root {parent=root, loc=<0 0 0>, rot=<0 0 0 1>}
 ik_hand_gun {parent=ik_hand_root, loc=<-48.1744 3.3546 104.4381>, rot=<0.6276 0.1309 0.4183 -0.6434>}
 ik_hand_l {parent=ik_hand_gun, loc=<59.3272 67.6997 34.3524>, rot=<-0.6158 -0.7027 -0.3565 0>}
 ik_hand_r {parent=ik_hand_gun, loc=<0 0 0>, rot=<0 0 0 1>}
]
variables [isEmoting {type=double, default=0}]
