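# Animation Blueprint for the VenomSoul female face accessory, adding secondary motion to its hood and hair.
# A cached per-bone blend of parent and Body mesh poses at clavicles and upper arms feeds an emote-aware Control Rig.
# Chained spring-driven AnimDynamics nodes then simulate hood and hair bones, followed by fixed additive bone offsets.
# Events update emote state and parameter-asset gravity settings, and cache Body and backbling state on customization load.
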AnimBlueprint /BRCosmetics/Characters/Player/Female/Medium/Bodies/F_MED_VenomSoul/Meshes/Parts/F_MED_VenomSoul_FaceAcc_AnimBP
target_skeleton=&/F_MED_VenomSoul_FaceAcc_Skeleton
graphs [
 AnimGraph {nodes [
   n1 = Control Rig!(Source=Use cached pose 'InitialPose'(), Alpha, isEmoting:double=0) -> (Pose) {controlRigAssetReference {blueprintRigClass=&/F_MED_VenomSoul_FaceAcc_CtrlRig.F_MED_VenomSoul_FaceAcc_CtrlRig_C}}
   n2 = Anim Dynamics - Bone: dyn_hood_bk!(ComponentPose=Local To Component(LocalPose=@n1), Alpha=1, GravityOverride:Vector="0,0,-980", ExternalForce="0,0,0") -> (Pose) {
    angularDampingOverride=0.9999
    angularSpringConstant=10000
    bAngularSpring=true
    bOverrideAngularDamping=true
    bUseGravityOverride=true
    gravityScale=0.35
    simulationSpace=RootRelative
    boundBone {boneName=dyn_hood_bk}
    componentLinearAccScale=<1 1 1>
    componentLinearVelScale=<1 1 1>
    gravityOverride {z=-980}
    physicsBodyDefinitions [{
      boundBone {boneName=dyn_hood_bk}
      boxExtents=<40 40 40>
      constraintSetup {angularLimitsMax=<0 30 20>, angularLimitsMin=<0 -30 -20>, angularTarget=<1 0 0>}
      localJointOffset=<4 0 0>
     }]
   }
   n3 = Anim Dynamics - Bone: dyn_hair_fr_00!(ComponentPose=@n2, Alpha=1, GravityOverride:Vector="0,0,-980", ExternalForce="0,0,0") -> (Pose) {
    angularDampingOverride=0.9999
    angularSpringConstant=5000
    bAngularSpring=true
    bOverrideAngularDamping=true
    bUseGravityOverride=true
    gravityScale=0.35
    simulationSpace=RootRelative
    boundBone {boneName=dyn_hair_fr_00}
    componentAppliedLinearAccClamp=<1500 1500 1500>
    componentLinearAccScale=<1 1 1>
    componentLinearVelScale=<1 1 1>
    gravityOverride {z=-980}
    physicsBodyDefinitions [{
      boundBone {boneName=dyn_hair_fr_00}
      boxExtents=<35 35 35>
      constraintSetup {angularLimitsMax=<0 0 1.5>, angularLimitsMin=<0 -1.5 -1.5>, angularTarget=<1 0 0>}
      localJointOffset=<4 0 0>
     }]
   }
   n4 = Anim Dynamics - Bone: dyn_hair_fr_01!(ComponentPose=@n3, Alpha=1, GravityOverride:Vector="0,0,-980", ExternalForce="0,0,0") -> (Pose) {
    angularDampingOverride=0.9999
    angularSpringConstant=5000
    bAngularSpring=true
    bOverrideAngularDamping=true
    bUseGravityOverride=true
    gravityScale=0.35
    simulationSpace=RootRelative
    boundBone {boneName=dyn_hair_fr_01}
    componentAppliedLinearAccClamp=<1500 1500 1500>
    componentLinearAccScale=<1 1 1>
    componentLinearVelScale=<1 1 1>
    gravityOverride {z=-980}
    physicsBodyDefinitions [{
      boundBone {boneName=dyn_hair_fr_01}
      boxExtents=<35 35 35>
      constraintSetup {angularLimitsMax=<0 0 3>, angularLimitsMin=<0 -3 -3>, angularTarget=<1 0 0>}
      localJointOffset=<4 0 0>
     }]
   }
   n5 = Anim Dynamics - Bone: dyn_hair_fr_02!(ComponentPose=@n4, Alpha=1, GravityOverride:Vector="0,0,-980", ExternalForce="0,0,0") -> (Pose) {
    angularDampingOverride=0.9999
    angularSpringConstant=5000
    bAngularSpring=true
    bOverrideAngularDamping=true
    bUseGravityOverride=true
    gravityScale=0.35
    simulationSpace=RootRelative
    boundBone {boneName=dyn_hair_fr_02}
    componentAppliedLinearAccClamp=<1500 1500 1500>
    componentLinearAccScale=<1 1 1>
    componentLinearVelScale=<1 1 1>
    gravityOverride {z=-980}
    physicsBodyDefinitions [{
      boundBone {boneName=dyn_hair_fr_02}
      boxExtents=<35 35 35>
      constraintSetup {angularLimitsMax=<0 0 5>, angularLimitsMin=<0 -5 -5>, angularTarget=<1 0 0>}
      localJointOffset=<4 0 0>
     }]
   }
   n6 = Anim Dynamics - Bone: dyn_hair_fr_05!(ComponentPose=@n5, Alpha=1, GravityOverride:Vector="0,0,-980", ExternalForce="0,0,0") -> (Pose) {
    angularDampingOverride=0.9999
    angularSpringConstant=5000
    bAngularSpring=true
    bOverrideAngularDamping=true
    bUseGravityOverride=true
    gravityScale=0.35
    simulationSpace=RootRelative
    boundBone {boneName=dyn_hair_fr_05}
    componentAppliedLinearAccClamp=<1500 1500 1500>
    componentLinearAccScale=<1 1 1>
    componentLinearVelScale=<1 1 1>
    gravityOverride {z=-980}
    physicsBodyDefinitions [{
      boundBone {boneName=dyn_hair_fr_05}
      boxExtents=<35 35 35>
      constraintSetup {angularLimitsMax=<0 0 5>, angularLimitsMin=<0 -5 -5>, angularTarget=<1 0 0>}
      localJointOffset=<4 0 0>
     }]
   }
   n7 = Anim Dynamics - Bone: dyn_hair_fr_03!(ComponentPose=@n6, Alpha=1, GravityOverride:Vector="0,0,-980", ExternalForce="0,0,0") -> (Pose) {
    angularDampingOverride=0.9999
    angularSpringConstant=5000
    bAngularSpring=true
    bOverrideAngularDamping=true
    bUseGravityOverride=true
    gravityScale=0.35
    simulationSpace=RootRelative
    boundBone {boneName=dyn_hair_fr_03}
    componentAppliedLinearAccClamp=<1500 1500 1500>
    componentLinearAccScale=<1 1 1>
    componentLinearVelScale=<1 1 1>
    gravityOverride {z=-980}
    physicsBodyDefinitions [{
      boundBone {boneName=dyn_hair_fr_03}
      boxExtents=<35 35 35>
      constraintSetup {angularLimitsMax=<0 0 5>, angularLimitsMin=<0 -5 -5>, angularTarget=<1 0 0>}
      localJointOffset=<4 0 0>
     }]
   }
   n8 = Anim Dynamics - Bone: dyn_hair_r_01!(ComponentPose=@n7, Alpha=1, GravityOverride:Vector="0,0,-980", ExternalForce="0,0,0") -> (Pose) {
    angularDampingOverride=0.9999
    angularSpringConstant=5000
    bAngularSpring=true
    bOverrideAngularDamping=true
    bUseGravityOverride=true
    gravityScale=0.35
    simulationSpace=RootRelative
    boundBone {boneName=dyn_hair_r_01}
    componentAppliedLinearAccClamp=<1500 1500 1500>
    componentLinearAccScale=<1 1 1>
    componentLinearVelScale=<1 1 1>
    gravityOverride {z=-980}
    physicsBodyDefinitions [{
      boundBone {boneName=dyn_hair_r_01}
      boxExtents=<30 30 30>
      constraintSetup {angularLimitsMax=<0 1.5 1.5>, angularLimitsMin=<0 -1.5 -1.5>, angularTarget=<1 0 0>}
      localJointOffset=<4 0 0>
     }]
   }
   n9 = Anim Dynamics - Bone: dyn_hair_r_02!(ComponentPose=@n8, Alpha=1, GravityOverride:Vector="0,0,-980", ExternalForce="0,0,0") -> (Pose) {
    angularDampingOverride=0.9999
    angularSpringConstant=5000
    bAngularSpring=true
    bOverrideAngularDamping=true
    bUseGravityOverride=true
    gravityScale=0.35
    simulationSpace=RootRelative
    boundBone {boneName=dyn_hair_r_02}
    componentAppliedLinearAccClamp=<1500 1500 1500>
    componentLinearAccScale=<1 1 1>
    componentLinearVelScale=<1 1 1>
    gravityOverride {z=-980}
    physicsBodyDefinitions [{
      boundBone {boneName=dyn_hair_r_02}
      boxExtents=<30 30 30>
      constraintSetup {angularLimitsMax=<0 3 3>, angularLimitsMin=<0 -3 -3>, angularTarget=<1 0 0>}
      localJointOffset=<4 0 0>
     }]
   }
   n10 = Anim Dynamics - Bone: dyn_hair_r_05!(ComponentPose=@n9, Alpha=1, GravityOverride:Vector="0,0,-980", ExternalForce="0,0,0") -> (Pose) {
    angularDampingOverride=0.9999
    angularSpringConstant=5000
    bAngularSpring=true
    bOverrideAngularDamping=true
    bUseGravityOverride=true
    gravityScale=0.35
    simulationSpace=RootRelative
    boundBone {boneName=dyn_hair_r_05}
    componentAppliedLinearAccClamp=<1500 1500 1500>
    componentLinearAccScale=<1 1 1>
    componentLinearVelScale=<1 1 1>
    gravityOverride {z=-980}
    physicsBodyDefinitions [{
      boundBone {boneName=dyn_hair_r_05}
      boxExtents=<40 40 40>
      constraintSetup {angularLimitsMax=<0 0 5>, angularLimitsMin=<0 -5 -5>, angularTarget=<1 0 0>}
      localJointOffset=<4 0 0>
     }]
   }
   n11 = Anim Dynamics - Bone: dyn_hair_r_03!(ComponentPose=@n10, Alpha=1, GravityOverride:Vector="0,0,-980", ExternalForce="0,0,0") -> (Pose) {
    angularDampingOverride=0.9999
    angularSpringConstant=5000
    bAngularSpring=true
    bOverrideAngularDamping=true
    bUseGravityOverride=true
    gravityScale=0.35
    simulationSpace=RootRelative
    boundBone {boneName=dyn_hair_r_03}
    componentAppliedLinearAccClamp=<1500 1500 1500>
    componentLinearAccScale=<1 1 1>
    componentLinearVelScale=<1 1 1>
    gravityOverride {z=-980}
    physicsBodyDefinitions [{
      boundBone {boneName=dyn_hair_r_03}
      boxExtents=<40 40 40>
      constraintSetup {angularLimitsMax=<0 0 5>, angularLimitsMin=<0 -5 -5>, angularTarget=<1 0 0>}
      localJointOffset=<4 0 0>
     }]
   }
   n12 = Anim Dynamics - Bone: dyn_hair_r_07!(ComponentPose=@n11, Alpha=1, GravityOverride:Vector="0,0,-980", ExternalForce="0,0,0") -> (Pose) {
    angularDampingOverride=0.9999
    angularSpringConstant=5000
    bAngularSpring=true
    bOverrideAngularDamping=true
    bUseGravityOverride=true
    gravityScale=0.35
    simulationSpace=RootRelative
    boundBone {boneName=dyn_hair_r_07}
    componentAppliedLinearAccClamp=<1500 1500 1500>
    componentLinearAccScale=<1 1 1>
    componentLinearVelScale=<1 1 1>
    gravityOverride {z=-980}
    physicsBodyDefinitions [{
      boundBone {boneName=dyn_hair_r_07}
      boxExtents=<40 40 40>
      constraintSetup {angularLimitsMax=<0 0 5>, angularLimitsMin=<0 -5 -5>, angularTarget=<1 0 0>}
      localJointOffset=<4 0 0>
     }]
   }
   n13 = Anim Dynamics - Bone: dyn_hair_fr_l_00!(ComponentPose=@n12, Alpha=1, GravityOverride:Vector="0,0,-980", ExternalForce="0,0,0") -> (Pose) {
    angularDampingOverride=0.9999
    angularSpringConstant=5000
    bAngularSpring=true
    bOverrideAngularDamping=true
    bUseGravityOverride=true
    gravityScale=0.35
    simulationSpace=RootRelative
    boundBone {boneName=dyn_hair_fr_l_00}
    componentAppliedLinearAccClamp=<1500 1500 1500>
    componentLinearAccScale=<1 1 1>
    componentLinearVelScale=<1 1 1>
    gravityOverride {z=-980}
    physicsBodyDefinitions [{
      boundBone {boneName=dyn_hair_fr_l_00}
      boxExtents=<40 40 40>
      constraintSetup {angularLimitsMax=<0 0 1.5>, angularLimitsMin=<0 -1.5 -1.5>, angularTarget=<1 0 0>}
      localJointOffset=<2 0 0>
     }]
   }
   n14 = Anim Dynamics - Bone: dyn_hair_fr_l_01!(ComponentPose=@n13, Alpha=1, GravityOverride:Vector="0,0,-980", ExternalForce="0,0,0") -> (Pose) {
    angularDampingOverride=0.9999
    angularSpringConstant=5000
    bAngularSpring=true
    bOverrideAngularDamping=true
    bUseGravityOverride=true
    gravityScale=0.35
    simulationSpace=RootRelative
    boundBone {boneName=dyn_hair_fr_l_01}
    componentAppliedLinearAccClamp=<1500 1500 1500>
    componentLinearAccScale=<1 1 1>
    componentLinearVelScale=<1 1 1>
    gravityOverride {z=-980}
    physicsBodyDefinitions [{
      boundBone {boneName=dyn_hair_fr_l_01}
      boxExtents=<40 40 40>
      constraintSetup {angularLimitsMax=<0 0 3>, angularLimitsMin=<0 -3 -3>, angularTarget=<1 0 0>}
      localJointOffset=<4 0 0>
     }]
   }
   n15 = Anim Dynamics - Bone: dyn_hair_fr_l_02!(ComponentPose=@n14, Alpha=1, GravityOverride:Vector="0,0,-980", ExternalForce="0,0,0") -> (Pose) {
    angularDampingOverride=0.9999
    angularSpringConstant=5000
    bAngularSpring=true
    bOverrideAngularDamping=true
    bUseGravityOverride=true
    gravityScale=0.35
    simulationSpace=RootRelative
    boundBone {boneName=dyn_hair_fr_l_02}
    componentAppliedLinearAccClamp=<1500 1500 1500>
    componentLinearAccScale=<1 1 1>
    componentLinearVelScale=<1 1 1>
    gravityOverride {z=-980}
    physicsBodyDefinitions [{
      boundBone {boneName=dyn_hair_fr_l_02}
      boxExtents=<40 40 40>
      constraintSetup {angularLimitsMax=<0 0 5>, angularLimitsMin=<0 -5 -5>, angularTarget=<1 0 0>}
      localJointOffset=<6 0 0>
     }]
   }
   n16 = Anim Dynamics - Bone: dyn_hair_l_00(ComponentPose=@n15, Alpha=1, ExternalForce="0,0,0") {
    angularDampingOverride=0.9999
    angularSpringConstant=5000
    bAngularSpring=true
    bOverrideAngularDamping=true
    bUseGravityOverride=true
    gravityScale=0.35
    simulationSpace=RootRelative
    boundBone {boneName=dyn_hair_l_00}
    componentAppliedLinearAccClamp=<1500 1500 1500>
    componentLinearAccScale=<1 1 1>
    componentLinearVelScale=<1 1 1>
    gravityOverride {z=-980}
    physicsBodyDefinitions [{
      boundBone {boneName=dyn_hair_l_00}
      boxExtents=<40 40 40>
      constraintSetup {angularLimitsMax=<0 5 3>, angularLimitsMin=<0 -5 -5>, angularTarget=<1 0 0>}
      localJointOffset=<4 0 0>
     }]
   }
   n17 = Anim Dynamics - Bone: dyn_hair_l_01(ComponentPose=@n16, Alpha=1, ExternalForce="0,0,0") {
    angularDampingOverride=0.9999
    angularSpringConstant=5000
    bAngularSpring=true
    bOverrideAngularDamping=true
    bUseGravityOverride=true
    gravityScale=0.35
    simulationSpace=RootRelative
    boundBone {boneName=dyn_hair_l_01}
    componentAppliedLinearAccClamp=<1500 1500 1500>
    componentLinearAccScale=<1 1 1>
    componentLinearVelScale=<1 1 1>
    gravityOverride {z=-980}
    physicsBodyDefinitions [{
      boundBone {boneName=dyn_hair_l_01}
      boxExtents=<40 40 40>
      constraintSetup {angularLimitsMax=<0 5 3>, angularLimitsMin=<0 -5 -5>, angularTarget=<1 0 0>}
      localJointOffset=<4 0 0>
     }]
   }
   n18 = Anim Dynamics - Bone: dyn_hair_l_04(ComponentPose=@n17, Alpha=1, ExternalForce="0,0,0") {
    angularDampingOverride=0.9999
    angularSpringConstant=5000
    bAngularSpring=true
    bOverrideAngularDamping=true
    bUseGravityOverride=true
    gravityScale=0.35
    simulationSpace=RootRelative
    boundBone {boneName=dyn_hair_l_04}
    componentAppliedLinearAccClamp=<1500 1500 1500>
    componentLinearAccScale=<1 1 1>
    componentLinearVelScale=<1 1 1>
    gravityOverride {z=-980}
    physicsBodyDefinitions [{
      boundBone {boneName=dyn_hair_l_04}
      boxExtents=<40 40 40>
      constraintSetup {angularLimitsMax=<0 5 3>, angularLimitsMin=<0 -5 -5>, angularTarget=<1 0 0>}
      localJointOffset=<4 0 0>
     }]
   }
   n19 = Anim Dynamics - Bone: dyn_hair_l_02(ComponentPose=@n18, Alpha=1, ExternalForce="0,0,0") {
    angularDampingOverride=0.9999
    angularSpringConstant=5000
    bAngularSpring=true
    bOverrideAngularDamping=true
    bUseGravityOverride=true
    gravityScale=0.35
    simulationSpace=RootRelative
    boundBone {boneName=dyn_hair_l_02}
    componentAppliedLinearAccClamp=<1500 1500 1500>
    componentLinearAccScale=<1 1 1>
    componentLinearVelScale=<1 1 1>
    gravityOverride {z=-980}
    physicsBodyDefinitions [{
      boundBone {boneName=dyn_hair_l_02}
      boxExtents=<40 40 40>
      constraintSetup {angularLimitsMax=<0 5 3>, angularLimitsMin=<0 -5 -5>, angularTarget=<1 0 0>}
      localJointOffset=<4 0 0>
     }]
   }
   n20 = "Transform (Modify) Bone - Bone: dyn_hair_fr_00"(ComponentPose=@n19, Alpha=1, Translation="0,0,0", Rotation="-5,0,0", Scale="0,0,0") {
    rotationMode=BMM_Additive
    rotationSpace=BCS_BoneSpace
    scaleSpace=BCS_BoneSpace
    translationMode=BMM_Additive
    translationSpace=BCS_BoneSpace
    alphaBoolBlend {blendOption=Linear}
    boneToModify {boneName=dyn_hair_fr_00}
    rotation {pitch=-5}
    scale=<0 0 0>
   }
   n21 = "Transform (Modify) Bone - Bone: dyn_hair_r_01"(ComponentPose=@n20, Alpha=1, Translation="0,0,0", Rotation="-5,0,0", Scale="0,0,0") {
    rotationMode=BMM_Additive
    rotationSpace=BCS_BoneSpace
    scaleSpace=BCS_BoneSpace
    translationMode=BMM_Additive
    translationSpace=BCS_BoneSpace
    alphaBoolBlend {blendOption=Linear}
    boneToModify {boneName=dyn_hair_r_01}
    rotation {pitch=-5}
    scale=<0 0 0>
   }
   n22 = "Transform (Modify) Bone - Bone: dyn_hair_l_00"(ComponentPose=@n21, Alpha=1, Translation="0,0,0", Rotation="0,10,0", Scale="0,0,0") {
    rotationMode=BMM_Additive
    rotationSpace=BCS_BoneSpace
    scaleSpace=BCS_BoneSpace
    translationMode=BMM_Additive
    translationSpace=BCS_BoneSpace
    alphaBoolBlend {blendOption=Linear}
    boneToModify {boneName=dyn_hair_l_00}
    rotation {yaw=10}
    scale=<0 0 0>
   }
   n23 = "Transform (Modify) Bone - Bone: dyn_hair_fr_l_00"(ComponentPose=@n22, Alpha=1, Translation="0,0,0", Rotation="-5,0,0", Scale="0,0,0") {
    rotationMode=BMM_Additive
    rotationSpace=BCS_BoneSpace
    scaleSpace=BCS_BoneSpace
    translationMode=BMM_Additive
    translationSpace=BCS_BoneSpace
    alphaBoolBlend {blendOption=Linear}
    boneToModify {boneName=dyn_hair_fr_l_00}
    rotation {pitch=-5}
    scale=<0 0 0>
   }
   n24 = "Transform (Modify) Bone - Bone: dyn_hair_fr_l_02"(ComponentPose=@n23, Alpha=1, Translation="0,0,0", Rotation="0,20,0", Scale="0,0,0") {
    rotationMode=BMM_Additive
    rotationSpace=BCS_BoneSpace
    scaleSpace=BCS_BoneSpace
    translationMode=BMM_Additive
    translationSpace=BCS_BoneSpace
    alphaBoolBlend {blendOption=Linear}
    boneToModify {boneName=dyn_hair_fr_l_02}
    rotation {yaw=20}
    scale=<0 0 0>
   }
   n25 = "Transform (Modify) Bone - Bone: dyn_hair_r_11"(ComponentPose=@n24, Alpha=1, Translation="0,0,-2", Rotation="0,0,0", Scale="0,0,0") {
    rotationMode=BMM_Additive
    rotationSpace=BCS_BoneSpace
    scaleSpace=BCS_BoneSpace
    translationMode=BMM_Additive
    translationSpace=BCS_BoneSpace
    alphaBoolBlend {blendOption=Linear}
    boneToModify {boneName=dyn_hair_r_11}
    scale=<0 0 0>
    translation {z=-2}
   }
   n26 = "Transform (Modify) Bone - Bone: dyn_hair_bk_r_01"(ComponentPose=@n25, Alpha=1, Translation="0,0,-4", Rotation="0,0,0", Scale="0,0,0") {
    rotationMode=BMM_Additive
    rotationSpace=BCS_BoneSpace
    scaleSpace=BCS_BoneSpace
    translationMode=BMM_Additive
    translationSpace=BCS_BoneSpace
    alphaBoolBlend {blendOption=Linear}
    boneToModify {boneName=dyn_hair_bk_r_01}
    scale=<0 0 0>
    translation {z=-4}
   }
   n27 = "Transform (Modify) Bone - Bone: dyn_hair_bk_low_10"(ComponentPose=@n26, Alpha=1, Translation="0,0,-2", Rotation="0,0,0", Scale="0,0,0") {
    rotationMode=BMM_Additive
    rotationSpace=BCS_BoneSpace
    scaleSpace=BCS_BoneSpace
    translationMode=BMM_Additive
    translationSpace=BCS_BoneSpace
    alphaBoolBlend {blendOption=Linear}
    boneToModify {boneName=dyn_hair_bk_low_10}
    scale=<0 0 0>
    translation {z=-2}
   }
   n28 = "Transform (Modify) Bone - Bone: dyn_hair_bk_low_04"(ComponentPose=@n27, Alpha=1, Translation="0,0,-4", Rotation="0,0,0", Scale="0,0,0") {
    rotationMode=BMM_Additive
    rotationSpace=BCS_BoneSpace
    scaleSpace=BCS_BoneSpace
    translationMode=BMM_Additive
    translationSpace=BCS_BoneSpace
    alphaBoolBlend {blendOption=Linear}
    boneToModify {boneName=dyn_hair_bk_low_04}
    scale=<0 0 0>
    translation {z=-4}
   }
   Output Pose(Result=Component To Local(ComponentPose=@n28)) {layerGroup=None}
   n29 = Layered blend per bone(
    BasePose=Copy Pose From Mesh() {bUseAttachedParent=true}
    BlendPoses_0=Copy Pose From Mesh(SourceMeshComponent=Get Body() -> (Body:Skeletal Mesh Component Budgeted Object Reference)) {bUseAttachedParent=true}
   ) {layerSetup [{branchFilters [
       {blendDepth=2, boneName=clavicle_r}
       {blendDepth=2, boneName=clavicle_l}
       {blendDepth=2, boneName=upperarm_l}
       {blendDepth=2, boneName=upperarm_r}
      ]}]}
   Save cached pose 'InitialPose'(Pose=@n29) {cacheName=InitialPose}
  ]}
 EventGraph {nodes [
   Event BlueprintUpdateAnimation()
   > n1 = Sequence()
   > n2 = "Gravity Override State Function (Parameter Asset) (Nativized)"(
    CharacterPartAnimInstance=Self-Reference()
    GameTimeInSeconds=GetGameTimeInSeconds()
    Parameters=/BRCosmetics/Characters/Player/Female/Medium/Bodies/F_MED_VenomSoul/Meshes/F_MED_VenomSoul_Parameters
   )
   > Set GravityOverride(GravityOverride=@n2)
   > Set RBAN Settings Out(RBAN Settings Out=@n2.OutRigidBodyAnimNodeParameters)
   Set Is Emoting(execute=@n1.then_1, Is Emoting:double="To Float (Boolean)"(InBool=Get bIsPlayingEmote())) -> (Output_Get:double)
   n3 = Event OnCharacterCustomizationLoaded()
   > n4 = Sequence!(execute) -> (then_0, then_1, then_2:Exec, then_3:Exec)
   > Set isWearingBackbling(isWearingBackbling=CheckForBackbling())
   Set Body(
    execute=@n4.then_1
    Body:Skeletal Mesh Component Budgeted Object Reference=Get Array Item with EFort Custom Part Type(Array=@n3.CustomMeshParts)
   ) -> (Output_Get:Skeletal Mesh Component Budgeted Object Reference)
  ]}
]
variables [
 Body {type=SkeletalMeshComponentBudgeted}
 GravityOverride {type=Vector}
 Is Emoting {type=number}
 isWearingBackbling {type=boolean}
 RBAN Settings Out {type=FortRigidBodyAnimNodeParameters}
]
