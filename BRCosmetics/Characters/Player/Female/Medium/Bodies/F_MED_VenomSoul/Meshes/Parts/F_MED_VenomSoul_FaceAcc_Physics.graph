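# Physics asset for the face accessory mesh of the VenomSoul medium female player body, covering dynamic hair and hood bones.
# Mostly capsule bodies cover the upper spine, clavicles, neck and head, plus hair and hood, with one box on left hair.
# Bangs, front and side hair form chains from the head with tight swing and twist limits, as do both hood pieces separately.
# Body and back-hair constraints are free on all angular axes.
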PhysicsAsset /BRCosmetics/Characters/Player/Female/Medium/Bodies/F_MED_VenomSoul/Meshes/Parts/F_MED_VenomSoul_FaceAcc_Physics
bodies {
 clavicle_l [{
   shapeType=Capsule
   length=8.5089
   radius=2.9223
   shapeName=clavicle_l_capsule
   center=<5.9026 3.1697 14.2461>
   rotation=<rot -22.0666 -15.1833 -80.8674>
  }]
 clavicle_r [{
   shapeType=Capsule
   length=7.4202
   radius=2.292
   shapeName=clavicle_r_capsule
   center=<-6.2065 1.8279 -3.0958>
   rotation=<rot 0 164.3195 -77.9842>
  }]
 dyn_hair_bangs_l_00 [{shapeType=Capsule, length=2.75, radius=0.75, center=<1.2845 0.1205 -0.0032>, rotation=<rot 57.9091 -89.9976 179.9999>}]
 dyn_hair_bangs_l_01 [{shapeType=Capsule, length=2.2317, radius=1.2683, center=<1.2845 0.1205 -0.0032>, rotation=<rot 57.9091 -89.9976 179.9999>}]
 dyn_hair_bangs_r_00 [{shapeType=Capsule, length=2.75, radius=0.75, center=<1.2845 0.1205 -0.0032>, rotation=<rot 87.8729 90.0585 0.0005>}]
 dyn_hair_bangs_r_01 [{shapeType=Capsule, length=2.2317, radius=1.2683, center=<1.2845 0.1205 -0.0032>, rotation=<rot 85.8709 90.0286 0.0001>}]
 dyn_hair_bk_00 [
  {shapeType=Capsule, length=6, radius=5.255, center=<3.1373 -6.4001 -11.4139>, rotation=<rot 2.4325 1.8202 -85.1202>}
  {shapeType=Capsule, length=10, radius=5.505, center=<0.0593 -12.255 -10.6497>, rotation=<rot -68.0441 0.4369 -7.6622>}
 ]
 dyn_hair_bk_l_00 [
  {shapeType=Capsule, length=6, radius=5.255, center=<7.0099 6.2527 -0.4892>, rotation=<rot -23.8702 1.4728 -6.1956>}
  {shapeType=Capsule, length=6, radius=5.255, center=<8.7983 9.7021 5.8246>, rotation=<rot -65.6894 166.2737 -145.5564>}
 ]
 dyn_hair_bk_r_01 [{shapeType=Capsule, length=0.505, radius=0.505, rotation=<rot 0 133.8733 -81.014>}]
 dyn_hair_fr_01 [{shapeType=Capsule, length=3, radius=0.75, center=<1.5056 0.1206 -0.0032>, rotation=<rot 90 0 -90>}]
 dyn_hair_fr_02 [{shapeType=Capsule, length=3, radius=0.75, center=<1.5056 0.1206 -0.0032>, rotation=<rot 90 0 -90>}]
 dyn_hair_fr_03 [{shapeType=Capsule, length=3, radius=0.75, center=<1.5056 0.1206 -0.0032>, rotation=<rot 90 0 -90>}]
 dyn_hair_fr_05 [
  {shapeType=Capsule, length=2, radius=0.75, center=<1.5056 0.1206 -0.0032>, rotation=<rot 90 0 -90>}
  {
   shapeType=Capsule
   length=4.2527
   radius=1.1067
   shapeName=dyn_hair_fr_05_capsule
   center=<3.1593 1.591 0.0101>
   rotation=<rot -89.8164 0 11.1974>
  }
 ]
 dyn_hair_fr_l_00 [{shapeType=Capsule, length=2.7223, radius=1.434, center=<1.0568 1.171 0.0583>, rotation=<rot 0 -71.543 -78.5485>}]
 dyn_hair_fr_l_01 [{shapeType=Capsule, length=2.75, radius=0.5, center=<1.2845 0.1205 -0.0032>, rotation=<rot 57.9091 -89.9976 179.9999>}]
 dyn_hair_fr_l_02 [{shapeType=Capsule, length=2.75, radius=0.5, center=<1.2845 0.1205 -0.0032>, rotation=<rot 57.9091 -89.9976 179.9999>}]
 dyn_hair_l_00 [{shapeType=Box, extentX=6.1017, extentY=6.794, extentZ=0.856, center=<2.0834 0 -0.4217>, rotation=<rot -15 0 0>}]
 dyn_hair_l_01 [{shapeType=Capsule, length=7.5, radius=0.75, center=<2.3941 0.1206 -0.0033>, rotation=<rot 90 90 0>}]
 dyn_hair_r_01 [{shapeType=Capsule, length=4.074, radius=2.2356, center=<3.6702 -0.381 1.9313>, rotation=<rot 0 77.571 -76.3909>}]
 dyn_hair_r_02 [{shapeType=Capsule, length=8, radius=0.75, center=<2.3941 0.1206 -0.0033>, rotation=<rot 90 90 0>}]
 dyn_hair_r_03 [{shapeType=Capsule, length=8, radius=0.75, center=<2.3941 0.1206 -0.0033>, rotation=<rot 90 90 0>}]
 dyn_hood_bk [{shapeType=Capsule, length=4.8011, radius=4.6299, center=<0.2327 -5.0753 1.45>, rotation=<rot 17.0558 -86.9128 -90>}]
 dyn_hood_ft [{shapeType=Capsule, length=4.8011, radius=4.6299, center=<-0.2981 4.7664 -1.5641>, rotation=<rot 17.0558 -86.9128 -90>}]
 head [{shapeType=Capsule, length=11.5147, radius=6.6133, center=<0.9373 1.8179 0.1133>, rotation=<rot 90 0 0>}]
 neck_01 [{shapeType=Capsule, length=4.0295, radius=6.5477, center=<2.1336 0.9263 0.0001>, rotation=<rot 90 0 13.5236>}]
 spine_05 [{shapeType=Capsule, length=10, radius=10, center=<2.3756 -0.0002 0>}]
}
constraints [
 {
  bone1Name=clavicle_l
  bone2Name=spine_05
  swing1LimitDegrees=45
  swing1Motion=Free
  swing2LimitDegrees=45
  swing2Motion=Free
  twistLimitDegrees=45
  twistMotion=Free
 }
 {
  bone1Name=clavicle_r
  bone2Name=spine_05
  swing1LimitDegrees=45
  swing1Motion=Free
  swing2LimitDegrees=45
  swing2Motion=Free
  twistLimitDegrees=45
  twistMotion=Free
 }
 {
  bone1Name=neck_01
  bone2Name=spine_05
  swing1LimitDegrees=45
  swing1Motion=Free
  swing2LimitDegrees=45
  swing2Motion=Free
  twistLimitDegrees=45
  twistMotion=Free
 }
 {
  bone1Name=head
  bone2Name=neck_01
  swing1LimitDegrees=45
  swing1Motion=Free
  swing2LimitDegrees=45
  swing2Motion=Free
  twistLimitDegrees=45
  twistMotion=Free
 }
 {bone1Name=dyn_hair_bangs_l_00, bone2Name=head, swing1LimitDegrees=25, swing2LimitDegrees=25, twistLimitDegrees=5}
 {bone1Name=dyn_hair_bangs_l_01, bone2Name=dyn_hair_bangs_l_00, swing1LimitDegrees=25, swing2LimitDegrees=25, twistLimitDegrees=5}
 {bone1Name=dyn_hair_bangs_r_00, bone2Name=head, swing1LimitDegrees=25, swing2LimitDegrees=10, twistLimitDegrees=5}
 {bone1Name=dyn_hair_bangs_r_01, bone2Name=dyn_hair_bangs_r_00, swing1LimitDegrees=25, swing2LimitDegrees=10, twistLimitDegrees=5}
 {
  bone1Name=dyn_hair_bk_00
  bone2Name=head
  swing1LimitDegrees=45
  swing1Motion=Free
  swing2LimitDegrees=45
  swing2Motion=Free
  twistLimitDegrees=45
  twistMotion=Free
 }
 {
  bone1Name=dyn_hair_bk_l_00
  bone2Name=head
  swing1LimitDegrees=45
  swing1Motion=Free
  swing2LimitDegrees=45
  swing2Motion=Free
  twistLimitDegrees=45
  twistMotion=Free
 }
 {
  bone1Name=dyn_hair_bk_r_01
  bone2Name=head
  swing1LimitDegrees=45
  swing1Motion=Free
  swing2LimitDegrees=45
  swing2Motion=Free
  twistLimitDegrees=45
  twistMotion=Free
 }
 {bone1Name=dyn_hair_fr_01, bone2Name=head, swing1LimitDegrees=12, swing2LimitDegrees=12, twistLimitDegrees=5}
 {bone1Name=dyn_hair_fr_02, bone2Name=dyn_hair_fr_01, swing1LimitDegrees=12, swing2LimitDegrees=12, twistLimitDegrees=5}
 {bone1Name=dyn_hair_fr_03, bone2Name=dyn_hair_fr_02, swing1LimitDegrees=12, swing2LimitDegrees=12, twistLimitDegrees=5}
 {bone1Name=dyn_hair_fr_05, bone2Name=dyn_hair_fr_02, swing1LimitDegrees=12, swing2LimitDegrees=12, twistLimitDegrees=5}
 {bone1Name=dyn_hair_fr_l_00, bone2Name=head, swing1LimitDegrees=12, swing2LimitDegrees=12, twistLimitDegrees=1}
 {bone1Name=dyn_hair_fr_l_01, bone2Name=dyn_hair_fr_l_00, swing1LimitDegrees=12, swing2LimitDegrees=12, twistLimitDegrees=1}
 {bone1Name=dyn_hair_fr_l_02, bone2Name=dyn_hair_fr_l_01, swing1LimitDegrees=12, swing2LimitDegrees=12, twistLimitDegrees=1}
 {bone1Name=dyn_hair_l_01, bone2Name=dyn_hair_l_00, swing1LimitDegrees=12, swing2LimitDegrees=5, twistLimitDegrees=5}
 {bone1Name=dyn_hair_l_00, bone2Name=head, swing1LimitDegrees=12, swing2LimitDegrees=5, twistLimitDegrees=5}
 {bone1Name=dyn_hair_r_01, bone2Name=head, swing1LimitDegrees=12, swing2LimitDegrees=12, twistLimitDegrees=5}
 {bone1Name=dyn_hair_r_02, bone2Name=dyn_hair_r_01, swing1LimitDegrees=12, swing2LimitDegrees=12, twistLimitDegrees=5}
 {bone1Name=dyn_hair_r_03, bone2Name=dyn_hair_r_02, swing1LimitDegrees=12, swing2LimitDegrees=12, twistLimitDegrees=5}
 {bone1Name=dyn_hood_bk, bone2Name=head, swing1LimitDegrees=10, swing2LimitDegrees=20, twistLimitDegrees=1}
 {bone1Name=dyn_hood_ft, bone2Name=head, swing1LimitDegrees=45, swing2LimitDegrees=45, twistLimitDegrees=1}
]
